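# M_TimeCalloutColorize is a translucent UI material that recolors a channel-masked texture, apparently for time callouts in the Showdown frontend.
# The texture's red channel, scaled by ColorMultiplier and clamped, blends from the default base toward PrimaryColor.
# The green channel, scaled and clamped the same way, blends that result toward HighlightColor.
# Texture, both colors and the multiplier are parameters, so material instances can restyle it.
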Material /CompeteOldTab/Frontend/Showdown/Materials/M_TimeCalloutColorize
graph {
 properties {blendMode=BLEND_Translucent, materialDomain=MD_UI, pixelDepthOffsetMode=PDOM_Legacy, usedShadingModels=Default Lit}
 nodes [
  n1 = TextureSampleParameter2D() {parameterName=Texture, texture=&T_BlacksRGB}
  n2 = ScalarParameter() {defaultValue=10, parameterName=ColorMultiplier}
  n3 = Multiply(A=@n1.R, B=@n2) {constB=2}
  LinearInterpolate(
   A=LinearInterpolate(B=VectorParameter() {parameterName=PrimaryColor, defaultValue=<rgba 0 0 0 1>}.RGB, Alpha=Clamp(None=@n3))
   B=VectorParameter() {parameterName=HighlightColor, defaultValue=<rgba 0 0 0 1>}.RGB
   Alpha=Clamp(None=Multiply(A=@n1.G, B=@n2) {constB=2})
  )
 ]
}
