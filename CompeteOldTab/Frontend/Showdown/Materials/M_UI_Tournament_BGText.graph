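# M_UI_Tournament_BGText is an additive UI material in the Showdown frontend, named for tournament background text.
# A Texture parameter, defaulting to T_BlacksRGB, supplies RGB that is tinted by a Color parameter with a blue default.
# The graph holds both a Multiply and a Lerp version of this tint, plus an unwired constant.
# No links to the material outputs appear, so the active branch is unclear.
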Material /CompeteOldTab/Frontend/Showdown/Materials/M_UI_Tournament_BGText
graph {
 properties {blendMode=BLEND_Additive, materialDomain=MD_UI, pixelDepthOffsetMode=PDOM_Legacy, usedShadingModels=Default Lit}
 nodes [
  Constant() {r=0.2}
  n1 = VectorParameter() {parameterName=Color, defaultValue=<rgba 0 0.0439 0.932 0.3>}
  n2 = TextureSampleParameter2D() {parameterName=Texture, texture=&T_BlacksRGB}
  LinearInterpolate(B=@n1.RGB, Alpha=@n2.RGB)
  Multiply(A=@n2.RGB, B=@n1.RGB)
 ]
}
