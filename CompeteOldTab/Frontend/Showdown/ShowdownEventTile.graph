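# CommonButtonLegacy tile for one Showdown tournament event, with nested info, violator and round-break widgets.
# SetData passes tournament and event IDs to the info and violator; info dispatchers drive colors, live/ended state and round-break visibility.
# EventIntro staggers entrance by index, while EventResize plays size transitions on tile, round break and info.
# Hover, focus-path and selection events animate the tile and children through a shared Anim Transition macro.
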WidgetBlueprint /CompeteOldTab/Frontend/Showdown/ShowdownEventTile
animations [
 Intro {
  binding_count=1
  display_label=""
  length_seconds=0.5132
  object=&self:Intro
  sequence {bindings [Overlay_InnerContainer {
     id=#1
     possessed_class=Overlay
     tracks [
      {
       type=MovieScene2DTransformTrack
       sections [{
         type=MovieScene2DTransformSection
         end_frame=10
         start_frame=0
         channels [
          Translation.X {type=MovieSceneScriptingFloatChannel}
          Translation.Y {type=MovieSceneScriptingFloatChannel, keys [[0,56.3386],[6,0.5984],[7.6747,0]]}
          Angle {type=MovieSceneScriptingFloatChannel}
          Scale.X {type=MovieSceneScriptingFloatChannel}
          Scale.Y {type=MovieSceneScriptingFloatChannel}
          Shear.X {type=MovieSceneScriptingFloatChannel}
          Shear.Y {type=MovieSceneScriptingFloatChannel}
         ]
        }]
      }
      {
       type=MovieSceneFloatTrack
       display_name=Render Opacity
       sections [{type=MovieSceneFloatSection, end_frame=10, start_frame=0, channels [None {type=MovieSceneScriptingFloatChannel, keys [[0,0],[6.062,1]]}]}]
      }
     ]
    }]}
 }
 TransitionSize {
  binding_count=7
  display_label=""
  length_seconds=0.3
  object=&self:TransitionSize
  sequence {
   bindings [
    Overlay_InnerContainer {id=#2, possessed_class=Overlay, tracks []}
    Violator {id=#3, possessed_class=ShowdownEventTileViolator_C, tracks []}
    Image_Bg {
     id=#4
     display_name=OverlayBGPlate
     possessed_class=Image
     tracks [{
       type=MovieScene2DTransformTrack
       sections [{
         type=MovieScene2DTransformSection
         end_frame=18
         start_frame=0
         channels [
          Translation.X {type=MovieSceneScriptingFloatChannel}
          Translation.Y {type=MovieSceneScriptingFloatChannel}
          Angle {type=MovieSceneScriptingFloatChannel}
          Scale.X {type=MovieSceneScriptingFloatChannel}
          Scale.Y {type=MovieSceneScriptingFloatChannel}
          Shear.X {type=MovieSceneScriptingFloatChannel, keys [[0,-16],[17.998,0]]}
          Shear.Y {type=MovieSceneScriptingFloatChannel}
         ]
        }]
      }]
    }
    OverlaySlot_0 {
     id=#5
     parent_id=#2
     possessed_class=OverlaySlot
     tracks [{
       type=MovieSceneMarginTrack
       display_name=Padding
       sections [{
         type=MovieSceneMarginSection
         end_frame=18
         start_frame=0
         channels [
          Left {type=MovieSceneScriptingFloatChannel, keys [[0,3],[17.998,8]]}
          Top {type=MovieSceneScriptingFloatChannel}
          Right {type=MovieSceneScriptingFloatChannel, keys [[0,3],[17.998,8]]}
          Bottom {type=MovieSceneScriptingFloatChannel}
         ]
        }]
      }]
    }
    SizeBox_0 {
     id=#6
     possessed_class=SizeBox
     tracks [
      {
       type=MovieSceneFloatTrack
       display_name=Width Override
       sections [{
         type=MovieSceneFloatSection
         end_frame=18
         start_frame=0
         channels [None {type=MovieSceneScriptingFloatChannel, keys [[0,148],[17.998,274]]}]
        }]
      }
      {
       type=MovieSceneFloatTrack
       display_name=Height Override
       sections [{
         type=MovieSceneFloatSection
         end_frame=18
         start_frame=0
         channels [None {type=MovieSceneScriptingFloatChannel, keys [[0,212],[17.998,114]]}]
        }]
      }
     ]
    }
    VerticalBoxSlot_5 {
     id=#7
     parent_id=#3
     possessed_class=VerticalBoxSlot
     tracks [{
       type=MovieSceneMarginTrack
       display_name=Padding
       sections [{
         type=MovieSceneMarginSection
         end_frame=18
         start_frame=0
         channels [
          Left {type=MovieSceneScriptingFloatChannel, keys [[0,30],[17.999,88.2363]]}
          Top {type=MovieSceneScriptingFloatChannel}
          Right {type=MovieSceneScriptingFloatChannel}
          Bottom {type=MovieSceneScriptingFloatChannel}
         ]
        }]
      }]
    }
    ShowdownEventTileInfo {
     id=#8
     possessed_class=ShowdownEventTileInfo_C
     tracks [{
       type=MovieSceneMarginTrack
       display_name=Padding
       sections [{
         type=MovieSceneMarginSection
         end_frame=18
         start_frame=0
         channels [
          Left {type=MovieSceneScriptingFloatChannel, keys [[0,0],[17.998,40]]}
          Top {type=MovieSceneScriptingFloatChannel, const [5,2]}
          Right {type=MovieSceneScriptingFloatChannel, keys [[0,0],[17.998,17]]}
          Bottom {type=MovieSceneScriptingFloatChannel, const [0,2]}
         ]
        }]
      }]
    }
   ]
   display_rate {numerator=60}
  }
 }
 Focus {
  binding_count=5
  display_label=""
  length_seconds=0.25
  object=&self:Focus
  sequence {
   bindings [
    SizeBox_0 {
     id=#9
     possessed_class=SizeBox
     tracks [
      {
       type=MovieSceneFloatTrack
       display_name=Width Override
       sections [{type=MovieSceneFloatSection, end_frame=15, start_frame=0, channels [None {type=MovieSceneScriptingFloatChannel, keys [[0,0],[14.999,84]]}]}]
      }
      {
       type=MovieSceneFloatTrack
       display_name=Height Override
       sections [{type=MovieSceneFloatSection, end_frame=15, start_frame=0, channels [None {type=MovieSceneScriptingFloatChannel, keys [[0,0],[14.999,64]]}]}]
      }
     ]
    }
    Image_Tail {
     id=#10
     possessed_class=Image
     tracks [
      {
       type=MovieScene2DTransformTrack
       sections [{
         type=MovieScene2DTransformSection
         channels [
          Translation.X {type=MovieSceneScriptingFloatChannel}
          Translation.Y {type=MovieSceneScriptingFloatChannel}
          Angle {type=MovieSceneScriptingFloatChannel}
          Scale.X {type=MovieSceneScriptingFloatChannel, keys [[5.719,0],[7.775,1]]}
          Scale.Y {type=MovieSceneScriptingFloatChannel, keys [[5.719,0],[7.775,1]]}
          Shear.X {type=MovieSceneScriptingFloatChannel}
          Shear.Y {type=MovieSceneScriptingFloatChannel}
         ]
        }]
      }
      {
       type=MovieSceneWidgetMaterialTrack
       sections [{
         type=MovieSceneParameterSection
         channels [
          Edge {type=MovieSceneScriptingFloatChannel, keys [[5.719,0.946],[12.6,0.1781]]}
          Scale {type=MovieSceneScriptingFloatChannel, keys [[5.719,0.3931],[12.6,1.0871]]}
         ]
        }]
      }
     ]
    }
    OverlaySlot_0 {
     id=#11
     parent_id=#12
     possessed_class=OverlaySlot
     tracks [{
       type=MovieSceneMarginTrack
       display_name=Padding
       sections [{
         type=MovieSceneMarginSection
         channels [
          Left {type=MovieSceneScriptingFloatChannel, keys [[0,0],[15,17]]}
          Top {type=MovieSceneScriptingFloatChannel}
          Right {type=MovieSceneScriptingFloatChannel, keys [[0,0],[15,17]]}
          Bottom {type=MovieSceneScriptingFloatChannel}
         ]
        }]
      }]
    }
    Image_Focus {
     id=#13
     possessed_class=Image
     tracks [
      {
       type=MovieScene2DTransformTrack
       sections [{
         type=MovieScene2DTransformSection
         channels [
          Translation.X {type=MovieSceneScriptingFloatChannel}
          Translation.Y {type=MovieSceneScriptingFloatChannel}
          Angle {type=MovieSceneScriptingFloatChannel}
          Scale.X {type=MovieSceneScriptingFloatChannel, thinned_from=5, keys [[0,1],[12.152,1.1233],[13.805,1.0873],[14.994,1.1]]}
          Scale.Y {type=MovieSceneScriptingFloatChannel, keys [[0,1],[12.152,1.1233],[13.805,1.0873],[14.994,1.1]]}
          Shear.X {type=MovieSceneScriptingFloatChannel}
          Shear.Y {type=MovieSceneScriptingFloatChannel}
         ]
        }]
      }
      {
       type=MovieSceneWidgetMaterialTrack
       sections [{
         type=MovieSceneParameterSection
         channels [
          BottomEdge {type=MovieSceneScriptingFloatChannel, keys [[0,0],[12.6,0.015],[13.905,0.0131],[14.992,0.0143]]}
          LeftEdge {type=MovieSceneScriptingFloatChannel, keys [[0,0],[12.6,0.0058],[14.015,0.0048],[14.992,0.0052]]}
          RightEdge {type=MovieSceneScriptingFloatChannel, keys [[0,0],[12.6,0.0296],[13.905,0.0274],[14.992,0.0286]]}
          TopEdge {type=MovieSceneScriptingFloatChannel, keys [[0,0],[12.6,0.0241],[13.905,0.0216],[14.992,0.0227]]}
          BottomRotation {type=MovieSceneScriptingFloatChannel, keys [[0,0],[12.6,-0.0039],[13.905,-0.0045],[14.992,-0.0045]]}
          RightRotation {type=MovieSceneScriptingFloatChannel, keys [[0,0],[12.6,-0.0054],[13.918,-0.0048],[14.992,-0.0051]]}
          LeftRotation {type=MovieSceneScriptingFloatChannel, keys [[0,0],[12.6,0.0033],[13.892,0.0023],[14.992,0.0028]]}
          TopRotation {type=MovieSceneScriptingFloatChannel, keys [[0,0],[12.6,0.0033],[13.892,0.0023],[14.992,0.0028]]}
         ]
        }]
      }
      {
       type=MovieSceneFloatTrack
       display_name=Render Opacity
       sections [{type=MovieSceneFloatSection, end_frame=15, start_frame=0, channels [None {type=MovieSceneScriptingFloatChannel, keys [[0,0],[3.154,1]]}]}]
      }
     ]
    }
    Overlay_InnerContainer {id=#12, display_name=CommonFocusBorder, possessed_class=Overlay, tracks []}
   ]
   display_rate {numerator=60}
  }
 }
 Hover {
  binding_count=1
  display_label=""
  length_seconds=0.3268
  object=&self:Hover
  sequence {bindings [Image_Hover {
     id=#14
     possessed_class=Image
     tracks [{
       type=MovieSceneFloatTrack
       display_name=Render Opacity
       sections [{type=MovieSceneFloatSection, end_frame=6, start_frame=0, channels [None {type=MovieSceneScriptingFloatChannel, keys [[0,0],[6.5363,1]]}]}]
      }]
    }]}
 }
 Focus_Violator {
  binding_count=3
  display_label=""
  length_seconds=0.25
  object=&self:Focus_Violator
  sequence {
   bindings [
    Violator {id=#15, possessed_class=ShowdownEventTileViolator_C, tracks []}
    SizeBox_2 {
     id=#16
     possessed_class=SizeBox
     tracks [{
       type=MovieSceneByteTrack
       display_name=Clipping
       sections [{type=MovieSceneByteSection, end_frame=15, start_frame=0, channels [None {type=MovieSceneScriptingByteChannel, keys [[0,0],[0.533,1]]}]}]
      }]
    }
    SizeBoxSlot_0 {
     id=#17
     parent_id=#15
     possessed_class=SizeBoxSlot
     tracks [{
       type=MovieSceneMarginTrack
       display_name=Padding
       sections [{
         type=MovieSceneMarginSection
         end_frame=15
         start_frame=0
         channels [
          Left {type=MovieSceneScriptingFloatChannel}
          Top {type=MovieSceneScriptingFloatChannel}
          Right {type=MovieSceneScriptingFloatChannel}
          Bottom {type=MovieSceneScriptingFloatChannel, keys [[0,0],[15,-47.879]]}
         ]
        }]
      }]
    }
   ]
   display_rate {numerator=60}
  }
 }
]
graphs [
 EventGraph {nodes [
   n1 = EventIntro!() -> (OutputDelegate, then, Index:double=~) {is=title K2Node_CustomEvent}
   > PlayAnimationReverse(InAnimation=Get Intro(), PlaybackSpeed=999)
   > Delay(Duration=float * float(A=0.35, B=@n1.Index) {is=title K2Node_PromotableOperator})
   > SetRenderOpacity(InOpacity=1)
   > PlayAnimationForward(InAnimation=Get Intro())
   n2 = EventResize!() -> (OutputDelegate, then, Out:bool=~) {is=title K2Node_CustomEvent}
   n3 = NOT Boolean(A=@n2.Out)
   n4 = Get ShowdownEventTile_RoundBreak() -> (ShowdownEventTile_RoundBreak:Showdown Event Tile Round Break Object Reference)
   > Anim Transition!(
    Exec:Exec
    In:bool=@n3
    Animation:Widget Animation Object Reference=Get TransitionSize!(self:Showdown Event Tile Round Break Object Reference=@n4) -> (TransitionSize:Widget Animation Object Reference)
    PlaySpeed:double=1
    Widget:User Widget Object Reference=@n4
   ) -> (ExecOut:Exec) {is=title K2Node_MacroInstance}
   > Anim Transition(In=@n3, Animation=Get TransitionSize!() -> (TransitionSize), PlaySpeed=1)
   n5 = Get ShowdownEventTileInfo() -> (ShowdownEventTileInfo:Showdown Event Tile Info Object Reference)
   > Anim Transition(In=@n3, Animation=Get TransitionSize!(self:Showdown Event Tile Info Object Reference=@n5) -> (TransitionSize), PlaySpeed=1, Widget=@n5)
   n6 = EventColorize!() -> (OutputDelegate, then, DisplayInfo:Fort Tournament Display Info Structure) {is=title K2Node_CustomEvent}
   n7 = Break Fort Tournament Display Info!(FortTournamentDisplayInfo:"Fort Tournament Display Info Structure (by ref)"=@n6.DisplayInfo) -> (
    TitleLine1:Text
    TitleLine2:Text
    ScheduleInfo:Text
    FlavorDescription:Text
    DetailsDescription:Text
    ShortFormatTitle:Text
    LongFormatTitle:Text
    BackgroundTitle:Text
    PinScoreRequirement:int
    PinEarnedText:Text
    BaseColor:Linear Color Structure
    PrimaryColor:Linear Color Structure
    SecondaryColor:Linear Color Structure
    HighlightColor:Linear Color Structure
    TitleColor:Linear Color Structure
    ShadowColor:Linear Color Structure
    BackgroundLeftColor:Linear Color Structure
    BackgroundRightColor:Linear Color Structure
    BackgroundTextColor:Linear Color Structure
    PosterFadeColor:Linear Color Structure
    AlertText:Text
    SeriesPointLeaderboardName:Text
    AlertType:EFortTournamentAlertType Enum
    RoundDisplayNames:Array of Texts
   ) {is=title K2Node_BreakStruct}
   > SetColorAndOpacity(
    self:Image Object Reference=[
     Get Image_Bg() -> (Image_Bg:Image Object Reference)
     Get Image_Bg!(self:Showdown Event Tile Round Break Object Reference=Get ShowdownEventTile_RoundBreak()) -> (Image_Bg)
    ]
    InColorAndOpacity:Linear Color Structure=@n7.PrimaryColor
   )
   > SetColorAndOpacity(self=Get Image_Ended() -> (Image_Ended:Image Object Reference), InColorAndOpacity=@n7.ShadowColor)
   > n8 = GetDynamicMaterial(self=Get Image_Focus() -> (Image_Focus:Image Object Reference))
   > SetVectorParameterValue(self=@n8, ParameterName=Color1, Value=@n7.PrimaryColor)
   > SetVectorParameterValue(self=@n8, ParameterName=Color2, Value=@n7.HighlightColor)
   > SetColorAndOpacity(self=Get Image_Tail() -> (Image_Tail:Image Object Reference), InColorAndOpacity=@n7.PrimaryColor)
   n9 = EventSetState!() -> (OutputDelegate, then, EventState:EFortShowdownEventState Enum=Unknown) {is=title K2Node_CustomEvent}
   n10 = OR Boolean!(
    A="Equal (Enum)"(A:EFortShowdownEventState Enum=@n9.EventState, B:EFortShowdownEventState Enum=CompletedParticipated)
    B="Equal (Enum)"(A=@n9.EventState, B=CompletedNotPartipated)
    C:bool="Equal (Enum)"(A=@n9.EventState, B=Cancelled)
   ) -> (ReturnValue)
   > SetRenderOpacity(self=Get Image_Ended(), InOpacity=Select(Option 0:double=0, Option 1:double=1, Index=@n10) -> (ReturnValue:double))
   > Set isLive(isLive:bool=OR Boolean(
     A="Equal (Enum)"(A=@n9.EventState, B=LiveParticipating)
     B="Equal (Enum)"(A=@n9.EventState, B=LiveNotParticipating)
     C="Equal (Enum)"(A=@n9.EventState, B=Live)
    )) -> (Output_Get:bool)
   > Set isEnded(isEnded:bool=@n10) -> (Output_Get:bool)
   n11 = EventColorizeDispatch!() -> (OutputDelegate, then, DisplayInfo:Fort Tournament Display Info Structure) {is=title K2Node_CustomEvent}
   > EventColorize!(execute, self, DisplayInfo:Fort Tournament Display Info Structure=@n11.DisplayInfo) -> (then) {is=title K2Node_CallFunction}
   n12 = EventMultiRoundStyle() {is=title K2Node_CustomEvent}
   > SetVisibility(self=Get ShowdownEventTile_RoundBreak(), InVisibility=HitTestInvisible)
   Event Construct()
   > SetRenderOpacity(InOpacity=0)
   n13 = Get ShowdownEventTileInfo()
   > Bind Event to Set Event State(self:Showdown Event Tile Info Object Reference=@n13, Delegate=@n9) {is=title K2Node_AddDelegate}
   > Bind Event to Set Display Info(self:Showdown Event Tile Info Object Reference=@n13, Delegate=@n11) {is=title K2Node_AddDelegate}
   > Bind Event to Multi Round Session(self:Showdown Event Tile Info Object Reference=@n13, Delegate=@n12) {is=title K2Node_AddDelegate}
   n14 = Event OnRemovedFromFocusPath!() -> (OutputDelegate, then, InFocusEvent:Focus Event Structure) {is=title K2Node_Event}
   n15 = Self-Reference()
   n16 = Get Hover() -> (Hover:Widget Animation Object Reference)
   Anim Transition(Exec=[@n14.then, @n20.then], Animation=@n16, PlaySpeed=1, Widget=@n15)
   n17 = Get CommonBorder_InputContainer() -> (CommonBorder_InputContainer:Common Border Object Reference)
   > SetVisibility(self=@n17, InVisibility=Hidden)
   n18 = Get ShowdownEventTileInfo()
   > EventAnimHover!(execute, self:Showdown Event Tile Info Object Reference=@n18, In:bool=false) -> (then)
   n19 = Event On Hovered()
   Anim Transition(Exec=[@n19.then, @n27.then], In=true, Animation=@n16, PlaySpeed=3, Widget=@n15)
   > SetVisibility(self=@n17, InVisibility=SelfHitTestInvisible)
   > EventAnimHover(self=@n18, In=true)
   n20 = Event On Unhovered()
   Event On Selected()
   n21 = Self-Reference()
   n22 = Get Focus() -> (Focus:Widget Animation Object Reference)
   > Anim Transition(In=true, Animation=@n22, PlaySpeed=1, Widget=@n21)
   n23 = Get ShowdownEventTileInfo()
   > EventAnimFocus!(execute, self:Showdown Event Tile Info Object Reference=@n23, In:bool=true) -> (then)
   n24 = Get Violator() -> (Violator:Showdown Event Tile Violator Object Reference)
   n25 = Get CommonWidgetSwitcher_Proximity!(self:Showdown Event Tile Violator Object Reference=@n24) -> (CommonWidgetSwitcher_Proximity:Common Widget Switcher Legacy Object Reference)
   > SetActiveWidgetIndex(self=@n25, Index=1)
   n26 = Get Focus!(self:Showdown Event Tile Violator Object Reference=@n24) -> (Focus)
   > Anim Transition(In=true, Animation=@n26, PlaySpeed=1, Widget=@n24)
   Event On Deselected()
   > Anim Transition(Animation=@n22, PlaySpeed=1, Widget=@n21)
   > EventAnimFocus(self=@n23, In=false)
   > SetActiveWidgetIndex(self=@n25, Index=0)
   > Anim Transition(Animation=@n26, PlaySpeed=1, Widget=@n24)
   n27 = Event OnAddedToFocusPath!() -> (OutputDelegate, then, InFocusEvent:Focus Event Structure) {is=title K2Node_Event}
   n28 = NOT Boolean(A=Get isLive() -> (isLive:bool))
   Branch(Condition=@n28)
   n29 = Get Focus_Violator() -> (Focus_Violator:Widget Animation Object Reference)
   > Anim Transition(In=true, Animation=@n29, PlaySpeed=1)
   Branch(Condition=@n28)
   > Anim Transition(Animation=@n29, PlaySpeed=1)
   n30 = Get InputActionWidget() -> (InputActionWidget:Common Action Widget Object Reference)
   SetVisibility(self=@n30, InVisibility=Hidden)
   SetVisibility(self=@n30, InVisibility=Collapsed)
   Event On Clicked()
  ]}
 OnWindowSelected {nodes [OnWindowSelected() {is=title K2Node_FunctionEntry}]}
 SetData {nodes [
   n1 = SetData!() -> (then, TournamentInfo:String, EventInfo:String) {is=title K2Node_FunctionEntry}
   > SetTournament!(
    execute
    self:Fort Showdown Detail View Object Reference=Get ShowdownEventTileInfo()
    TournamentId:String=@n1
    EventId:String=@n1.EventInfo
    bForceRefreshUI:bool=true
   ) -> (then)
   > SetTournament(self=Get Violator(), TournamentId=@n1, EventId=@n1.EventInfo, bForceRefreshUI=true)
   > Set TournamentId(TournamentId:String=@n1) -> (Output_Get:String)
   > Set EventId(EventId:String=@n1.EventInfo) -> (Output_Get:String)
   > EventColorize()
  ]}
]
variables [
 EventId {type=string, default=""}
 isEnded {type=boolean}
 isLive {type=boolean}
 OnWindowSelected {}
 TournamentId {type=string, default=""}
]
widgets {
 info {inheritedWidgetCount=1, parentClass=&CommonButtonLegacy, rootWidgetClass=&Overlay, widgetCount=18}
 widgets [
  Overlay_Container {widgetClassPath=&Overlay, widget=Overlay_Container}
  HorizontalBox_1 {
   widgetClassPath=&HorizontalBox
   parent=Overlay_Container
   widget=HorizontalBox_1
   slot {class=OverlaySlot, properties {horizontalAlignment=HAlign_Center, verticalAlignment=VAlign_Bottom}}
  }
  ShowdownEventTile_RoundBreak {
   widgetClassPath=&/ShowdownEventTile_RoundBreak
   bIsVariable=true
   parent=HorizontalBox_1
   widget=ShowdownEventTile_RoundBreak
   properties {visibility=Collapsed}
   slot {class=HorizontalBoxSlot}
  }
  CommonFocusBorder {
   widgetClassPath=&CommonBorder
   parent=HorizontalBox_1
   widget=CommonFocusBorder
   properties {padding=<ltrb 2 0 2 0>}
   slot {class=HorizontalBoxSlot}
  }
  Overlay_InnerContainer {
   widgetClassPath=&Overlay
   parent=CommonFocusBorder
   widget=Overlay_InnerContainer
   slot {class=BorderSlot, properties {padding=<ltrb 2 0 2 0>}}
  }
  Image_Tail {
   widgetClassPath=&Image
   bIsVariable=true
   parent=Overlay_InnerContainer
   widget=Image_Tail
   properties {
    brush {
     bIsDynamicallyLoaded=false
     drawAs=Image
     imageType=NoImage
     mirroring=NoMirror
     resourceName=None
     resourceObject=/CompeteOldTab/Frontend/Showdown/Materials/M_UI_InspectTail_Inst1
     tiling=NoTile
     imageSize=<64 64>
     margin=<ltrb 0 0 0 0>
     outlineSettings {
      bUseBrushTransparency=false
      roundingType=HalfHeightRadius
      width=0
      color {colorUseRule=UseColor_Specified, specifiedColor=<rgba 0 0 0 0>}
      cornerRadii=<0 0 0 0>
     }
     tintColor {colorUseRule=UseColor_Specified, specifiedColor=<rgba 1 1 1 1>}
     uVRegion {bIsValid=false, max=<0 0>, min=<0 0>}
    }
    colorAndOpacity=<rgba 0.0185 1 0.9131 1>
    renderTransform {angle=0, scale=<0 0>, shear=<0 0>, translation=<5 60>}
    renderTransformPivot=<1 0>
   }
   slot {class=OverlaySlot, properties {verticalAlignment=VAlign_Bottom}}
  }
  SizeBox_0 {
   widgetClassPath=&SizeBox
   parent=Overlay_InnerContainer
   widget=SizeBox_0
   properties {bOverride_HeightOverride=true, bOverride_WidthOverride=true, heightOverride=212, widthOverride=148}
   slot {class=OverlaySlot, properties {horizontalAlignment=HAlign_Center, verticalAlignment=VAlign_Bottom}}
  }
  OverlayBGPlate {
   widgetClassPath=&Overlay
   parent=SizeBox_0
   widget=OverlayBGPlate
   properties {renderTransform {angle=0, scale=<1 1>, shear=<-16 0>, translation=<0 0>}}
   slot {class=SizeBoxSlot}
  }
  Image_Bg {
   widgetClassPath=&Image
   bIsVariable=true
   parent=OverlayBGPlate
   widget=Image_Bg
   properties {
    brush {
     bIsDynamicallyLoaded=false
     drawAs=Image
     imageType=NoImage
     mirroring=NoMirror
     resourceName=None
     resourceObject=null
     tiling=NoTile
     imageSize=<148 212.5361>
     margin=<ltrb 0 0 0 0>
     outlineSettings {
      bUseBrushTransparency=false
      roundingType=HalfHeightRadius
      width=0
      color {colorUseRule=UseColor_Specified, specifiedColor=<rgba 0 0 0 0>}
      cornerRadii=<0 0 0 0>
     }
     tintColor {colorUseRule=UseColor_Specified, specifiedColor=<rgba 1 1 1 1>}
     uVRegion {bIsValid=false, max=<0 0>, min=<0 0>}
    }
    colorAndOpacity=<rgba 0.0009 0.8632 0.8388 1>
   }
   slot {class=OverlaySlot, properties {horizontalAlignment=HAlign_Fill, verticalAlignment=VAlign_Fill}}
  }
  Image_Ended {
   widgetClassPath=&Image
   bIsVariable=true
   parent=OverlayBGPlate
   widget=Image_Ended
   properties {
    renderOpacity=0
    brush {
     bIsDynamicallyLoaded=false
     drawAs=Image
     imageType=NoImage
     mirroring=NoMirror
     resourceName=None
     resourceObject=null
     tiling=NoTile
     imageSize=<148 212.5361>
     margin=<ltrb 0 0 0 0>
     outlineSettings {
      bUseBrushTransparency=false
      roundingType=HalfHeightRadius
      width=0
      color {colorUseRule=UseColor_Specified, specifiedColor=<rgba 0 0 0 0>}
      cornerRadii=<0 0 0 0>
     }
     tintColor {colorUseRule=UseColor_Specified, specifiedColor=<rgba 1 1 1 1>}
     uVRegion {bIsValid=false, max=<0 0>, min=<0 0>}
    }
   }
   slot {class=OverlaySlot, properties {horizontalAlignment=HAlign_Fill, verticalAlignment=VAlign_Fill}}
  }
  Image_Hover {
   widgetClassPath=&Image
   bIsVariable=true
   parent=OverlayBGPlate
   widget=Image_Hover
   properties {
    renderOpacity=0
    brush {
     bIsDynamicallyLoaded=false
     drawAs=Image
     imageType=NoImage
     mirroring=NoMirror
     resourceName=None
     resourceObject=null
     tiling=NoTile
     imageSize=<148 212.5361>
     margin=<ltrb 0 0 0 0>
     outlineSettings {
      bUseBrushTransparency=false
      roundingType=HalfHeightRadius
      width=0
      color {colorUseRule=UseColor_Specified, specifiedColor=<rgba 0 0 0 0>}
      cornerRadii=<0 0 0 0>
     }
     tintColor {colorUseRule=UseColor_Specified, specifiedColor=<rgba 1 1 1 1>}
     uVRegion {bIsValid=false, max=<0 0>, min=<0 0>}
    }
   }
   slot {class=OverlaySlot, properties {horizontalAlignment=HAlign_Fill, verticalAlignment=VAlign_Fill}}
  }
  Image_Focus {
   widgetClassPath=&Image
   bIsVariable=true
   parent=OverlayBGPlate
   widget=Image_Focus
   properties {
    renderOpacity=0
    brush {
     bIsDynamicallyLoaded=false
     drawAs=Image
     imageType=NoImage
     mirroring=NoMirror
     resourceName=None
     resourceObject=/CompeteOldTab/Frontend/Showdown/Materials/M_UI_SkewBox_EventWindowFocus
     tiling=NoTile
     imageSize=<148 212.5361>
     margin=<ltrb 0 0 0 0>
     outlineSettings {
      bUseBrushTransparency=false
      roundingType=HalfHeightRadius
      width=0
      color {colorUseRule=UseColor_Specified, specifiedColor=<rgba 0 0 0 0>}
      cornerRadii=<0 0 0 0>
     }
     tintColor {colorUseRule=UseColor_Specified, specifiedColor=<rgba 1 1 1 1>}
     uVRegion {bIsValid=false, max=<0 0>, min=<0 0>}
    }
    renderTransform {angle=0, scale=<1.1 1.1>, shear=<0 0>, translation=<0 0>}
   }
   slot {class=OverlaySlot, properties {horizontalAlignment=HAlign_Fill, verticalAlignment=VAlign_Fill, padding=<ltrb 0 0 0 -8>}}
  }
  VerticalBox_0 {
   widgetClassPath=&VerticalBox
   parent=Overlay_InnerContainer
   widget=VerticalBox_0
   slot {class=OverlaySlot, properties {horizontalAlignment=HAlign_Fill, verticalAlignment=VAlign_Fill}}
  }
  SizeBox_2 {
   widgetClassPath=&SizeBox
   parent=VerticalBox_0
   widget=SizeBox_2
   properties {bOverride_HeightOverride=true, heightOverride=60.603}
   slot {class=VerticalBoxSlot, properties {padding=<ltrb 30 -60 -300 -60>}}
  }
  Violator {
   widgetClassPath=/Game/ShowdownEventTileViolator
   bIsVariable=true
   parent=SizeBox_2
   widget=Violator
   properties {bTimeRemainingShowSeconds=true, bTimeRemainingUppercase=true, timeRemainingFormat="IN {0}", timeRemainingFormatType=Approximate}
   slot {class=SizeBoxSlot}
  }
  ShowdownEventTileInfo {
   widgetClassPath=&/ShowdownEventTileInfo
   bIsVariable=true
   parent=VerticalBox_0
   widget=ShowdownEventTileInfo
   properties {bTimeRemainingShowSeconds=true, timeRemainingFormat="{0}", timeRemainingFormatType=Approximate, padding=<ltrb 0 5 0 0>}
   slot {class=VerticalBoxSlot, properties {padding=<ltrb 0 60 0 0>, size {sizeRule=Fill, value=1}}}
  }
  CommonBorder_InputContainer {
   widgetClassPath=&CommonBorder
   bIsVariable=true
   parent=Overlay_InnerContainer
   widget=CommonBorder_InputContainer
   properties {padding=<ltrb 0 0 0 0>}
   slot {class=OverlaySlot, properties {verticalAlignment=VAlign_Bottom}}
  }
  InputActionWidget {
   widgetClassPath=&CommonActionWidget
   bInherited=true
   bIsVariable=true
   parent=CommonBorder_InputContainer
   widget=InputActionWidget
   properties {
    visibility=Collapsed
    progressMaterialBrush {
     bIsDynamicallyLoaded=false
     drawAs=NoDrawType
     imageType=NoImage
     mirroring=NoMirror
     resourceName=None
     resourceObject=null
     tiling=NoTile
     imageSize=<32 32>
     margin=<ltrb 0 0 0 0>
     outlineSettings {
      bUseBrushTransparency=false
      roundingType=HalfHeightRadius
      width=0
      color {colorUseRule=UseColor_Specified, specifiedColor=<rgba 0 0 0 0>}
      cornerRadii=<0 0 0 0>
     }
     tintColor {colorUseRule=UseColor_Specified, specifiedColor=<rgba 1 1 1 1>}
     uVRegion {bIsValid=false, max=<0 0>, min=<0 0>}
    }
    renderTransform {angle=0, scale=<1 1>, shear=<0 0>, translation=<-14 -4>}
   }
   slot {class=BorderSlot, properties {padding=<ltrb 0 0 0 0>}}
  }
 ]
}
