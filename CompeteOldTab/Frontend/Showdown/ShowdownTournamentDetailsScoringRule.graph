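# Showdown tournament details row widget that shows one scoring rule's description and point value over a tilted, faded icon.
# SetData looks up rule display info through CMSContext, then formats a single or interval description from the requirement bounds, substituting "Victory Royale" for first placement.
# Points show signed per-rule values, or unsigned accumulated values when nonzero.
# Non-positive values tint the content orange, and the tournament's ShadowColor sets the background.
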WidgetBlueprint /CompeteOldTab/Frontend/Showdown/ShowdownTournamentDetailsScoringRule
graphs [
 EventGraph {links [], nodes []}
 SetData {nodes [
   n1 = SetData!() -> (then, ScoreRule:Fort Showdown Scoring Rule Info Structure, TournamentDisplayInfo:Fort Tournament Display Info Structure) {is=title K2Node_FunctionEntry}
   n2 = Break Fort Showdown Scoring Rule Info!(FortShowdownScoringRuleInfo:"Fort Showdown Scoring Rule Info Structure (by ref)"=@n1) -> (DisplayInfoId:Name, PointValue:int, AccumulatedPointValue:int, ScoreRequirement:int, UpperBoundRequirement:int, LowerBoundRequirement:int) {is=title K2Node_BreakStruct}
   n3 = "Equal (Name)"!(A:Name=@n2, B:Name=Placement) -> (ReturnValue:bool) {is=title K2Node_CallFunction}
   n4 = "Equal (Integer)"(A=@n2.ScoreRequirement, B=1) {is=title K2Node_CallFunction}
   n5 = SelectInt(
    A=@n2.ScoreRequirement
    B=@n2.LowerBoundRequirement
    bPickA=AND Boolean(A="Equal (Integer)"(A=@n2.UpperBoundRequirement) {is=title K2Node_PromotableOperator}, B="Equal (Integer)"(A=@n2.LowerBoundRequirement))
   )
   n6 = Break Fort Showdown Scoring Rule Display Info!(FortShowdownScoringRuleDisplayInfo:"Fort Showdown Scoring Rule Display Info Structure (by ref)"=GetScoringRuleDisplayInfo!(
     self:CMSContext Object Reference=Get CMSContext!() -> (ReturnValue:CMSContext Object Reference) {is=title K2Node_GetBlueprintContextEx}
     RuleName:Name=@n2
    ) -> (ReturnValue:Fort Showdown Scoring Rule Display Info Structure)) -> (
    Name:Name
    Description:Text
    PosterDescription:Text
    PosterIntervalDescription:Text
    Icon:Texture 2D Soft Object Reference
    bShowScoreToastNotifications:bool=~
   ) {is=title K2Node_BreakStruct}
   n7 = Select(
    Option 0:Text=@n6.PosterDescription
    Option 1:Text=@n6.PosterIntervalDescription
    Index="Not Equal (Integer)"(A=@n2.UpperBoundRequirement, B=@n2.LowerBoundRequirement)
   ) -> (ReturnValue:Text)
   > "SetText (Text)"(
    self=Get RuleNameText() -> (RuleNameText:Common Text Object Reference)
    InText=Select(
     Option 0=Format Text!(Format=@n7, 0:int=@n5, 1:int=@n2.UpperBoundRequirement) -> (Result)
     Option 1="LOCTEXT(\"Victory Royale\")"
     Index=AND Boolean(A=@n3, B=@n4)
    )
   )
   n8 = "Equal (Integer)"(A=@n2.AccumulatedPointValue)
   n9 = SelectInt(A=@n2.PointValue, B=@n2.AccumulatedPointValue, bPickA=@n8)
   > "SetText (Text)"(
    self=Get RuleValueText() -> (RuleValueText:Common Text Object Reference)
    InText="To Text (Integer)"(Value=@n9, bAlwaysSign=@n8, bUseGrouping=false)
   )
   > SetBrushFromLazyTexture!(
    execute
    self:Common Lazy Image Object Reference=Get RuleIcon() -> (RuleIcon:Common Lazy Image Object Reference)
    LazyTexture:"Texture 2D Soft Object Reference (by ref)"=@n6.Icon
    bMatchSize:bool=false
   ) -> (then)
   > SetContentColorAndOpacity!(
    execute
    self:Border Object Reference=Get CommonBorder_TextContainer() -> (CommonBorder_TextContainer:Common Border Object Reference)
    InContentColorAndOpacity:Linear Color Structure=Select(
     Option 0:Linear Color Structure=(R=1,G=1,B=1,A=1)
     Option 1:Linear Color Structure=(R=1,G=0.2961,B=0.0612,A=1)
     Index="integer <= integer"!(A:int=@n9, B:int=0) -> (ReturnValue:bool) {is=title K2Node_CallFunction}
    ) -> (ReturnValue:Linear Color Structure)
   ) -> (then)
   > SetBrushColor(
    self=Get CommonBorder_BG() -> (CommonBorder_BG:Common Border Object Reference)
    InBrushColor=Break Fort Tournament Display Info!(FortTournamentDisplayInfo:"Fort Tournament Display Info Structure (by ref)"=@n1.TournamentDisplayInfo) -> (
     TitleLine1:Text
     TitleLine2:Text
     ScheduleInfo:Text
     FlavorDescription:Text
     DetailsDescription:Text
     ShortFormatTitle:Text
     LongFormatTitle:Text
     BackgroundTitle:Text
     PinScoreRequirement:int
     PinEarnedText:Text
     BaseColor:Linear Color Structure
     PrimaryColor:Linear Color Structure
     SecondaryColor:Linear Color Structure
     HighlightColor:Linear Color Structure
     TitleColor:Linear Color Structure
     ShadowColor:Linear Color Structure
     BackgroundLeftColor:Linear Color Structure
     BackgroundRightColor:Linear Color Structure
     BackgroundTextColor:Linear Color Structure
     PosterFadeColor:Linear Color Structure
     AlertText:Text
     SeriesPointLeaderboardName:Text
     AlertType:EFortTournamentAlertType Enum
     RoundDisplayNames:Array of Texts
    ) {is=title K2Node_BreakStruct}.ShadowColor
   )
  ]}
]
variables []
widgets {
 info {inheritedWidgetCount=0, parentClass=&CommonUserWidget, rootWidgetClass=&CommonBorder, widgetCount=7}
 widgets [
  CommonBorder_BG {
   widgetClassPath=&CommonBorder
   bIsVariable=true
   widget=CommonBorder_BG
   properties {
    background {
     bIsDynamicallyLoaded=false
     drawAs=Image
     imageType=NoImage
     mirroring=NoMirror
     resourceName=None
     resourceObject=null
     tiling=NoTile
     imageSize=<32 32>
     margin=<ltrb 0 0 0 0>
     outlineSettings {
      bUseBrushTransparency=false
      roundingType=HalfHeightRadius
      width=0
      color {colorUseRule=UseColor_Specified, specifiedColor=<rgba 0 0 0 0>}
      cornerRadii=<0 0 0 0>
     }
     tintColor {colorUseRule=UseColor_Specified, specifiedColor=<rgba 1 1 1 1>}
     uVRegion {bIsValid=false, max=<0 0>, min=<0 0>}
    }
    brushColor=<rgba 0.0185 1 0.9047 0.06>
    padding=<ltrb 0 0 0 0>
   }
  }
  Overlay_0 {
   widgetClassPath=&Overlay
   parent=CommonBorder_BG
   widget=Overlay_0
   properties {clipping=ClipToBounds}
   slot {class=BorderSlot, properties {padding=<ltrb 0 0 0 0>}}
  }
  RuleIcon {
   widgetClassPath=/Script/CommonUI.CommonLazyImage
   bIsVariable=true
   parent=Overlay_0
   widget=RuleIcon
   properties {
    brush {
     bIsDynamicallyLoaded=false
     drawAs=Image
     imageType=NoImage
     mirroring=NoMirror
     resourceName=None
     resourceObject=/Game/Athena/HUD/Art/killsico_alt.KillsIco_Alt
     tiling=NoTile
     imageSize=<100 100>
     margin=<ltrb 0 0 0 0>
     outlineSettings {
      bUseBrushTransparency=false
      roundingType=HalfHeightRadius
      width=0
      color {colorUseRule=UseColor_Specified, specifiedColor=<rgba 0 0 0 0>}
      cornerRadii=<0 0 0 0>
     }
     tintColor {colorUseRule=UseColor_Specified, specifiedColor=<rgba 1 1 1 1>}
     uVRegion {bIsValid=false, max=<0 0>, min=<0 0>}
    }
    colorAndOpacity=<rgba 1 1 1 0.1048>
    renderTransform {angle=-17, scale=<1 1>, shear=<0 0>, translation=<0 0>}
   }
   slot {class=OverlaySlot, properties {verticalAlignment=VAlign_Center, padding=<ltrb -25 -25 -25 -25>}}
  }
  CommonBorder_TextContainer {
   widgetClassPath=&CommonBorder
   bIsVariable=true
   parent=Overlay_0
   widget=CommonBorder_TextContainer
   slot {class=OverlaySlot, properties {horizontalAlignment=HAlign_Fill, padding=<ltrb 8 8 8 8>}}
  }
  HorizontalBox_0 {widgetClassPath=&HorizontalBox, parent=CommonBorder_TextContainer, widget=HorizontalBox_0, slot {class=BorderSlot}}
  RuleNameText {
   widgetClassPath=&CommonTextBlock
   bIsVariable=true
   parent=HorizontalBox_0
   widget=RuleNameText
   properties {
    text="Victory Royale "
    font {
     bForceMonospaced=false
     bMaterialIsStencil=false
     fontMaterial=null
     fontObject=&Burbank
     letterSpacing=0
     monospacedWidth=1
     size=22
     skewAmount=0
     typefaceFontName=Burbank-PlayerPower
     outlineSettings {
      bApplyOutlineToDropShadows=false
      bMiteredCorners=false
      bSeparateFillAlpha=false
      outlineMaterial=null
      outlineSize=0
      outlineColor=<rgba 0 0 0 1>
     }
    }
   }
   slot {
    class=HorizontalBoxSlot
    properties {horizontalAlignment=HAlign_Left, verticalAlignment=VAlign_Center, padding=<ltrb 10 0 20 0>, size {sizeRule=Fill, value=1}}
   }
  }
  RuleValueText {
   widgetClassPath=&CommonTextBlock
   bIsVariable=true
   parent=HorizontalBox_0
   widget=RuleValueText
   properties {
    text=+1
    font {
     bForceMonospaced=false
     bMaterialIsStencil=false
     fontMaterial=null
     fontObject=&Burbank
     letterSpacing=0
     monospacedWidth=1
     size=30
     skewAmount=0
     typefaceFontName=Burbank-Black
     outlineSettings {
      bApplyOutlineToDropShadows=false
      bMiteredCorners=false
      bSeparateFillAlpha=false
      outlineMaterial=null
      outlineSize=0
      outlineColor=<rgba 0 0 0 1>
     }
    }
   }
   slot {class=HorizontalBoxSlot, properties {verticalAlignment=VAlign_Center, padding=<ltrb 0 0 20 0>}}
  }
 ]
}
