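# Back face of the Showdown tournament poster, a detail-view widget showing titles, flavor text, best-score pin and event dates.
# Refresh Data fills text from TournamentDisplayInfo, applies its colors to text and existing date entries, and shows the pin for non-negative scores.
# It then rebuilds the date grid with one TournamentDateEntry per event, splitting into two columns above seven.
# Construct adjusts mobile title padding; GetTournamentDateText formats a month-year range but is uncalled here.
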WidgetBlueprint /CompeteOldTab/Frontend/Showdown/ShowdownTournamentPosterBack
graphs [
 EventGraph {nodes [
   EventColorize() {is=title K2Node_CustomEvent}
   n1 = Get TournamentDisplayInfo() -> (TournamentDisplayInfo:Fort Tournament Display Info Structure)
   n2 = Break Fort Tournament Display Info!(FortTournamentDisplayInfo:"Fort Tournament Display Info Structure (by ref)"=@n1) -> (
    TitleLine1:Text
    TitleLine2:Text
    ScheduleInfo:Text
    FlavorDescription:Text
    DetailsDescription:Text
    ShortFormatTitle:Text
    LongFormatTitle:Text
    BackgroundTitle:Text
    PinScoreRequirement:int
    PinEarnedText:Text
    BaseColor:Linear Color Structure
    PrimaryColor:Linear Color Structure
    SecondaryColor:Linear Color Structure
    HighlightColor:Linear Color Structure
    TitleColor:Linear Color Structure
    ShadowColor:Linear Color Structure
    BackgroundLeftColor:Linear Color Structure
    BackgroundRightColor:Linear Color Structure
    BackgroundTextColor:Linear Color Structure
    PosterFadeColor:Linear Color Structure
    AlertText:Text
    SeriesPointLeaderboardName:Text
    AlertType:EFortTournamentAlertType Enum
    RoundDisplayNames:Array of Texts
   ) {is=title K2Node_BreakStruct}
   > SetColorAndOpacity(
    self=[
     Get TitleLine1() -> (TitleLine1:Common Text Object Reference)
     Get CommonTextBlock_MetaTitle() -> (CommonTextBlock_MetaTitle:Common Text Object Reference)
     Get TitleLine2() -> (TitleLine2:Common Text Object Reference)
    ]
    InColorAndOpacity=Make Slate Color(SpecifiedColor=@n2.TitleColor)
   )
   > SetColorAndOpacity(
    self=[
     Get FlavorText() -> (FlavorText:Common Text Object Reference)
     Get CommonTextBlock_MetaDate() -> (CommonTextBlock_MetaDate:Common Text Object Reference)
    ]
    InColorAndOpacity=Make Slate Color(SpecifiedColor=@n2.BaseColor)
   )
   > SetColorAndOpacity(
    self=[
     Get DateText() -> (DateText:Common Text Object Reference)
     Get CommonTextBlock_EventDates() -> (CommonTextBlock_EventDates:Common Text Object Reference)
    ]
    InColorAndOpacity=Make Slate Color(SpecifiedColor=@n2.PrimaryColor)
   )
   n3 = Get UniformGridPanel_DateGrid() -> (UniformGridPanel_DateGrid:Uniform Grid Panel Object Reference)
   > n4 = For Loop(LastIndex=int - int(A=GetChildrenCount!(self:Panel Widget Object Reference=@n3) -> (ReturnValue:int=0), B=1) {is=title K2Node_CallFunction})
   > n5 = Cast To TournamentDateEntry!(execute, Object=GetChildAt!(self:Panel Widget Object Reference=@n3, Index:int=@n4) -> (ReturnValue:Widget Object Reference)) -> (then, CastFailed, AsTournament Date Entry:Tournament Date Entry Object Reference) {is=title K2Node_DynamicCast}
   > EventColor!(execute, self:Tournament Date Entry Object Reference=@n5, showdownData:"Fort Tournament Display Info Structure (by ref)"=@n1) -> (then)
   EventPopulateEventWindowData() {is=title K2Node_CustomEvent}
   n6 = Get UniformGridPanel_DateGrid()
   > ClearChildren(self=@n6)
   > n7 = GetTournamentEvents!(execute, self:Fort Showdown Detail View Object Reference, bGroupByRoundNumber:bool=false) -> (then, ReturnValue:Array of Strings)
   > n8 = For Each Loop(Array=@n7)
   > n9 = Create Widget!(
    execute:Exec
    Class:User Widget Class Reference=&/TournamentDateEntry.TournamentDateEntry_C
    OwningPlayer:Player Controller Object Reference
   ) -> (then:Exec, ReturnValue:Tournament Date Entry Object Reference)
   > n10 = AddChildToUniformGrid!(execute, self:Uniform Grid Panel Object Reference=@n6, Content:Widget Object Reference=@n9, InRow:int=0, InColumn:int=0) -> (then, ReturnValue:Uniform Grid Slot Object Reference)
   > SetTournament!(
    execute
    self:Fort Showdown Detail View Object Reference=@n9
    TournamentId:String=GetTournamentId!(self:Fort Showdown Detail View Object Reference) -> (ReturnValue:String)
    EventId:String=@n8
    bForceRefreshUI:bool=true
   ) -> (then)
   n11 = Length(TargetArray:Array of Strings=@n7)
   n12 = "integer > integer"(A=@n11, B=7) {is=title K2Node_CallFunction}
   n13 = Round(A=float / float(A="To Float (Integer)"(InInt=@n11), B=Select(Option 0:double=1, Option 1:double=2, Index=@n12) -> (ReturnValue:double)) {is=title K2Node_CallFunction}) {is=title K2Node_CallFunction}
   n14 = "% (Integer)"(A=@n8.Array Index, B=@n13) -> (ReturnValue=0)
   > SetRow!(
    execute
    self:Uniform Grid Slot Object Reference=@n10
    InRow:int=Select(Option 0:int=@n14, Option 1:int=@n14, Index="integer > integer"(A=@n8.Array Index, B=@n13)) -> (ReturnValue:int)
   ) -> (then)
   > SetColumn!(
    execute
    self:Uniform Grid Slot Object Reference=@n10
    InColumn:int=int / int!(A:int=@n8.Array Index, B:int=@n13) -> (ReturnValue:int=0) {is=title K2Node_CallFunction}
   ) -> (then)
   > SetVerticalAlignment(self:Uniform Grid Slot Object Reference=@n10, InVerticalAlignment=VAlign_Fill)
   > SetHorizontalAlignment(self:Uniform Grid Slot Object Reference=@n10, InHorizontalAlignment=HAlign_Fill)
   Event Construct()
   > Branch(Condition=IsMobileGame())
   > SetPadding!(
    execute
    self:Vertical Box Slot Object Reference=SlotAsVerticalBoxSlot(Widget=Get ScaleBoxTournamentTitle() -> (ScaleBoxTournamentTitle:Scale Box Object Reference))
    InPadding:Margin Structure=Get TitlePaddingMobile() -> (TitlePaddingMobile:Margin Structure)
   ) -> (then)
   Event Refresh Data() {is=title K2Node_Event}
   n15 = Break Fort Tournament Display Info(FortTournamentDisplayInfo=Get TournamentDisplayInfo())
   > "SetText (Text)"(self=Get TitleLine1(), InText=TextToUpper!(InText:"Text (by ref)"=@n15) -> (ReturnValue:Text))
   > "SetText (Text)"(self=Get TitleLine2(), InText=TextToUpper(InText=@n15.TitleLine2))
   > "SetText (Text)"(self=Get CommonTextBlock_MetaDate(), InText=TextToUpper(InText=@n15.ScheduleInfo))
   > "SetText (Text)"(self=Get FlavorText(), InText=@n15.FlavorDescription)
   > "SetText (Text)"(self=Get CommonTextBlock_MetaTitle(), InText=TextToUpper(InText=@n15.ShortFormatTitle))
   > "SetText (Text)"(self=Get DateText(), InText=TextToUpper(InText=@n15.LongFormatTitle))
   > EventColorize() {is=title K2Node_CallFunction}
   > SetTournament(self=Get ShowdownPin() -> (ShowdownPin:Showdown Pin Object Reference), TournamentId=GetTournamentId(), bForceRefreshUI=true)
   n16 = GetTournamentBestScore!(self:Fort Showdown Detail View Object Reference) -> (ReturnValue:int=0)
   > SetVisibility(
    self=Get ShowdownPin()
    InVisibility=Select(
     Option 0:ESlateVisibility Enum=Collapsed
     Option 1:ESlateVisibility Enum=HitTestInvisible
     Index="integer >= integer"!(A:int=@n16, B:int=0) -> (ReturnValue:bool) {is=title K2Node_CallFunction}
    ) -> (ReturnValue:ESlateVisibility Enum)
   )
   > EventPopulateEventWindowData() {is=title K2Node_CallFunction}
  ]}
 GetTournamentDateText {nodes [
   GetTournamentDateText() {is=title K2Node_FunctionEntry}
   n1 = BreakDateTime!(InDateTime:Date Time Structure=GetTournamentEndTime!(self:Fort Showdown Detail View Object Reference) -> (ReturnValue:Date Time Structure)) -> (Year:int=0, Month:int=0, Day:int=0, Hour:int=0, Minute:int=0, Second:int=0, Millisecond:int=0)
   n2 = "Equal (Integer)"(
    A=BreakDateTime(InDateTime=GetTournamentBeginTime!(self:Fort Showdown Detail View Object Reference) -> (ReturnValue:Date Time Structure)) -> (Year=0, Month=0, Day=0, Hour=0, Minute=0, Second=0, Millisecond=0).Month
    B=BreakDateTime(InDateTime=GetTournamentEndTime()) -> (Year=0, Month=0, Day=0, Hour=0, Minute=0, Second=0, Millisecond=0).Month
   ) {is=title K2Node_CallFunction}
   n3 = BreakDateTime(InDateTime=GetTournamentBeginTime()) -> (Year=0, Month=0, Day=0, Hour=0, Minute=0, Second=0, Millisecond=0)
   n4 = "Equal (Integer)"(A=@n3, B=@n1)
   n5 = "To Text (Integer)"(Value=@n3, bUseGrouping=false)
   n6 = Select!(
    Option 0:Text
    Option 1:Text="LOCTEXT(\"January\")"
    Option 2:Text="LOCTEXT(\"February\")"
    Option 3:Text="LOCTEXT(\"March\")"
    Option 4:Text="LOCTEXT(\"April\")"
    Option 5:Text="LOCTEXT(\"May\")"
    Option 6:Text="LOCTEXT(\"June\")"
    Option 7:Text="LOCTEXT(\"July\")"
    Option 8:Text="LOCTEXT(\"August\")"
    Option 9:Text="LOCTEXT(\"September\")"
    Option 10:Text="LOCTEXT(\"October\")"
    Option 11:Text="LOCTEXT(\"November\")"
    Option 12:Text="LOCTEXT(\"December\")"
    Index:int=BreakDateTime(InDateTime=GetTournamentBeginTime()) -> (Year=0, Month=0, Day=0, Hour=0, Minute=0, Second=0, Millisecond=0).Month
   ) -> (ReturnValue:Text)
   n7 = BreakDateTime(InDateTime=GetTournamentEndTime()) -> (Year=0, Month=0, Day=0, Hour=0, Minute=0, Second=0, Millisecond=0)
   > Return Node!(
    execute
    TournamentDate:Text=Format Text!(
     Format="LOCTEXT(\"{StartMonth} {StartYear}{Separator}{EndMonth} {EndYear}\")"
     StartMonth:Text=Select!(Option 0=@n6, Option 1, Index:bool=@n2) -> (ReturnValue)
     StartYear:Text=Select(Option 0=@n5, Index=@n4)
     Separator:Text=Select(Option 0="LOCTEXT(\" - \")", Index=@n2)
     EndMonth:Text=Select!(
      Option 0
      Option 1="LOCTEXT(\"January\")"
      Option 2:Text="LOCTEXT(\"February\")"
      Option 3:Text="LOCTEXT(\"March\")"
      Option 4:Text="LOCTEXT(\"April\")"
      Option 5:Text="LOCTEXT(\"May\")"
      Option 6:Text="LOCTEXT(\"June\")"
      Option 7:Text="LOCTEXT(\"July\")"
      Option 8:Text="LOCTEXT(\"August\")"
      Option 9:Text="LOCTEXT(\"September\")"
      Option 10:Text="LOCTEXT(\"October\")"
      Option 11:Text="LOCTEXT(\"November\")"
      Option 12:Text="LOCTEXT(\"December\")"
      Index:int=@n7.Month
     ) -> (ReturnValue)
     EndYear:Text="To Text (Integer)"(Value=@n1, bUseGrouping=false)
    ) -> (Result)
   )
  ]}
]
variables []
widgets {
 info {inheritedWidgetCount=6, parentClass=/Script/FortniteUI.FortShowdownDetailView, rootWidgetClass=&SizeBox, widgetCount=57}
 widgets [
  SizeBox_0 {
   widgetClassPath=&SizeBox
   widget=SizeBox_0
   properties {bOverride_HeightOverride=true, bOverride_WidthOverride=true, heightOverride=1080, widthOverride=750}
  }
  Overlay_0 {widgetClassPath=&Overlay, parent=SizeBox_0, widget=Overlay_0, slot {class=SizeBoxSlot}}
  Image_PosterBack {
   widgetClassPath=/Script/EpicCMSUIFramework.EpicCMSImage
   bInherited=true
   bIsVariable=true
   parent=Overlay_0
   widget=Image_PosterBack
   properties {
    loadingBackgroundBrush {
     bIsDynamicallyLoaded=false
     drawAs=Image
     imageType=NoImage
     mirroring=NoMirror
     resourceName=None
     resourceObject=null
     tiling=NoTile
     imageSize=<32 32>
     margin=<ltrb 0 0 0 0>
     outlineSettings {
      bUseBrushTransparency=false
      roundingType=HalfHeightRadius
      width=0
      color {colorUseRule=UseColor_Specified, specifiedColor=<rgba 0 0 0 0>}
      cornerRadii=<0 0 0 0>
     }
     tintColor {colorUseRule=UseColor_Specified, specifiedColor=<rgba 0.0052 0.0052 0.0056 1>}
     uVRegion {bIsValid=false, max=<0 0>, min=<0 0>}
    }
    loadingFailFallback {
     bIsDynamicallyLoaded=false
     drawAs=Image
     imageType=NoImage
     mirroring=NoMirror
     resourceName=None
     resourceObject=null
     tiling=NoTile
     imageSize=<32 32>
     margin=<ltrb 0 0 0 0>
     outlineSettings {
      bUseBrushTransparency=false
      roundingType=HalfHeightRadius
      width=0
      color {colorUseRule=UseColor_Specified, specifiedColor=<rgba 0 0 0 0>}
      cornerRadii=<0 0 0 0>
     }
     tintColor {colorUseRule=UseColor_Specified, specifiedColor=<rgba 0.1725 0.1771 0.1679 1>}
     uVRegion {bIsValid=false, max=<0 0>, min=<0 0>}
    }
   }
   slot {class=OverlaySlot, properties {horizontalAlignment=HAlign_Fill, verticalAlignment=VAlign_Fill}}
  }
  VerticalBox_3 {
   widgetClassPath=&VerticalBox
   parent=Overlay_0
   widget=VerticalBox_3
   slot {class=OverlaySlot, properties {horizontalAlignment=HAlign_Fill, verticalAlignment=VAlign_Fill, padding=<ltrb 30 30 30 30>}}
  }
  HorizontalBox_0 {widgetClassPath=&HorizontalBox, parent=VerticalBox_3, widget=HorizontalBox_0, slot {class=VerticalBoxSlot}}
  SizeBox_Pin {
   widgetClassPath=&SizeBox
   bIsVariable=true
   parent=HorizontalBox_0
   widget=SizeBox_Pin
   properties {bOverride_HeightOverride=true, bOverride_WidthOverride=true, heightOverride=190, widthOverride=190}
   slot {class=HorizontalBoxSlot, properties {horizontalAlignment=HAlign_Left, verticalAlignment=VAlign_Bottom, padding=<ltrb -18 0 0 -10>}}
  }
  ScaleBox_4 {widgetClassPath=&ScaleBox, parent=SizeBox_Pin, widget=ScaleBox_4, slot {class=SizeBoxSlot}}
  ShowdownPin {
   widgetClassPath=&/ShowdownPin
   bIsVariable=true
   parent=ScaleBox_4
   widget=ShowdownPin
   properties {bRepresentsTournamentBestPin=true, colorSilhouette=true, dEVPinScore=12, dEVShowPin=true}
   slot {class=ScaleBoxSlot, properties {verticalAlignment=VAlign_Bottom}}
  }
  SizeBox_2 {
   widgetClassPath=&SizeBox
   parent=HorizontalBox_0
   widget=SizeBox_2
   properties {bOverride_WidthOverride=true, heightOverride=267.7652}
   slot {class=HorizontalBoxSlot, properties {padding=<ltrb 0 0 20 0>, size {sizeRule=Fill, value=1}}}
  }
  VerticalBox_2 {widgetClassPath=&VerticalBox, parent=SizeBox_2, widget=VerticalBox_2, slot {class=SizeBoxSlot}}
  VerticalBox_1 {
   widgetClassPath=&VerticalBox
   parent=VerticalBox_2
   widget=VerticalBox_1
   slot {class=VerticalBoxSlot, properties {size {sizeRule=Fill, value=1}}}
  }
  ScaleBoxTournamentTitle {
   widgetClassPath=&ScaleBox
   bIsVariable=true
   parent=VerticalBox_1
   widget=ScaleBoxTournamentTitle
   slot {class=VerticalBoxSlot, properties {verticalAlignment=VAlign_Bottom, size {sizeRule=Fill, value=1}}}
  }
  TitleLine1 {
   widgetClassPath=&CommonTextBlock
   bIsVariable=true
   parent=ScaleBoxTournamentTitle
   widget=TitleLine1
   properties {
    text=ALPHA
    colorAndOpacity {colorUseRule=UseColor_Specified, specifiedColor=<rgba 0.9911 0.9131 0.0409 1>}
    font {
     bForceMonospaced=false
     bMaterialIsStencil=false
     fontMaterial=null
     fontObject=&Burbank
     letterSpacing=0
     monospacedWidth=1
     size=98
     skewAmount=0
     typefaceFontName=BurbankRegular-Black
     outlineSettings {
      bApplyOutlineToDropShadows=false
      bMiteredCorners=false
      bSeparateFillAlpha=false
      outlineMaterial=null
      outlineSize=0
      outlineColor=<rgba 0 0 0 1>
     }
    }
    margin=<ltrb 0 0 0 -20>
    renderTransform {angle=0, scale=<1 1>, shear=<-11 0>, translation=<0 0>}
   }
   slot {class=ScaleBoxSlot, properties {horizontalAlignment=HAlign_Left, verticalAlignment=VAlign_Bottom}}
  }
  ScaleBox_1 {
   widgetClassPath=&ScaleBox
   parent=VerticalBox_1
   widget=ScaleBox_1
   slot {class=VerticalBoxSlot, properties {verticalAlignment=VAlign_Top, size {sizeRule=Automatic, value=0.5}}}
  }
  TitleLine2 {
   widgetClassPath=&CommonTextBlock
   bIsVariable=true
   parent=ScaleBox_1
   widget=TitleLine2
   properties {
    text=TOURNAMENT
    colorAndOpacity {colorUseRule=UseColor_Specified, specifiedColor=<rgba 0.9911 0.9131 0.0409 1>}
    font {
     bForceMonospaced=false
     bMaterialIsStencil=false
     fontMaterial=null
     fontObject=&Burbank
     letterSpacing=0
     monospacedWidth=1
     size=90
     skewAmount=0
     typefaceFontName=BurbankRegular-Black
     outlineSettings {
      bApplyOutlineToDropShadows=false
      bMiteredCorners=false
      bSeparateFillAlpha=false
      outlineMaterial=null
      outlineSize=0
      outlineColor=<rgba 0 0 0 1>
     }
    }
    margin=<ltrb 0 0 0 -10>
    renderTransform {angle=0, scale=<1 1>, shear=<-11 0>, translation=<0 0>}
   }
   slot {class=ScaleBoxSlot, properties {horizontalAlignment=HAlign_Left, verticalAlignment=VAlign_Top}}
  }
  ScaleBox_3 {
   widgetClassPath=&ScaleBox
   parent=VerticalBox_1
   widget=ScaleBox_3
   properties {stretchDirection=DownOnly}
   slot {class=VerticalBoxSlot, properties {size {sizeRule=Automatic, value=0.25}}}
  }
  DateText {
   widgetClassPath=&CommonTextBlock
   bIsVariable=true
   parent=ScaleBox_3
   widget=DateText
   properties {
    text="SEPT  14 - OCT 24"
    colorAndOpacity {colorUseRule=UseColor_Specified, specifiedColor=<rgba 0.0185 1 0.9047 1>}
    font {
     bForceMonospaced=false
     bMaterialIsStencil=false
     fontMaterial=null
     fontObject=&Burbank
     letterSpacing=0
     monospacedWidth=1
     size=40
     skewAmount=0
     typefaceFontName=Burbank-PlayerPower
     outlineSettings {
      bApplyOutlineToDropShadows=false
      bMiteredCorners=false
      bSeparateFillAlpha=false
      outlineMaterial=null
      outlineSize=0
      outlineColor=<rgba 0 0 0 1>
     }
    }
    margin=<ltrb 0 -10 0 0>
    renderTransform {angle=0, scale=<1 1>, shear=<-7 0>, translation=<0 0>}
   }
   slot {class=ScaleBoxSlot, properties {horizontalAlignment=HAlign_Left, verticalAlignment=VAlign_Top}}
  }
  SizeBoxDescription {
   widgetClassPath=&SizeBox
   parent=VerticalBox_3
   widget=SizeBoxDescription
   properties {bOverride_MaxDesiredHeight=true, maxDesiredHeight=150}
   slot {class=VerticalBoxSlot, properties {verticalAlignment=VAlign_Top, padding=<ltrb 20 20 20 20>}}
  }
  ScrollBox_0 {
   widgetClassPath=&ScrollBox
   parent=SizeBoxDescription
   widget=ScrollBox_0
   properties {widgetStyle {
     barThickness=9
     bottomShadowBrush {
      bIsDynamicallyLoaded=false
      drawAs=Box
      imageType=FullColor
      mirroring=NoMirror
      resourceName=../../../Engine/Content/Slate/Common/ScrollBoxShadowBottom.png
      resourceObject=null
      tiling=NoTile
      imageSize=<16 8>
      margin=<ltrb 0.5 0 0.5 1>
      outlineSettings {
       bUseBrushTransparency=false
       roundingType=HalfHeightRadius
       width=0
       color {colorUseRule=UseColor_Specified, specifiedColor=<rgba 0 0 0 0>}
       cornerRadii=<0 0 0 0>
      }
      tintColor {colorUseRule=UseColor_Specified, specifiedColor=<rgba 1 1 1 1>}
      uVRegion {bIsValid=false, max=<0 0>, min=<0 0>}
     }
     horizontalScrolledContentPadding=<ltrb 0 0 1 0>
     leftShadowBrush {
      bIsDynamicallyLoaded=false
      drawAs=NoDrawType
      imageType=FullColor
      mirroring=NoMirror
      resourceName=../../../Engine/Content/Slate/Common/ScrollBoxShadowLeft.png
      resourceObject=null
      tiling=NoTile
      imageSize=<8 16>
      margin=<ltrb 1 0.5 0 0.5>
      outlineSettings {
       bUseBrushTransparency=false
       roundingType=HalfHeightRadius
       width=0
       color {colorUseRule=UseColor_Specified, specifiedColor=<rgba 0 0 0 0>}
       cornerRadii=<0 0 0 0>
      }
      tintColor {colorUseRule=UseColor_Specified, specifiedColor=<rgba 1 1 1 1>}
      uVRegion {bIsValid=false, max=<0 0>, min=<0 0>}
     }
     rightShadowBrush {
      bIsDynamicallyLoaded=false
      drawAs=Box
      imageType=FullColor
      mirroring=NoMirror
      resourceName=../../../Engine/Content/Slate/Common/ScrollBoxShadowRight.png
      resourceObject=null
      tiling=NoTile
      imageSize=<8 16>
      margin=<ltrb 0 0.5 1 0.5>
      outlineSettings {
       bUseBrushTransparency=false
       roundingType=HalfHeightRadius
       width=0
       color {colorUseRule=UseColor_Specified, specifiedColor=<rgba 0 0 0 0>}
       cornerRadii=<0 0 0 0>
      }
      tintColor {colorUseRule=UseColor_Specified, specifiedColor=<rgba 1 1 1 1>}
      uVRegion {bIsValid=false, max=<0 0>, min=<0 0>}
     }
     topShadowBrush {
      bIsDynamicallyLoaded=false
      drawAs=NoDrawType
      imageType=FullColor
      mirroring=NoMirror
      resourceName=../../../Engine/Content/Slate/Common/ScrollBoxShadowTop.png
      resourceObject=null
      tiling=NoTile
      imageSize=<16 8>
      margin=<ltrb 0.5 1 0.5 0>
      outlineSettings {
       bUseBrushTransparency=false
       roundingType=HalfHeightRadius
       width=0
       color {colorUseRule=UseColor_Specified, specifiedColor=<rgba 0 0 0 0>}
       cornerRadii=<0 0 0 0>
      }
      tintColor {colorUseRule=UseColor_Specified, specifiedColor=<rgba 1 1 1 1>}
      uVRegion {bIsValid=false, max=<0 0>, min=<0 0>}
     }
     verticalScrolledContentPadding=<ltrb 0 0 0 1>
    }}
   slot {class=SizeBoxSlot}
  }
  FlavorText {
   widgetClassPath=&CommonTextBlock
   bIsVariable=true
   parent=ScrollBox_0
   widget=FlavorText
   properties {
    text="Lorem Ipsum Dolor Amit Lorem Ipsum Dolor Amit Lorem Ipsum Dolor Amit Lorem Ipsum Dolor Amit Lorem Ipsum Dolor Amit "
    wrapTextAt=650
    font {
     bForceMonospaced=false
     bMaterialIsStencil=false
     fontMaterial=null
     fontObject=&BurbankSmall
     letterSpacing=0
     monospacedWidth=1
     size=20
     skewAmount=0
     typefaceFontName=NotoSans-Regular
     outlineSettings {
      bApplyOutlineToDropShadows=false
      bMiteredCorners=false
      bSeparateFillAlpha=false
      outlineMaterial=null
      outlineSize=0
      outlineColor=<rgba 0 0 0 1>
     }
    }
   }
   slot {class=ScrollBoxSlot}
  }
  SizeBoxEventDatesTitle {
   widgetClassPath=&SizeBox
   parent=VerticalBox_3
   widget=SizeBoxEventDatesTitle
   properties {bOverride_WidthOverride=true, heightOverride=125.6344}
   slot {class=VerticalBoxSlot, properties {verticalAlignment=VAlign_Top, padding=<ltrb 20 0 20 0>}}
  }
  ScaleBox_6 {
   widgetClassPath=&ScaleBox
   parent=SizeBoxEventDatesTitle
   widget=ScaleBox_6
   properties {stretchDirection=DownOnly}
   slot {class=SizeBoxSlot}
  }
  CommonTextBlock_EventDates {
   widgetClassPath=&CommonTextBlock
   bIsVariable=true
   parent=ScaleBox_6
   widget=CommonTextBlock_EventDates
   properties {
    text=EVENT DATES
    colorAndOpacity {colorUseRule=UseColor_Specified, specifiedColor=<rgba 0.0185 1 0.9047 1>}
    font {
     bForceMonospaced=false
     bMaterialIsStencil=false
     fontMaterial=null
     fontObject=&Burbank
     letterSpacing=0
     monospacedWidth=1
     size=40
     skewAmount=0
     typefaceFontName=Burbank-Black
     outlineSettings {
      bApplyOutlineToDropShadows=false
      bMiteredCorners=false
      bSeparateFillAlpha=false
      outlineMaterial=null
      outlineSize=0
      outlineColor=<rgba 0 0 0 1>
     }
    }
   }
   slot {class=ScaleBoxSlot}
  }
  ScaleBox_7 {
   widgetClassPath=&ScaleBox
   parent=VerticalBox_3
   widget=ScaleBox_7
   slot {class=VerticalBoxSlot, properties {size {sizeRule=Fill, value=1}}}
  }
  UniformGridPanel_DateGrid {
   widgetClassPath=/Script/UMG.UniformGridPanel
   bIsVariable=true
   parent=ScaleBox_7
   widget=UniformGridPanel_DateGrid
   properties {slotPadding=<ltrb 10 4 20 4>}
   slot {class=ScaleBoxSlot, properties {verticalAlignment=VAlign_Top}}
  }
  TournamentDateEntry_C_9 {
   widgetClassPath=&/TournamentDateEntry
   bIsVariable=true
   parent=UniformGridPanel_DateGrid
   widget=TournamentDateEntry_C_9
   properties {isTitle=true, title=ROUND 1}
   slot {class=UniformGridSlot, properties {horizontalAlignment=HAlign_Fill, verticalAlignment=VAlign_Fill}}
  }
  TournamentDateEntry {
   widgetClassPath=&TournamentDateEntry
   bIsVariable=true
   parent=UniformGridPanel_DateGrid
   widget=TournamentDateEntry
   properties {title=""}
   slot {class=UniformGridSlot, properties {horizontalAlignment=HAlign_Fill, row=1, verticalAlignment=VAlign_Fill}}
  }
  TournamentDateEntry_C_0 {
   widgetClassPath=&TournamentDateEntry
   bIsVariable=true
   parent=UniformGridPanel_DateGrid
   widget=TournamentDateEntry_C_0
   properties {title=""}
   slot {class=UniformGridSlot, properties {horizontalAlignment=HAlign_Fill, row=2, verticalAlignment=VAlign_Fill}}
  }
  TournamentDateEntry_C_1 {
   widgetClassPath=&TournamentDateEntry
   bIsVariable=true
   parent=UniformGridPanel_DateGrid
   widget=TournamentDateEntry_C_1
   properties {title=""}
   slot {class=UniformGridSlot, properties {horizontalAlignment=HAlign_Fill, row=3, verticalAlignment=VAlign_Fill}}
  }
  TournamentDateEntry_C_2 {
   widgetClassPath=&TournamentDateEntry
   bIsVariable=true
   parent=UniformGridPanel_DateGrid
   widget=TournamentDateEntry_C_2
   properties {title=""}
   slot {class=UniformGridSlot, properties {horizontalAlignment=HAlign_Fill, row=4, verticalAlignment=VAlign_Fill}}
  }
  TournamentDateEntry_C_3 {
   widgetClassPath=&TournamentDateEntry
   bIsVariable=true
   parent=UniformGridPanel_DateGrid
   widget=TournamentDateEntry_C_3
   properties {title=""}
   slot {class=UniformGridSlot, properties {horizontalAlignment=HAlign_Fill, row=5, verticalAlignment=VAlign_Fill}}
  }
  TournamentDateEntry_C_10 {
   widgetClassPath=&TournamentDateEntry
   bIsVariable=true
   parent=UniformGridPanel_DateGrid
   widget=TournamentDateEntry_C_10
   properties {isTitle=true, title=ROUND 2}
   slot {class=UniformGridSlot, properties {column=1, horizontalAlignment=HAlign_Fill, verticalAlignment=VAlign_Fill}}
  }
  TournamentDateEntry_C_4 {
   widgetClassPath=&TournamentDateEntry
   bIsVariable=true
   parent=UniformGridPanel_DateGrid
   widget=TournamentDateEntry_C_4
   properties {title=""}
   slot {class=UniformGridSlot, properties {column=1, horizontalAlignment=HAlign_Fill, row=1, verticalAlignment=VAlign_Fill}}
  }
  TournamentDateEntry_C_5 {
   widgetClassPath=&TournamentDateEntry
   bIsVariable=true
   parent=UniformGridPanel_DateGrid
   widget=TournamentDateEntry_C_5
   properties {title=""}
   slot {class=UniformGridSlot, properties {column=1, horizontalAlignment=HAlign_Fill, row=2, verticalAlignment=VAlign_Fill}}
  }
  TournamentDateEntry_C_6 {
   widgetClassPath=&TournamentDateEntry
   bIsVariable=true
   parent=UniformGridPanel_DateGrid
   widget=TournamentDateEntry_C_6
   properties {title=""}
   slot {class=UniformGridSlot, properties {column=1, horizontalAlignment=HAlign_Fill, row=3, verticalAlignment=VAlign_Fill}}
  }
  TournamentDateEntry_C_7 {
   widgetClassPath=&TournamentDateEntry
   bIsVariable=true
   parent=UniformGridPanel_DateGrid
   widget=TournamentDateEntry_C_7
   properties {title=""}
   slot {class=UniformGridSlot, properties {column=1, horizontalAlignment=HAlign_Fill, row=4, verticalAlignment=VAlign_Fill}}
  }
  TournamentDateEntry_C_8 {
   widgetClassPath=&TournamentDateEntry
   bIsVariable=true
   parent=UniformGridPanel_DateGrid
   widget=TournamentDateEntry_C_8
   properties {title=""}
   slot {class=UniformGridSlot, properties {column=1, horizontalAlignment=HAlign_Fill, row=5, verticalAlignment=VAlign_Fill}}
  }
  TournamentDateEntry_C_11 {
   widgetClassPath=&TournamentDateEntry
   bIsVariable=true
   parent=UniformGridPanel_DateGrid
   widget=TournamentDateEntry_C_11
   properties {isTitle=true, title=ROUND 1}
   slot {class=UniformGridSlot, properties {column=2, horizontalAlignment=HAlign_Fill, verticalAlignment=VAlign_Fill}}
  }
  TournamentDateEntry_C_12 {
   widgetClassPath=&TournamentDateEntry
   bIsVariable=true
   parent=UniformGridPanel_DateGrid
   widget=TournamentDateEntry_C_12
   properties {title=""}
   slot {class=UniformGridSlot, properties {column=2, horizontalAlignment=HAlign_Fill, row=1, verticalAlignment=VAlign_Fill}}
  }
  Spacer_0 {
   widgetClassPath=&Spacer
   parent=VerticalBox_3
   widget=Spacer_0
   properties {size=<1 15>}
   slot {class=VerticalBoxSlot, properties {size {sizeRule=Fill, value=0.4}}}
  }
  VerticalBox_0 {
   widgetClassPath=&VerticalBox
   parent=VerticalBox_3
   widget=VerticalBox_0
   slot {class=VerticalBoxSlot, properties {verticalAlignment=VAlign_Bottom, padding=<ltrb 0 0 0 10>}}
  }
  SizeBox_4 {
   widgetClassPath=&SizeBox
   parent=VerticalBox_0
   widget=SizeBox_4
   properties {bOverride_WidthOverride=true}
   slot {class=VerticalBoxSlot}
  }
  ScaleBox_5 {widgetClassPath=&ScaleBox, parent=SizeBox_4, widget=ScaleBox_5, properties {stretchDirection=DownOnly}, slot {class=SizeBoxSlot}}
  CommonTextBlock_MetaTitle {
   widgetClassPath=&CommonTextBlock
   bIsVariable=true
   parent=ScaleBox_5
   widget=CommonTextBlock_MetaTitle
   properties {
    justification=Center
    text=TOURNAMENT SERIES
    colorAndOpacity {colorUseRule=UseColor_Specified, specifiedColor=<rgba 0.9911 0.9131 0.0409 1>}
    font {
     bForceMonospaced=false
     bMaterialIsStencil=false
     fontMaterial=null
     fontObject=&Burbank
     letterSpacing=0
     monospacedWidth=1
     size=35
     skewAmount=0
     typefaceFontName=BurbankRegular-Black
     outlineSettings {
      bApplyOutlineToDropShadows=false
      bMiteredCorners=false
      bSeparateFillAlpha=false
      outlineMaterial=null
      outlineSize=0
      outlineColor=<rgba 0 0 0 1>
     }
    }
   }
   slot {class=ScaleBoxSlot}
  }
  HorizontalBox_1 {widgetClassPath=&HorizontalBox, parent=VerticalBox_0, widget=HorizontalBox_1, slot {class=VerticalBoxSlot}}
  Image_Zag {
   widgetClassPath=&Image
   parent=HorizontalBox_1
   widget=Image_Zag
   properties {
    brush {
     bIsDynamicallyLoaded=false
     drawAs=Box
     imageType=NoImage
     mirroring=NoMirror
     resourceName=None
     resourceObject=/CompeteOldTab/Frontend/Showdown/Materials/M_UI_HZag_Inst
     tiling=NoTile
     imageSize=<128 12>
     margin=<ltrb 1 0 0 0>
     outlineSettings {
      bUseBrushTransparency=false
      roundingType=HalfHeightRadius
      width=0
      color {colorUseRule=UseColor_Specified, specifiedColor=<rgba 0 0 0 0>}
      cornerRadii=<0 0 0 0>
     }
     tintColor {colorUseRule=UseColor_Specified, specifiedColor=<rgba 1 1 1 1>}
     uVRegion {bIsValid=false, max=<0 0>, min=<0 0>}
    }
    renderTransform {angle=0, scale=<-1 1>, shear=<0 0>, translation=<0 0>}
   }
   slot {class=HorizontalBoxSlot, properties {verticalAlignment=VAlign_Center, padding=<ltrb -15 -5 0 0>, size {sizeRule=Fill, value=1}}}
  }
  SizeBox_3 {
   widgetClassPath=&SizeBox
   parent=HorizontalBox_1
   widget=SizeBox_3
   properties {bOverride_MaxDesiredWidth=true, maxDesiredWidth=535}
   slot {class=HorizontalBoxSlot, properties {padding=<ltrb 20 0 20 0>}}
  }
  ScaleBox_2 {widgetClassPath=&ScaleBox, parent=SizeBox_3, widget=ScaleBox_2, properties {stretchDirection=DownOnly}, slot {class=SizeBoxSlot}}
  CommonTextBlock_MetaDate {
   widgetClassPath=&CommonTextBlock
   bIsVariable=true
   parent=ScaleBox_2
   widget=CommonTextBlock_MetaDate
   properties {
    justification=Center
    text=DATE RANGE
    font {
     bForceMonospaced=false
     bMaterialIsStencil=false
     fontMaterial=null
     fontObject=&BurbankSmall
     letterSpacing=0
     monospacedWidth=1
     size=28
     skewAmount=0
     typefaceFontName=NotoSans-Regular
     outlineSettings {
      bApplyOutlineToDropShadows=false
      bMiteredCorners=false
      bSeparateFillAlpha=false
      outlineMaterial=null
      outlineSize=0
      outlineColor=<rgba 0 0 0 1>
     }
    }
   }
   slot {class=ScaleBoxSlot}
  }
  Image_Zag2 {
   widgetClassPath=&Image
   parent=HorizontalBox_1
   widget=Image_Zag2
   properties {brush {
     bIsDynamicallyLoaded=false
     drawAs=Box
     imageType=NoImage
     mirroring=NoMirror
     resourceName=None
     resourceObject=&M_UI_HZag_Inst
     tiling=NoTile
     imageSize=<128 12>
     margin=<ltrb 1 0 0 0>
     outlineSettings {
      bUseBrushTransparency=false
      roundingType=HalfHeightRadius
      width=0
      color {colorUseRule=UseColor_Specified, specifiedColor=<rgba 0 0 0 0>}
      cornerRadii=<0 0 0 0>
     }
     tintColor {colorUseRule=UseColor_Specified, specifiedColor=<rgba 1 1 1 1>}
     uVRegion {bIsValid=false, max=<0 0>, min=<0 0>}
    }}
   slot {class=HorizontalBoxSlot, properties {verticalAlignment=VAlign_Center, padding=<ltrb 0 -5 -15 0>, size {sizeRule=Fill, value=1}}}
  }
  HorizontalBox_3 {
   widgetClassPath=&HorizontalBox
   parent=VerticalBox_3
   widget=HorizontalBox_3
   slot {class=VerticalBoxSlot, properties {horizontalAlignment=HAlign_Center, verticalAlignment=VAlign_Center, padding=<ltrb 20 0 20 0>}}
  }
  Image_FooterWatermark {
   widgetClassPath=&Image
   bIsVariable=true
   parent=HorizontalBox_3
   widget=Image_FooterWatermark
   properties {brush {
     bIsDynamicallyLoaded=false
     drawAs=Image
     imageType=NoImage
     mirroring=NoMirror
     resourceName=None
     resourceObject=/CompeteOldTab/Frontend/Showdown/Materials/M_UI_TournamentMark
     tiling=NoTile
     imageSize=<68 68>
     margin=<ltrb 0 0 0 0>
     outlineSettings {
      bUseBrushTransparency=false
      roundingType=HalfHeightRadius
      width=0
      color {colorUseRule=UseColor_Specified, specifiedColor=<rgba 0 0 0 0>}
      cornerRadii=<0 0 0 0>
     }
     tintColor {colorUseRule=UseColor_Specified, specifiedColor=<rgba 1 1 1 1>}
     uVRegion {bIsValid=false, max=<0 0>, min=<0 0>}
    }}
   slot {class=HorizontalBoxSlot, properties {horizontalAlignment=HAlign_Center, verticalAlignment=VAlign_Center}}
  }
  Text_TimeRemaining {widgetClassPath=&CommonTextBlock, bInherited=true, widget=null}
  Text_DebugId {widgetClassPath=&TextBlock, bInherited=true, widget=null, properties=~}
  Image_PosterFront {widgetClassPath=&EpicCMSImage, bInherited=true, widget=null}
  Image_LoadingScreen {widgetClassPath=&EpicCMSImage, bInherited=true, widget=null}
  Image_PlaylistTile {widgetClassPath=&EpicCMSImage, bInherited=true, widget=null}
 ]
}
