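# Tournament session details widget for the Showdown compete tab, derived from FortShowdownDetailView.
# Refresh Data fills title, session number, dates, score rows, big stats and bus fare from event and results data.
# A switcher shows stats, countdown or a status message by participation state, with wording chosen by eligibility and event state.
# Tournament colors tint the subheader and buttons; leaderboard and payout buttons animate on hover and fire delegates on click.
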WidgetBlueprint /CompeteOldTab/Frontend/Showdown/TournamentEventWindowDetails
animations [
 RefreshDate {
  binding_count=2
  display_label=""
  length_seconds=0.4
  object=&self:RefreshDate
  sequence {bindings [
    EventDateText {id=#1, possessed_class=CommonTextBlock, tracks []}
    HorizontalBoxSlot_1 {
     id=#2
     parent_id=#1
     possessed_class=HorizontalBoxSlot
     tracks [{
       type=MovieSceneMarginTrack
       display_name=Padding
       sections [{
         type=MovieSceneMarginSection
         end_frame=7
         start_frame=0
         channels [
          Left {type=MovieSceneScriptingFloatChannel}
          Top {type=MovieSceneScriptingFloatChannel}
          Right {type=MovieSceneScriptingFloatChannel}
          Bottom {type=MovieSceneScriptingFloatChannel, keys [[0,-55.1346],[4.9987,0],[5.4837,0]]}
         ]
        }]
      }]
    }
   ]}
 }
 RefreshTime {
  binding_count=2
  display_label=""
  length_seconds=0.4
  object=&self:RefreshTime
  sequence {bindings [
    EventTimeText {id=#3, possessed_class=CommonTextBlock, tracks []}
    HorizontalBoxSlot_2 {
     id=#4
     parent_id=#3
     possessed_class=HorizontalBoxSlot
     tracks [{
       type=MovieSceneMarginTrack
       display_name=Padding
       sections [{
         type=MovieSceneMarginSection
         end_frame=8
         start_frame=0
         channels [
          Left {type=MovieSceneScriptingFloatChannel}
          Top {type=MovieSceneScriptingFloatChannel}
          Right {type=MovieSceneScriptingFloatChannel}
          Bottom {type=MovieSceneScriptingFloatChannel, keys [[2.999,-65.0579],[7.9983,0]]}
         ]
        }]
      }]
    }
   ]}
 }
 RefreshData {
  binding_count=1
  display_label=""
  length_seconds=0.3936
  object=&self:RefreshData
  sequence {bindings [CommonWidgetSwitcher_isFuture {
     id=#5
     possessed_class=CommonWidgetSwitcherLegacy
     tracks [{
       type=MovieSceneFloatTrack
       display_name=Render Opacity
       sections [{type=MovieSceneFloatSection, start_frame=0, channels [None {type=MovieSceneScriptingFloatChannel, keys [[0,0],[0.9597,0],[7.871,1]]}]}]
      }]
    }]}
 }
 LeaderboardHover {
  binding_count=1
  display_label=Leaderboard Hover
  length_seconds=0.3512
  object=&self:LeaderboardHover
  sequence {bindings [CommonBorderLeaderboardButtonHover {
     id=#6
     possessed_class=CommonBorder
     tracks [{
       type=MovieSceneColorTrack
       display_name=Brush Color
       sections [{
         type=MovieSceneColorSection
         end_frame=7
         start_frame=0
         channels [
          Color.R {type=MovieSceneScriptingFloatChannel}
          Color.G {type=MovieSceneScriptingFloatChannel}
          Color.B {type=MovieSceneScriptingFloatChannel}
          Color.A {type=MovieSceneScriptingFloatChannel, keys [[0,0],[7.024,1]]}
         ]
        }]
      }]
    }]}
 }
 PayoutHover {
  binding_count=1
  length_seconds=0.3512
  object=&self:PayoutHover
  sequence {bindings [CommonBorderLeaderboardButtonHover {
     id=#6
     display_name=CommonBorderPayoutButtonHover
     possessed_class=CommonBorder
     tracks [{
       type=MovieSceneColorTrack
       display_name=Brush Color
       sections [{
         type=MovieSceneColorSection
         end_frame=7
         start_frame=0
         channels [
          Color.R {type=MovieSceneScriptingFloatChannel}
          Color.G {type=MovieSceneScriptingFloatChannel}
          Color.B {type=MovieSceneScriptingFloatChannel}
          Color.A {type=MovieSceneScriptingFloatChannel, keys [[0,0],[7.024,1]]}
         ]
        }]
      }]
    }]}
 }
]
graphs [
 EventGraph {nodes [
   "On Button Clicked (Button_Leaderboard)"!() -> (OutputDelegate, then, Button:Common Button Legacy Object Reference) {is=title K2Node_ComponentBoundEvent}
   > Branch(Condition=IsEventLeaderboardAvailable!(self:Fort Showdown Detail View Object Reference) -> (ReturnValue:bool))
   > Call View Leaderboard!(
    execute:Exec
    self:Tournament Event Window Details Object Reference
    TournamentId:String=GetTournamentId!(self:Fort Showdown Detail View Object Reference) -> (ReturnValue:String)
    EventId:String=GetEventId!(self:Fort Showdown Detail View Object Reference) -> (ReturnValue:String)
   ) -> (then:Exec) {is=title K2Node_CallDelegate}
   "On Button Hovered (Button_Leaderboard)"!() -> (OutputDelegate, then, Button:Common Button Legacy Object Reference) {is=title K2Node_ComponentBoundEvent}
   n1 = Get LeaderboardHover() -> (LeaderboardHover:Widget Animation Object Reference)
   > Anim Transition!(Exec:Exec, In:bool=true, Animation:Widget Animation Object Reference=@n1, PlaySpeed:double=3, Widget:User Widget Object Reference) -> (ExecOut:Exec) {is=title K2Node_MacroInstance}
   "On Button Unhovered (Button_Leaderboard)"!() -> (OutputDelegate, then, Button:Common Button Legacy Object Reference) {is=title K2Node_ComponentBoundEvent}
   > Anim Transition(Animation=@n1, PlaySpeed=1)
   "On Button Clicked (Button_Payout)"!() -> (OutputDelegate, then, Button:Common Button Legacy Object Reference) {is=title K2Node_ComponentBoundEvent}
   > Call View Payouts!(execute:Exec, self:Tournament Event Window Details Object Reference, EventWindowId:String=GetEventId(), EventId:String=GetTournamentId()) -> (then:Exec) {is=title K2Node_CallDelegate}
   "On Button Hovered (Button_Payout)"!() -> (OutputDelegate, then, Button:Common Button Legacy Object Reference) {is=title K2Node_ComponentBoundEvent}
   n2 = Get PayoutHover() -> (PayoutHover:Widget Animation Object Reference)
   > Anim Transition(In=true, Animation=@n2, PlaySpeed=3)
   "On Button Unhovered (Button_Payout)"!() -> (OutputDelegate, then, Button:Common Button Legacy Object Reference) {is=title K2Node_ComponentBoundEvent}
   > Anim Transition(Animation=@n2, PlaySpeed=1)
   EventColorize() {is=title K2Node_CustomEvent}
   n3 = Break Fort Tournament Display Info!(FortTournamentDisplayInfo:"Fort Tournament Display Info Structure (by ref)"=Get TournamentDisplayInfo() -> (TournamentDisplayInfo:Fort Tournament Display Info Structure)) -> (
    TitleLine1:Text
    TitleLine2:Text
    ScheduleInfo:Text
    FlavorDescription:Text
    DetailsDescription:Text
    ShortFormatTitle:Text
    LongFormatTitle:Text
    BackgroundTitle:Text
    PinScoreRequirement:int
    PinEarnedText:Text
    BaseColor:Linear Color Structure
    PrimaryColor:Linear Color Structure
    SecondaryColor:Linear Color Structure
    HighlightColor:Linear Color Structure
    TitleColor:Linear Color Structure
    ShadowColor:Linear Color Structure
    BackgroundLeftColor:Linear Color Structure
    BackgroundRightColor:Linear Color Structure
    BackgroundTextColor:Linear Color Structure
    PosterFadeColor:Linear Color Structure
    AlertText:Text
    SeriesPointLeaderboardName:Text
    AlertType:EFortTournamentAlertType Enum
    RoundDisplayNames:Array of Texts
   ) {is=title K2Node_BreakStruct}
   > SetBrushColor(
    self=[
     Get Border_SubheaderBG() -> (Border_SubheaderBG:Border Object Reference)
     Get CommonBorder_PayoutButtonBG() -> (CommonBorder_PayoutButtonBG:Common Border Object Reference)
     Get CommonBorder_LeaderboardButtonBG() -> (CommonBorder_LeaderboardButtonBG:Common Border Object Reference)
    ]
    InBrushColor=@n3.SecondaryColor
   )
   > SetColorAndOpacity(
    self=Get MatchTypeText() -> (MatchTypeText:Common Text Object Reference)
    InColorAndOpacity=Make Slate Color(SpecifiedColor=@n3.PrimaryColor)
   )
   Event Refresh Data() {is=title K2Node_Event}
   > SetVisibility(
    self=Get Button_Leaderboard() -> (Button_Leaderboard:Events Simple Button Object Reference)
    InVisibility=Select(Option 0=Collapsed, Option 1=Visible, Index=IsEventLeaderboardAvailable())
   )
   > Set TimeRemainingFormat(TimeRemainingFormat:Text=Select!(
     Unknown:Text="LOCTEXT(\"This session starts in {0}.\\r\\nCome back then to test your mettle!\")"
     Public:Text="LOCTEXT(\"This session starts in {0}.\\r\\nCome back then to test your mettle!\")"
     Private:Text="LOCTEXT(\"This session starts in {0}. \\r\\nYou\\'ve been granted access to this session! Come back when the session is live to climb your way to the top!\")"
     Locked:Text="LOCTEXT(\"This session starts in {0}.\\r\\nYou don\\'t yet qualify to participate. Check the Tournament Details for more information.\")"
     Index:EFortEventWindowEligibility Enum=GetEventWindowEligibility!(self:Fort Showdown Detail View Object Reference) -> (ReturnValue:EFortEventWindowEligibility Enum=Unknown)
    ) -> (ReturnValue:Text)) -> (Output_Get:Text)
   n4 = GetEventWindowEligibility() -> (ReturnValue=Unknown)
   > SetText(
    self=Get RichTextBlock_OtherReason() -> (RichTextBlock_OtherReason:Rich Text Block Object Reference)
    InText=Select!(
     Unknown:"Text (by ref)"
     FutureTBD:"Text (by ref)"
     FutureScheduled:"Text (by ref)"
     FutureNext:"Text (by ref)"
     Live:"Text (by ref)"=Select!(
      Unknown:Text="LOCTEXT(\"<WindowLocked>This session is </><WindowLockedB>CURRENTLY LIVE</><WindowLocked>, but we can\\'t determine your eligibility at this time.</>\")"
      Public:Text="LOCTEXT(\"<WindowLocked>This session is </><WindowLockedB>LIVE!</>\")"
      Private:Text="LOCTEXT(\"<WindowLocked>This session is </><WindowLockedB>LIVE!</> <WindowLocked>You have unlocked access to participate, so what are you waiting for?</>\")"
      Locked:Text="LOCTEXT(\"<WindowLocked>This session is currently LIVE, but access requires prior player qualification.\\r\\nCheck out the tournament details to see if this session is streaming to follow the action!</>\")"
      Index:EFortEventWindowEligibility Enum=@n4
     ) -> (ReturnValue:Text)
     LiveParticipating:"Text (by ref)"
     LiveNotParticipating:"Text (by ref)"
     Completed:"Text (by ref)"=Select(
      Unknown="LOCTEXT(\"<WindowLocked>This session </><WindowLockedB>HAS CONCLUDED!</> \")"
      Public="LOCTEXT(\"<WindowLocked>This session </><WindowLockedB>HAS CONCLUDED!</> <WindowLocked>Looks like you didn\\'t play in this one.</>\")"
      Private="LOCTEXT(\"<WindowLocked>This session </><WindowLockedB>HAS CONCLUDED!</>\\r\\n<WindowLocked>Check out the leaderboard to see who came out on top!</>\")"
      Locked="LOCTEXT(\"<WindowLocked>This session </><WindowLockedB>HAS CONCLUDED!</>\\r\\n<WindowLocked>Check out the leaderboard to see who came out on top!</>\")"
      Index=@n4
     )
     CompletedParticipated:"Text (by ref)"
     CompletedNotPartipated:"Text (by ref)"
     Cancelled:"Text (by ref)"=Select(
      Unknown="LOCTEXT(\"<WindowLocked>This session </><WindowLockedB>HAS CONCLUDED!</> \")"
      Public="LOCTEXT(\"<WindowLocked>This event </><WindowLockedB>HAS CONCLUDED!</> <WindowLocked>Looks like you didn\\'t play in this one.</>\")"
      Private="LOCTEXT(\"<WindowLocked>This event </><WindowLockedB>HAS CONCLUDED!</> <WindowLocked>Check out the leaderboard to see who came out on top!</>\")"
      Locked="LOCTEXT(\"<WindowLocked>This event </><WindowLockedB>HAS CONCLUDED!</> <WindowLocked>Check out the leaderboard to see who came out on top!</>\")"
      Index=@n4
     )
     Index:EFortShowdownEventState Enum=GetEventTimeState!(self:Fort Showdown Detail View Object Reference) -> (ReturnValue:EFortShowdownEventState Enum=Unknown)
    ) -> (ReturnValue:"Text (by ref)")
   )
   > SetActiveWidgetIndex(
    self=Get CommonWidgetSwitcher_isFuture() -> (CommonWidgetSwitcher_isFuture:Common Widget Switcher Legacy Object Reference)
    Index=Select!(
     Unknown:int=0
     FutureTBD:int=1
     FutureScheduled:int=1
     FutureNext:int=1
     Live:int=0
     LiveParticipating:int=0
     LiveNotParticipating:int=0
     Completed:int=0
     CompletedParticipated:int=0
     CompletedNotPartipated:int=2
     Cancelled:int=2
     Index:EFortShowdownEventState Enum=GetEventPlayerParticipationState!(self:Fort Showdown Detail View Object Reference) -> (ReturnValue:EFortShowdownEventState Enum=Unknown)
    ) -> (ReturnValue:int)
   )
   > n5 = Sequence!(execute) -> (then_0, then_1, then_2:Exec, then_3:Exec, then_4:Exec, then_5:Exec, then_6:Exec, then_7:Exec)
   n6 = GetEventResultsSummary!(execute=@n5.then_1, self:Fort Showdown Detail View Object Reference) -> (then, ReturnValue:Fort Showdown Event Best Results Summary Structure)
   n7 = Break Fort Showdown Event Best Results Summary!(FortShowdownEventBestResultsSummary:"Fort Showdown Event Best Results Summary Structure (by ref)"=@n6) -> (
    TotalScore:int
    MatchesPlayed:int
    NumVictoryRoyales:int
    PlacementPoints:int
    EliminationPoints:int
    EntryFeePoints:int
    bIsValidSummary:bool=~
   ) {is=title K2Node_BreakStruct}
   > SetData!(
    execute
    self:Event Score Row Object Reference=Get VictoryRoyaleInfo() -> (VictoryRoyaleInfo:Event Score Row Object Reference)
    Label:Text="LOCTEXT(\"Victory Royale\")"
    Value:int=@n7.NumVictoryRoyales
   ) -> (then)
   > SetData(self=Get PlacementInfo() -> (PlacementInfo:Event Score Row Object Reference), Label="LOCTEXT(\"Placement\")", Value=@n7.PlacementPoints)
   > SetData(
    self=Get EliminationsInfo() -> (EliminationsInfo:Event Score Row Object Reference)
    Label="LOCTEXT(\"Eliminations\")"
    Value=@n7.EliminationPoints
   )
   n8 = Get TournamentDisplayInfo()
   n9 = Get PointsScoredInfo() -> (PointsScoredInfo:Event Score Big Stat Object Reference)
   > SetData!(
    execute
    self:Event Score Big Stat Object Reference=@n9
    Label="LOCTEXT(\"TOTAL\\r\\nPOINTS\")"
    Value=@n7
    ValueContext:int=GetTournamentPinUnlockScore!(self:Fort Showdown Detail View Object Reference) -> (ReturnValue:int=0)
    FormatterText:Text="LOCTEXT(\"<WindowDetailsPointTotal>/ {0}</> <WindowDetailsPointTotal_Small>pts</>\")"
    Tournament Display Info:Fort Tournament Display Info Structure=@n8
   ) -> (then)
   > SetData(
    self=Get MatchesPlayedInfo() -> (MatchesPlayedInfo:Event Score Big Stat Object Reference)
    Label="LOCTEXT(\"MATCHES\\r\\nPLAYED\")"
    Value=@n7.MatchesPlayed
    ValueContext=GetTournamentMatchCap!(self:Fort Showdown Detail View Object Reference) -> (ReturnValue:int=0)
    FormatterText="LOCTEXT(\"<WindowDetailsPointTotal>/ {0}</> <WindowDetailsPointTotal_Small>max</>\")"
    Tournament Display Info=@n8
   )
   n10 = Break Showdown Tournament Entry!(ShowdownTournamentEntry:"Showdown Tournament Entry Structure (by ref)"=Get CMSTournamentData() -> (CMSTournamentData:Showdown Tournament Entry Structure)) -> (
    TournamentDisplayId:String
    TitleLine1:String
    TitleLine2:String
    ScheduleInfo:String
    PosterFrontImage:String
    PosterBackImage:String
    FlavorDescription:String
    DetailsDescription:String
    ShortFormatTitle:String
    LongFormatTitle:String
    BackgroundTitle:String
    PinScoreRequirement:int
    PinEarnedText:String
    BaseColor:String
    PrimaryColor:String
    SecondaryColor:String
    HighlightColor:String
    TitleColor:String
    ShadowColor:String
    BackgroundLeftColor:String
    BackgroundRightColor:String
    BackgroundTextColor:String
    PosterFadeColor:String
    PlaylistTileImage:String
    LoadingScreenImage:String
    TournamentViewBackgroundImage:String
    TournamentLandingPageImage:String
    SquarePosterImage:String
    StyleInfoId:String
    AlertText:String
    SeriesPointLeaderboardName:String
    PlaylistDescription:String
    AlertType:EFortTournamentAlertType Enum
    RoundDisplayNames:Array of Texts
   ) {is=title K2Node_BreakStruct}
   "SetText (Text)"(
    execute=@n5.then_2
    self=Get TournamentContextText() -> (TournamentContextText:Common Text Object Reference)
    InText=TextToUpper!(InText:"Text (by ref)"="To Text (String)"(InString=@n10.LongFormatTitle)) -> (ReturnValue:Text)
   )
   n11 = "integer > integer"(A=@n7.MatchesPlayed, B=0) {is=title K2Node_CallFunction}
   n12 = "Not Equal (Enum)"!(
    A:EFortShowdownMatchType Enum=GetTournamentMatchType!(self:Fort Showdown Detail View Object Reference) -> (ReturnValue:EFortShowdownMatchType Enum=Unknown)
    B:EFortShowdownMatchType Enum=Solo
   ) -> (ReturnValue:bool)
   > SetVisibility(
    self=Get HorizontalBox_SquadInfoContainer() -> (HorizontalBox_SquadInfoContainer:Horizontal Box Object Reference)
    InVisibility=Select!(Option 0:ESlateVisibility Enum=Collapsed, Option 1:ESlateVisibility Enum=Collapsed, Index:bool=AND Boolean(A=@n12, B=@n11)) -> (ReturnValue:ESlateVisibility Enum)
   )
   Anim Transition(Exec=@n5.then_3, In=true, Animation=Get RefreshData() -> (RefreshData:Widget Animation Object Reference), PlaySpeed=1)
   EventColorize(execute=@n5.then_4) {is=title K2Node_CallFunction}
   n13 = IsEventLeaderboardAvailable()
   n14 = Get AllowLeaderboardAccess() -> (AllowLeaderboardAccess:bool)
   SetVisibility(
    execute=@n5.then_5
    self=[Get Overlay_LeaderboardButton() -> (Overlay_LeaderboardButton:Overlay Object Reference), Get CommonBorder_LeaderboardButtonBG()]
    InVisibility=Select(Option 0=Collapsed, Option 1=Visible, Index=AND Boolean(A=@n13, B=@n14))
   )
   n15 = GetEventResultsSummary(execute=@n5.then_6)
   n16 = Break Fort Showdown Event Best Results Summary(FortShowdownEventBestResultsSummary=@n15) -> (bIsValidSummary=~)
   > "SetText (Text)"(
    self=Get Text_BusFare() -> (Text_BusFare:Common Text Object Reference)
    InText=Format Text!(Format="LOCTEXT(\"({0} BUS FARE PAID)\")", 0:int=@n16.EntryFeePoints) -> (Result)
   )
   > SetVisibility(
    self=Get Text_BusFare()
    InVisibility=Select(
     Option 0=Hidden
     Option 1=SelfHitTestInvisible
     Index="Not Equal (Integer)"!(A:int=@n16.EntryFeePoints, B:int=0) -> (ReturnValue:bool) {is=title K2Node_CallFunction}
    )
   )
   n17 = GetEventWindowSessionNumber!(self:Fort Showdown Detail View Object Reference) -> (ReturnValue:int=0)
   "SetText (Text)"(
    execute=@n5.then_0
    self=Get MatchTypeText()
    InText=TextToUpper(InText=Format Text!(Format="LOCTEXT(\"Session {SessionNumber}\")", SessionNumber:int=@n17) -> (Result))
   )
   n18 = Get EventDateText() -> (EventDateText:Common Text Object Reference)
   n19 = GetEventDateAsText!(
    self:Fort Showdown Detail View Object Reference
    DateTime:Date Time Structure=GetEventEndTime!(self:Fort Showdown Detail View Object Reference) -> (ReturnValue:Date Time Structure)
    Style:EFortDateTimeStyle Enum=Short
   ) -> (ReturnValue:Text)
   n20 = GetEventDateAsText(DateTime=GetEventBeginTime!(self:Fort Showdown Detail View Object Reference) -> (ReturnValue:Date Time Structure), Style=Short)
   n21 = GetEventDateAsText(DateTime=GetEventEndTime(), Style=Short)
   n22 = Format Text!(
    Format="LOCTEXT(\"{StartDate} {StartTime} -{EndDate} {EndTime}\")"
    StartDate:Text=GetEventDateAsText(DateTime=GetEventBeginTime(), Style=Short)
    StartTime:Text=GetEventTimeAsText!(self:Fort Showdown Detail View Object Reference, DateTime:Date Time Structure=GetEventBeginTime(), Style:EFortDateTimeStyle Enum=Short) -> (ReturnValue:Text)
    EndDate:Text=Select(
     Option 0:Text=Format Text!(Format="LOCTEXT(\" {EndDate}\")", EndDate=@n21) -> (Result)
     Option 1:Text
     Index="Equal Exactly (Text)"!(A:"Text (by ref)"=@n20, B:"Text (by ref)"=@n19) -> (ReturnValue:bool)
    ) -> (ReturnValue:Text)
    EndTime:Text=GetEventTimeAsText(DateTime=GetEventEndTime(), Style=Long)
   ) -> (Result)
   > n23 = Branch(Condition="Not Equal Exactly (Text)"!(A:"Text (by ref)"="GetText (Text)"!(self:Text Object Reference=@n18) -> (ReturnValue:Text), B:"Text (by ref)"=@n22) -> (ReturnValue:bool))
   > n24 = PlayAnimation(InAnimation=Get RefreshDate() -> (RefreshDate:Widget Animation Object Reference))
   "SetText (Text)"(execute=[@n24.then, @n23.else], self=Get EventDateText(), InText=@n22)
   Reroute Node(InputPin:Event Score Big Stat Object Reference=@n9) -> (OutputPin:Event Score Big Stat Object Reference)
  ]}
 ViewLeaderboard {nodes [ViewLeaderboard!() -> (then, TournamentId:String, EventId:String) {is=title K2Node_FunctionEntry}]}
 ViewPayouts {nodes [ViewPayouts!() -> (then, EventWindowId:String, EventId:String) {is=title K2Node_FunctionEntry}]}
]
variables [
 AllowLeaderboardAccess {type=boolean}
 ViewLeaderboard {}
 ViewPayouts {}
]
widgets {
 info {inheritedWidgetCount=6, parentClass=/Script/FortniteUI.FortShowdownDetailView, rootWidgetClass=&Overlay, widgetCount=42}
 widgets [
  Overlay_0 {widgetClassPath=&Overlay, widget=Overlay_0}
  VerticalBox_5 {
   widgetClassPath=&VerticalBox
   parent=Overlay_0
   widget=VerticalBox_5
   slot {class=OverlaySlot, properties {horizontalAlignment=HAlign_Fill, verticalAlignment=VAlign_Fill}}
  }
  TournamentContextText {
   widgetClassPath=&CommonTextBlock
   bIsVariable=true
   parent=VerticalBox_5
   widget=TournamentContextText
   properties {
    text=COMPETITIVE MATCH
    font {
     bForceMonospaced=false
     bMaterialIsStencil=false
     fontMaterial=null
     fontObject=&Burbank
     letterSpacing=0
     monospacedWidth=1
     size=53
     skewAmount=0
     typefaceFontName=Burbank-Black
     outlineSettings {
      bApplyOutlineToDropShadows=false
      bMiteredCorners=false
      bSeparateFillAlpha=false
      outlineMaterial=null
      outlineSize=0
      outlineColor=<rgba 0 0 0 1>
     }
    }
   }
   slot {class=VerticalBoxSlot, properties {padding=<ltrb 30 0 0 -14>}}
  }
  SizeBox_0 {
   widgetClassPath=&SizeBox
   parent=VerticalBox_5
   widget=SizeBox_0
   properties {bOverride_HeightOverride=true, heightOverride=56}
   slot {class=VerticalBoxSlot, properties {horizontalAlignment=HAlign_Left}}
  }
  HorizontalBox_1 {widgetClassPath=&HorizontalBox, parent=SizeBox_0, widget=HorizontalBox_1, slot {class=SizeBoxSlot}}
  SizeBox_2 {
   widgetClassPath=&SizeBox
   parent=HorizontalBox_1
   widget=SizeBox_2
   properties {bOverride_WidthOverride=true, maxDesiredWidth=800, widthOverride=1100}
   slot {class=HorizontalBoxSlot, properties {horizontalAlignment=HAlign_Left, size {sizeRule=Fill, value=1}}}
  }
  Border_SubheaderBG {
   widgetClassPath=&Border
   bIsVariable=true
   parent=SizeBox_2
   widget=Border_SubheaderBG
   properties {
    clipping=ClipToBounds
    verticalAlignment=VAlign_Center
    background {
     bIsDynamicallyLoaded=false
     drawAs=Box
     imageType=NoImage
     mirroring=NoMirror
     resourceName=None
     resourceObject=/CompeteOldTab/Frontend/Showdown/Materials/M_UI_SkewBox_SubBarBG
     tiling=NoTile
     imageSize=<171.3667 64>
     margin=<ltrb 0.1 0 0.1 0>
     outlineSettings {
      bUseBrushTransparency=false
      roundingType=HalfHeightRadius
      width=0
      color {colorUseRule=UseColor_Specified, specifiedColor=<rgba 0 0 0 0>}
      cornerRadii=<0 0 0 0>
     }
     tintColor {colorUseRule=UseColor_Specified, specifiedColor=<rgba 1 1 1 1>}
     uVRegion {bIsValid=false, max=<0 0>, min=<0 0>}
    }
    brushColor=<rgba 0 0.2086 1 1>
    padding=<ltrb 30 -10 20 -10>
   }
   slot {class=SizeBoxSlot}
  }
  HorizontalBox_0 {
   widgetClassPath=&HorizontalBox
   parent=Border_SubheaderBG
   widget=HorizontalBox_0
   slot {class=BorderSlot, properties {verticalAlignment=VAlign_Center, padding=<ltrb 30 -10 20 -10>}}
  }
  MatchTypeText {
   widgetClassPath=&CommonTextBlock
   bIsVariable=true
   parent=HorizontalBox_0
   widget=MatchTypeText
   properties {
    scrollStyle=&TextScrollStyle-Base_C
    text=SOLO EVENT
    colorAndOpacity {colorUseRule=UseColor_Specified, specifiedColor=<rgba 0.0185 1 0.9047 1>}
    font {
     bForceMonospaced=false
     bMaterialIsStencil=false
     fontMaterial=null
     fontObject=&Burbank
     letterSpacing=0
     monospacedWidth=1
     size=50
     skewAmount=0
     typefaceFontName=Burbank-Black
     outlineSettings {
      bApplyOutlineToDropShadows=false
      bMiteredCorners=false
      bSeparateFillAlpha=false
      outlineMaterial=null
      outlineSize=0
      outlineColor=<rgba 0 0 0 1>
     }
    }
    renderTransform {angle=0, scale=<1 0.75>, shear=<-8 0>, translation=<0 0>}
    renderTransformPivot=<0 0.5>
   }
   slot {class=HorizontalBoxSlot, properties {verticalAlignment=VAlign_Center, padding=<ltrb 0 -10 0 -10>}}
  }
  EventDateText {
   widgetClassPath=&CommonTextBlock
   bIsVariable=true
   parent=HorizontalBox_0
   widget=EventDateText
   properties {
    scrollStyle=&TextScrollStyle-Base_C
    text=10.10.2018
    font {
     bForceMonospaced=false
     bMaterialIsStencil=false
     fontMaterial=null
     fontObject=&Burbank
     letterSpacing=0
     monospacedWidth=1
     size=35
     skewAmount=0
     typefaceFontName=Burbank-PlayerPower
     outlineSettings {
      bApplyOutlineToDropShadows=false
      bMiteredCorners=false
      bSeparateFillAlpha=false
      outlineMaterial=null
      outlineSize=0
      outlineColor=<rgba 0 0 0 1>
     }
    }
   }
   slot {class=HorizontalBoxSlot, properties {verticalAlignment=VAlign_Center, padding=<ltrb 40 0 0 0>, size {sizeRule=Fill, value=1}}}
  }
  Overlay_LeaderboardButton {
   widgetClassPath=&Overlay
   bIsVariable=true
   parent=HorizontalBox_1
   widget=Overlay_LeaderboardButton
   slot {class=HorizontalBoxSlot, properties {horizontalAlignment=HAlign_Left}}
  }
  CommonBorder_LeaderboardButtonBG {
   widgetClassPath=&CommonBorder
   bIsVariable=true
   parent=Overlay_LeaderboardButton
   widget=CommonBorder_LeaderboardButtonBG
   properties {
    background {
     bIsDynamicallyLoaded=false
     drawAs=Box
     imageType=NoImage
     mirroring=NoMirror
     resourceName=None
     resourceObject=/CompeteOldTab/Frontend/Showdown/Materials/M_UI_SkewBox_SubBarBG2
     tiling=NoTile
     imageSize=<171.3667 64>
     margin=<ltrb 0.1 0 0.1 0>
     outlineSettings {
      bUseBrushTransparency=false
      roundingType=HalfHeightRadius
      width=0
      color {colorUseRule=UseColor_Specified, specifiedColor=<rgba 0 0 0 0>}
      cornerRadii=<0 0 0 0>
     }
     tintColor {colorUseRule=UseColor_Specified, specifiedColor=<rgba 1 1 1 1>}
     uVRegion {bIsValid=false, max=<0 0>, min=<0 0>}
    }
    brushColor=<rgba 0 0.2086 1 1>
    padding=<ltrb 0 0 0 0>
   }
   slot {class=OverlaySlot, properties {horizontalAlignment=HAlign_Fill, verticalAlignment=VAlign_Fill}}
  }
  CommonBorderLeaderboardButtonHover {
   widgetClassPath=&CommonBorder
   parent=CommonBorder_LeaderboardButtonBG
   widget=CommonBorderLeaderboardButtonHover
   properties {
    background {
     bIsDynamicallyLoaded=false
     drawAs=Box
     imageType=NoImage
     mirroring=NoMirror
     resourceName=None
     resourceObject=&M_UI_SkewBox_SubBarBG2
     tiling=NoTile
     imageSize=<171.3667 64>
     margin=<ltrb 0.1 0 0.1 0>
     outlineSettings {
      bUseBrushTransparency=false
      roundingType=HalfHeightRadius
      width=0
      color {colorUseRule=UseColor_Specified, specifiedColor=<rgba 0 0 0 0>}
      cornerRadii=<0 0 0 0>
     }
     tintColor {colorUseRule=UseColor_Specified, specifiedColor=<rgba 1 1 1 1>}
     uVRegion {bIsValid=false, max=<0 0>, min=<0 0>}
    }
    brushColor=<rgba 1 1 1 0>
    padding=<ltrb 20 -5 20 0>
   }
   slot {class=BorderSlot, properties {padding=<ltrb 0 0 0 0>}}
  }
  Button_Leaderboard {
   widgetClassPath=&/EventsSimpleButton
   bIsVariable=true
   parent=CommonBorderLeaderboardButtonHover
   widget=Button_Leaderboard
   properties {
    buttonNameText=Session Leaderboard
    clickMethod=DownAndUp
    onButtonBaseClicked=(&self:WidgetTree.Button_Leaderboard.HandleOnButtonClicked)
    onButtonBaseDoubleClicked=(&self:WidgetTree.Button_Leaderboard.HandleOnButtonDoubleClicked)
    onButtonBaseHovered=(&self:WidgetTree.Button_Leaderboard.HandleOnButtonHovered)
    onButtonBaseUnhovered=(&self:WidgetTree.Button_Leaderboard.HandleOnButtonUnhovered)
    onSelectedChangedBase=(&self:WidgetTree.Button_Leaderboard.HandleOnSelectedChanged)
    pressMethod=DownAndUp
    xL=false
    color {
     ButtonColor1=<rgba 0 0.9823 0.2362 0>
     ButtonColor2=<rgba 0.008 0.9823 0.2423 0>
     ButtonHover=<rgba 1 1 1 1>
     TextColor=<rgba 1 1 1 1>
     TextHover=<rgba 0 0 0 1>
     TextPressed=<rgba 0 0 0 1>
     TextPressedHover=<rgba 0 0 0 1>
    }
    triggeringInputAction {dataTable=&InputActionDataTable, rowName=GenericFaceButtonTop}
   }
   slot {class=BorderSlot, properties {padding=<ltrb 20 -5 20 0>}}
  }
  Overlay_PayoutButton {
   widgetClassPath=&Overlay
   bIsVariable=true
   parent=HorizontalBox_1
   widget=Overlay_PayoutButton
   slot {class=HorizontalBoxSlot, properties {horizontalAlignment=HAlign_Left}}
  }
  CommonBorder_PayoutButtonBG {
   widgetClassPath=&CommonBorder
   bIsVariable=true
   parent=Overlay_PayoutButton
   widget=CommonBorder_PayoutButtonBG
   properties {
    background {
     bIsDynamicallyLoaded=false
     drawAs=Box
     imageType=NoImage
     mirroring=NoMirror
     resourceName=None
     resourceObject=&M_UI_SkewBox_SubBarBG2
     tiling=NoTile
     imageSize=<171.3667 64>
     margin=<ltrb 0.1 0 0.1 0>
     outlineSettings {
      bUseBrushTransparency=false
      roundingType=HalfHeightRadius
      width=0
      color {colorUseRule=UseColor_Specified, specifiedColor=<rgba 0 0 0 0>}
      cornerRadii=<0 0 0 0>
     }
     tintColor {colorUseRule=UseColor_Specified, specifiedColor=<rgba 1 1 1 1>}
     uVRegion {bIsValid=false, max=<0 0>, min=<0 0>}
    }
    brushColor=<rgba 0 0.2086 1 1>
    padding=<ltrb 0 0 0 0>
   }
   slot {class=OverlaySlot, properties {horizontalAlignment=HAlign_Fill, verticalAlignment=VAlign_Fill}}
  }
  CommonBorderPayoutButtonHover {
   widgetClassPath=&CommonBorder
   parent=CommonBorder_PayoutButtonBG
   widget=CommonBorderPayoutButtonHover
   properties {
    background {
     bIsDynamicallyLoaded=false
     drawAs=Box
     imageType=NoImage
     mirroring=NoMirror
     resourceName=None
     resourceObject=&M_UI_SkewBox_SubBarBG2
     tiling=NoTile
     imageSize=<171.3667 64>
     margin=<ltrb 0.1 0 0.1 0>
     outlineSettings {
      bUseBrushTransparency=false
      roundingType=HalfHeightRadius
      width=0
      color {colorUseRule=UseColor_Specified, specifiedColor=<rgba 0 0 0 0>}
      cornerRadii=<0 0 0 0>
     }
     tintColor {colorUseRule=UseColor_Specified, specifiedColor=<rgba 1 1 1 1>}
     uVRegion {bIsValid=false, max=<0 0>, min=<0 0>}
    }
    brushColor=<rgba 1 1 1 0>
    padding=<ltrb 20 -5 20 0>
   }
   slot {class=BorderSlot, properties {padding=<ltrb 0 0 0 0>}}
  }
  Button_Payout {
   widgetClassPath=&EventsSimpleButton
   bIsVariable=true
   parent=CommonBorderPayoutButtonHover
   widget=Button_Payout
   properties {
    buttonNameText="Prizes && Rewards"
    clickMethod=DownAndUp
    onButtonBaseClicked=(&self:WidgetTree.Button_Payout.HandleOnButtonClicked)
    onButtonBaseDoubleClicked=(&self:WidgetTree.Button_Payout.HandleOnButtonDoubleClicked)
    onButtonBaseHovered=(&self:WidgetTree.Button_Payout.HandleOnButtonHovered)
    onButtonBaseUnhovered=(&self:WidgetTree.Button_Payout.HandleOnButtonUnhovered)
    onSelectedChangedBase=(&self:WidgetTree.Button_Payout.HandleOnSelectedChanged)
    pressMethod=DownAndUp
    xL=false
    color {
     ButtonColor1=<rgba 0 0.9823 0.2362 0>
     ButtonColor2=<rgba 0.008 0.9823 0.2423 0>
     ButtonHover=<rgba 1 1 1 1>
     TextColor=<rgba 1 1 1 1>
     TextHover=<rgba 0 0 0 1>
     TextPressed=<rgba 0 0 0 1>
     TextPressedHover=<rgba 0 0 0 1>
    }
    triggeringInputAction {dataTable=&InputActionDataTable, rowName=ShowEventPrizesAndRewards}
   }
   slot {class=BorderSlot, properties {padding=<ltrb 20 -5 20 0>}}
  }
  HorizontalBox_SquadInfoContainer {
   widgetClassPath=&HorizontalBox
   bIsVariable=true
   parent=VerticalBox_5
   widget=HorizontalBox_SquadInfoContainer
   slot {class=VerticalBoxSlot, properties {padding=<ltrb 30 15 0 0>}}
  }
  CommonTextBlock_Team {
   widgetClassPath=&CommonTextBlock
   parent=HorizontalBox_SquadInfoContainer
   widget=CommonTextBlock_Team
   properties {
    text=WITH:
    font {
     bForceMonospaced=false
     bMaterialIsStencil=false
     fontMaterial=null
     fontObject=&BurbankSmall
     letterSpacing=0
     monospacedWidth=1
     size=18
     skewAmount=0
     typefaceFontName=NotoSans-Regular
     outlineSettings {
      bApplyOutlineToDropShadows=false
      bMiteredCorners=false
      bSeparateFillAlpha=false
      outlineMaterial=null
      outlineSize=0
      outlineColor=<rgba 0 0 0 1>
     }
    }
   }
   slot {class=HorizontalBoxSlot}
  }
  CommonTextBlock_SquadInfo {
   widgetClassPath=&CommonTextBlock
   parent=HorizontalBox_SquadInfoContainer
   widget=CommonTextBlock_SquadInfo
   properties {
    text=TEAMMATES
    colorAndOpacity {colorUseRule=UseColor_Specified, specifiedColor=<rgba 0.0185 1 0.9047 1>}
    font {
     bForceMonospaced=false
     bMaterialIsStencil=false
     fontMaterial=null
     fontObject=&BurbankSmall
     letterSpacing=0
     monospacedWidth=1
     size=18
     skewAmount=0
     typefaceFontName=NotoSans-Regular
     outlineSettings {
      bApplyOutlineToDropShadows=false
      bMiteredCorners=false
      bSeparateFillAlpha=false
      outlineMaterial=null
      outlineSize=0
      outlineColor=<rgba 0 0 0 1>
     }
    }
   }
   slot {class=HorizontalBoxSlot, properties {padding=<ltrb 20 0 0 0>}}
  }
  CommonWidgetSwitcher_isFuture {
   widgetClassPath=&CommonWidgetSwitcherLegacy
   bIsVariable=true
   parent=VerticalBox_5
   widget=CommonWidgetSwitcher_isFuture
   properties {transitionCurveType=CubicIn, transitionDuration=0}
   slot {class=VerticalBoxSlot, properties {padding=<ltrb 56 13 0 0>, size {sizeRule=Fill, value=1}}}
  }
  VerticalBoxPlayerStats {
   widgetClassPath=&VerticalBox
   parent=CommonWidgetSwitcher_isFuture
   widget=VerticalBoxPlayerStats
   slot {class=WidgetSwitcherSlot}
  }
  SizeBox_1 {
   widgetClassPath=&SizeBox
   parent=VerticalBoxPlayerStats
   widget=SizeBox_1
   properties {bOverride_MaxDesiredWidth=true, maxDesiredWidth=1200}
   slot {class=VerticalBoxSlot, properties {horizontalAlignment=HAlign_Left}}
  }
  ScaleBox_0 {widgetClassPath=&ScaleBox, parent=SizeBox_1, widget=ScaleBox_0, slot {class=SizeBoxSlot}}
  HorizontalBox_3 {widgetClassPath=&HorizontalBox, parent=ScaleBox_0, widget=HorizontalBox_3, slot {class=ScaleBoxSlot}}
  VerticalBox_0 {
   widgetClassPath=&VerticalBox
   parent=HorizontalBox_3
   widget=VerticalBox_0
   slot {class=HorizontalBoxSlot, properties {padding=<ltrb 0 0 75 0>}}
  }
  VictoryRoyaleInfo {
   widgetClassPath=&/EventScoreRow
   bIsVariable=true
   parent=VerticalBox_0
   widget=VictoryRoyaleInfo
   slot {class=VerticalBoxSlot, properties {size {sizeRule=Fill, value=1}}}
  }
  EliminationsInfo {
   widgetClassPath=&EventScoreRow
   bIsVariable=true
   parent=VerticalBox_0
   widget=EliminationsInfo
   slot {class=VerticalBoxSlot, properties {size {sizeRule=Fill, value=1}}}
  }
  PlacementInfo {
   widgetClassPath=&EventScoreRow
   bIsVariable=true
   parent=VerticalBox_0
   widget=PlacementInfo
   properties {finalEntry=true}
   slot {class=VerticalBoxSlot, properties {size {sizeRule=Fill, value=1}}}
  }
  VerticalBox_1 {
   widgetClassPath=&VerticalBox
   parent=HorizontalBox_3
   widget=VerticalBox_1
   slot {class=HorizontalBoxSlot, properties {horizontalAlignment=HAlign_Right, verticalAlignment=VAlign_Center, size {sizeRule=Fill, value=1}}}
  }
  MatchesPlayedInfo {
   widgetClassPath=&/EventScoreBigStat
   bIsVariable=true
   parent=VerticalBox_1
   widget=MatchesPlayedInfo
   slot {class=VerticalBoxSlot, properties {horizontalAlignment=HAlign_Left, padding=<ltrb 0 -10 0 0>}}
  }
  PointsScoredInfo {
   widgetClassPath=&EventScoreBigStat
   bIsVariable=true
   parent=VerticalBox_1
   widget=PointsScoredInfo
   slot {class=VerticalBoxSlot, properties {horizontalAlignment=HAlign_Left, padding=<ltrb 0 -15 0 -10>}}
  }
  Text_BusFare {
   widgetClassPath=&CommonTextBlock
   bIsVariable=true
   parent=VerticalBox_1
   widget=Text_BusFare
   properties {
    text=({0} BUS FARE)
    textTransformPolicy=ToUpper
    font {
     bForceMonospaced=false
     bMaterialIsStencil=false
     fontMaterial=null
     fontObject=&Burbank
     letterSpacing=0
     monospacedWidth=1
     size=24
     skewAmount=0
     typefaceFontName=Burbank-PlayerPower
     outlineSettings {
      bApplyOutlineToDropShadows=false
      bMiteredCorners=false
      bSeparateFillAlpha=false
      outlineMaterial=null
      outlineSize=0
      outlineColor=<rgba 0 0 0 1>
     }
    }
    renderTransform {angle=0, scale=<1 1>, shear=<-12 0>, translation=<0 0>}
    shadowOffset=<0 0>
   }
   slot {class=VerticalBoxSlot, properties {horizontalAlignment=HAlign_Center, padding=<ltrb 0 0 0 -20>}}
  }
  CommonBorder_StartingScore {
   widgetClassPath=&CommonBorder
   bIsVariable=true
   parent=VerticalBox_1
   widget=CommonBorder_StartingScore
   properties {horizontalAlignment=HAlign_Left, verticalAlignment=VAlign_Center, contentColorAndOpacity=<rgba 0.0185 1 0.9047 1>, padding=<ltrb 0 0 0 0>}
   slot {class=VerticalBoxSlot, properties {padding=<ltrb 0 -10 0 -10>}}
  }
  Text_TimeRemaining {
   widgetClassPath=&CommonTextBlock
   bInherited=true
   bIsVariable=true
   parent=CommonWidgetSwitcher_isFuture
   widget=Text_TimeRemaining
   properties {
    text="This session starts in {0}. \r\nCome back then to test your mettle!"
    wrapTextAt=900
    font {
     bForceMonospaced=false
     bMaterialIsStencil=false
     fontMaterial=null
     fontObject=&Burbank
     letterSpacing=0
     monospacedWidth=1
     size=20
     skewAmount=0
     typefaceFontName=Burbank-PlayerPower
     outlineSettings {
      bApplyOutlineToDropShadows=false
      bMiteredCorners=false
      bSeparateFillAlpha=false
      outlineMaterial=null
      outlineSize=0
      outlineColor=<rgba 0 0 0 1>
     }
    }
   }
   slot {class=WidgetSwitcherSlot, properties {verticalAlignment=VAlign_Center}}
  }
  RichTextBlock_OtherReason {
   widgetClassPath=/Script/UMG.RichTextBlock
   bIsVariable=true
   parent=CommonWidgetSwitcher_isFuture
   widget=RichTextBlock_OtherReason
   properties {
    text="<WindowLocked>This event is</> <WindowLockedB>LOCKED</><WindowLocked>.</>"
    textStyleSet=&/TournamentStyles
    wrapTextAt=900
    decoratorClasses [/Game/Athena/HUD/EndGame/VictoryImageStyles.VictoryImageStyles_C]
   }
   slot {class=WidgetSwitcherSlot, properties {verticalAlignment=VAlign_Center}}
  }
  Text_DebugId {
   widgetClassPath=&TextBlock
   bInherited=true
   parent=Overlay_0
   widget=Text_DebugId
   properties {
    autoWrapText=true
    justification=Center
    text=EventId: Event EventWindowId: EventWindow
    wrapTextAt=300
    wrappingPolicy=AllowPerCharacterWrapping
    font {
     bForceMonospaced=false
     bMaterialIsStencil=false
     fontMaterial=null
     fontObject=&Roboto
     letterSpacing=0
     monospacedWidth=1
     size=20
     skewAmount=0
     typefaceFontName=Bold
     outlineSettings {
      bApplyOutlineToDropShadows=false
      bMiteredCorners=false
      bSeparateFillAlpha=false
      outlineMaterial=null
      outlineSize=2
      outlineColor=<rgba 0 0 0 1>
     }
    }
   }
   slot {class=OverlaySlot}
  }
  Image_PosterFront {widgetClassPath=/Script/EpicCMSUIFramework.EpicCMSImage, bInherited=true, widget=null}
  Image_PosterBack {widgetClassPath=&EpicCMSImage, bInherited=true, widget=null}
  Image_LoadingScreen {widgetClassPath=&EpicCMSImage, bInherited=true, widget=null}
  Image_PlaylistTile {widgetClassPath=&EpicCMSImage, bInherited=true, widget=null}
 ]
}
