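# Widget Blueprint for a tournament event leaderboard screen, extending FortEventLeaderboardScreen, with event details, leaderboard tabs and an entry details panel.
# Init feeds tournament and window IDs to the child views, using series-point leaderboard IDs when available and no window is given.
# Entry selection or list double-click fills the details panel.
# Tournament display colors tint buttons and tabs; Back or touch-only Close hides details and fires a close dispatcher.
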WidgetBlueprint /CompeteOldTab/UI/Events/Leaderboards/EventLeaderboardScreen
graphs [
 EventCloseButton {nodes [EventCloseButton() {is=title K2Node_FunctionEntry}]}
 EventGraph {nodes [
   "On Button Clicked (Button_ViewLiveGames)"!() -> (OutputDelegate, then, Button:Common Button Legacy Object Reference) {is=title K2Node_ComponentBoundEvent}
   > Call Event View Live Games!(execute:Exec, self:Event Leaderboard Screen Object Reference) -> (then:Exec) {is=title K2Node_CallDelegate}
   n1 = "On Button Clicked (Button_Back)"!() -> (OutputDelegate, then, Button:Common Button Legacy Object Reference) {is=title K2Node_ComponentBoundEvent}
   SetVisibility(
    execute=[@n1.then, @n2.then]
    self=Get EntryDetails() -> (EntryDetails:Event Leaderboard Entry Details Object Reference)
    InVisibility=Hidden
   )
   > Call Event Close Button!(execute:Exec, self:Event Leaderboard Screen Object Reference) -> (then:Exec) {is=title K2Node_CallDelegate}
   n2 = "On Button Clicked (CloseButton)"!() -> (OutputDelegate, then, Button:Common Button Legacy Object Reference) {is=title K2Node_ComponentBoundEvent}
   n3 = EventColorize!() -> (OutputDelegate, then, TournamentDisplayInfo:Fort Tournament Display Info Structure) {is=title K2Node_CustomEvent}
   > Set CurrentDisplayInfo(CurrentDisplayInfo:Fort Tournament Display Info Structure=@n3.TournamentDisplayInfo) -> (Output_Get:Fort Tournament Display Info Structure)
   > n4 = Sequence()
   > n5 = Cast To LeaderboardTab!(execute, Object=Get ActivatableContent_LeaderboardTab() -> (ActivatableContent_LeaderboardTab:Fort Leaderboard View Tab Object Reference)) -> (then, CastFailed, AsLeaderboard Tab:Leaderboard Tab Object Reference) {is=title K2Node_DynamicCast}
   > Colorize!(execute, self:Leaderboard Tab Object Reference=@n5, Color Info:Fort Tournament Display Info Structure=@n3.TournamentDisplayInfo) -> (then)
   > EventColorize!(
    execute
    self:Event Leaderboard Entry Details Object Reference=Get EntryDetails()
    TournamentDisplayInfo:Fort Tournament Display Info Structure=@n3.TournamentDisplayInfo
   ) -> (then) {is=title K2Node_CallFunction}
   n6 = Cast To EventsSimpleButton!(execute=@n4.then_1, Object=Get Button_ViewLiveGames() -> (Button_ViewLiveGames:Common Button Base Object Reference)) -> (then, CastFailed, AsEvents Simple Button:Events Simple Button Object Reference) {is=title K2Node_DynamicCast}
   n7 = Break Fort Tournament Display Info!(FortTournamentDisplayInfo:"Fort Tournament Display Info Structure (by ref)"=@n3.TournamentDisplayInfo) -> (
    TitleLine1:Text
    TitleLine2:Text
    ScheduleInfo:Text
    FlavorDescription:Text
    DetailsDescription:Text
    ShortFormatTitle:Text
    LongFormatTitle:Text
    BackgroundTitle:Text
    PinScoreRequirement:int
    PinEarnedText:Text
    BaseColor:Linear Color Structure
    PrimaryColor:Linear Color Structure
    SecondaryColor:Linear Color Structure
    HighlightColor:Linear Color Structure
    TitleColor:Linear Color Structure
    ShadowColor:Linear Color Structure
    BackgroundLeftColor:Linear Color Structure
    BackgroundRightColor:Linear Color Structure
    BackgroundTextColor:Linear Color Structure
    PosterFadeColor:Linear Color Structure
    AlertText:Text
    SeriesPointLeaderboardName:Text
    AlertType:EFortTournamentAlertType Enum
    RoundDisplayNames:Array of Texts
   ) {is=title K2Node_BreakStruct}
   > UpdateColor!(
    execute
    self:Events Simple Button Object Reference=@n6
    ParameterName:String=ButtonColor1
    New Color:Linear Color Structure=@n7.SecondaryColor
   ) -> (then)
   > UpdateColor(self=@n6, ParameterName=ButtonColor2, New Color=@n7.SecondaryColor)
   EventIntro() {is=title K2Node_CustomEvent}
   > EventIntro(self:Event Leaderboard Entry Details Object Reference=Get EntryDetails()) {is=title K2Node_CallFunction}
   Event Leaderboard Screen On Activated() {is=title K2Node_CustomEvent}
   > HandleLeaderboardSetInView(self:Fort Event Leaderboard Screen Object Reference)
   > Set ProfileSelectorInstance!(
    execute
    ProfileSelectorInstance:Event Player Stat Profile Selector Object Reference
    self:Event Leaderboard Entry Details Object Reference=Get EntryDetails()
   ) -> (then, Output_Get:Event Player Stat Profile Selector Object Reference)
   n8 = Item Double-Clicked!() -> (OutputDelegate, then, Item:Object Reference) {is=title K2Node_CustomEvent}
   > n9 = Cast To FortEventLeaderboardEntryData!(execute, Object=@n8.Item) -> (then, CastFailed, AsFort Event Leaderboard Entry Data:Fort Event Leaderboard Entry Data Object Reference) {is=title K2Node_DynamicCast}
   n10 = Get EntryDetails()
   > n11 = Show Details!(
    execute
    self:Event Leaderboard Entry Details Object Reference=@n10
    EntryData:Fort Event Leaderboard Entry Data Object Reference=@n9
    EventType:ETournamentDisplayType Enum=Default
   ) -> (then)
   LeaderboardEntryDoubleClicked(execute=[@n11.then, @n9.CastFailed], self:Event Leaderboard Entry Details Object Reference=@n10)
   n12 = Event OnLeaderboardEntrySelected!() -> (OutputDelegate, then, EntryData:Fort Event Leaderboard Entry Data Object Reference) {is=title K2Node_Event}
   > Show Details(
    self=Get EntryDetails()
    EntryData=@n12.EntryData
    EventType=Get EventDisplayType() -> (EventDisplayType:ETournamentDisplayType Enum=Default)
   )
   n13 = Event OnAlternativeLeaderboardsTabVisibilityChange!() -> (OutputDelegate, then, bIsVisible:bool) {is=title K2Node_Event}
   > SetVisibility(
    self=Get SizeBox_AlternativeLeaderboardsParent() -> (SizeBox_AlternativeLeaderboardsParent:Horizontal Box Object Reference)
    InVisibility=Select(Option 0=Hidden, Option 1=SelfHitTestInvisible, Index=@n13.bIsVisible)
   )
   Event Construct()
   > Cast To LeaderboardListViewWrapper!(
    execute
    Object=Get ActivatableContent_LeaderboardListWrapper!(self:Fort Leaderboard View Tab Object Reference=Get ActivatableContent_LeaderboardTab()) -> (ActivatableContent_LeaderboardListWrapper:Fort Leaderboard List View Wrapper Object Reference)
   ) -> (then, CastFailed, AsLeaderboard List View Wrapper:Leaderboard List View Wrapper Object Reference) {is=title K2Node_DynamicCast}
   > Bind Event to On Item Double Clicked(
    self:List View Object Reference=GetListView!(self:Fort Leaderboard View Tab Object Reference=Get ActivatableContent_LeaderboardTab()) -> (ReturnValue:List View Object Reference)
    Delegate=@n8
   ) {is=title K2Node_AddDelegate}
   Event OnLeaderboardDisplayTypeChanged!() -> (OutputDelegate, then, NewDisplayType:ELeaderboardDisplayType Enum=Default) {is=title K2Node_Event}
   > Set Leaderboard Key!(
    execute
    self:Event Leaderboard Entry Details Object Reference=Get EntryDetails()
    LeaderboardInstanceId:String=Get RepresentedEventWindowId() -> (RepresentedEventWindowId:String)
    LeaderboardGroupingId:String=Get RepresentedEventId() -> (RepresentedEventId:String)
   ) -> (then) {is=title K2Node_CallFunction}
   n14 = Event OnDetailsPanelUpdate!() -> (OutputDelegate, then, bShowDetailsPanel:bool) {is=title K2Node_Event}
   > SetVisibility(
    self=Get Overlay_EntryDetailsVisibility() -> (Overlay_EntryDetailsVisibility:Overlay Object Reference)
    InVisibility=Select(Option 0=Hidden, Option 1=Visible, Index=@n14.bShowDetailsPanel)
   )
   n15 = Event InitializeAlternativeLeaderboardsTabButton!() -> (OutputDelegate, then, TabButton:Common Button Base Object Reference, TabText:"Text (by ref)") {is=title K2Node_Event}
   > n16 = Cast To EventsSimpleTab!(execute, Object=@n15.TabButton) -> (then, CastFailed, AsEvents Simple Tab:Events Simple Tab Object Reference) {is=title K2Node_DynamicCast}
   > Set ButtonNameText!(execute, ButtonNameText:Text=@n15.TabText, self:Events Simple Tab Object Reference=@n16) -> (then, Output_Get:Text)
   n17 = Get!() -> (
    CurrentDisplayInfo_TitleLine1:Text
    CurrentDisplayInfo_TitleLine2:Text
    CurrentDisplayInfo_ScheduleInfo:Text
    CurrentDisplayInfo_FlavorDescription:Text
    CurrentDisplayInfo_DetailsDescription:Text
    CurrentDisplayInfo_ShortFormatTitle:Text
    CurrentDisplayInfo_LongFormatTitle:Text
    CurrentDisplayInfo_BackgroundTitle:Text
    CurrentDisplayInfo_PinScoreRequirement:int
    CurrentDisplayInfo_PinEarnedText:Text
    CurrentDisplayInfo_BaseColor:Linear Color Structure
    CurrentDisplayInfo_PrimaryColor:Linear Color Structure
    CurrentDisplayInfo_SecondaryColor:Linear Color Structure
    CurrentDisplayInfo_HighlightColor:Linear Color Structure
    CurrentDisplayInfo_TitleColor:Linear Color Structure
    CurrentDisplayInfo_ShadowColor:Linear Color Structure
    CurrentDisplayInfo_BackgroundLeftColor:Linear Color Structure
    CurrentDisplayInfo_BackgroundRightColor:Linear Color Structure
    CurrentDisplayInfo_BackgroundTextColor:Linear Color Structure
    CurrentDisplayInfo_PosterFadeColor:Linear Color Structure
    CurrentDisplayInfo_AlertText:Text
    CurrentDisplayInfo_SeriesPointLeaderboardName:Text
    CurrentDisplayInfo_AlertType:EFortTournamentAlertType Enum
    CurrentDisplayInfo_RoundDisplayNames:Array of Texts
   )
   > UpdateColor(self:Events Simple Tab Object Reference=@n16, ParameterName=ButtonColor1, New Color=@n17.CurrentDisplayInfo_PrimaryColor)
   > UpdateColor(self=@n16, ParameterName=ButtonColor2, New Color=@n17.CurrentDisplayInfo_PrimaryColor)
   > UpdateColor(self=@n16, ParameterName=TextColor, New Color=@n17.CurrentDisplayInfo_ShadowColor)
   n18 = Event InitializeFriendsToggleTabButton!() -> (OutputDelegate, then, TabButton:Common Button Base Object Reference, TabText:"Text (by ref)") {is=title K2Node_Event}
   > n19 = Cast To EventsSimpleTab(Object=@n18.TabButton)
   > Set ButtonNameText(ButtonNameText=@n18.TabText, self=@n19)
   n20 = Get()
   > UpdateColor(self=@n19, ParameterName=ButtonColor1, New Color=@n20.CurrentDisplayInfo_PrimaryColor)
   > UpdateColor(self=@n19, ParameterName=ButtonColor2, New Color=@n20.CurrentDisplayInfo_PrimaryColor)
   > UpdateColor(self=@n19, ParameterName=TextColor, New Color=@n20.CurrentDisplayInfo_ShadowColor)
  ]}
 EventViewLiveGames {nodes [EventViewLiveGames() {is=title K2Node_FunctionEntry}]}
 Init {nodes [
   Init!() -> (then, EventWindowId:String, TournamentSeriesId:String) {is=title K2Node_FunctionEntry}
   n1 = Get TournamentSeriesId() -> (TournamentSeriesId:String)
   n2 = Get EventWindowId() -> (EventWindowId:String)
   > SetTournament!(
    execute
    self:Fort Showdown Detail View Object Reference=Get EventLeaderboardEventDetails() -> (EventLeaderboardEventDetails:Event Leaderboard Event Details Object Reference)
    TournamentId:String=@n1
    EventId:String=@n2
    bForceRefreshUI:bool=true
   ) -> (then)
   > SetTournamentValues!(execute, self:Event Leaderboard Entry Details Object Reference=Get EntryDetails(), InTournamentId:String=@n1, InEventWindowId:String=@n2) -> (then)
   > EventIntro(self:Self Object Reference)
   > n3 = Cast To LeaderboardTab(Object=Get ActivatableContent_LeaderboardTab())
   > Set Event!(execute, self:Leaderboard Tab Object Reference=@n3, EventId:String=Get TournamentSeriesId(), EventWindowId:String=Get EventWindowId()) -> (then) {is=title K2Node_CallFunction}
   n4 = Get EventLeaderboardEventDetails()
   n5 = Get EventWindowId()
   > n6 = Branch(Condition=AND Boolean(A=IsSeriesPointLeaderboardAvailable!(self:Fort Showdown Detail View Object Reference=@n4) -> (ReturnValue:bool), B=IsEmpty(InString=@n5)))
   n7 = GetSeriesPointLeaderboardIds!(self:Fort Showdown Detail View Object Reference=Get EventLeaderboardEventDetails()) -> (OutEventId:String, OutEventWindowId:String)
   > n8 = SetEvent!(execute, self:Fort Event Leaderboard Screen Object Reference, EventId:String=@n7, EventWindowId:String=@n7.OutEventWindowId) -> (then)
   n9 = Get EntryDetails()
   Branch(
    execute=[@n8.then, @n10.then]
    Condition="Equal (Enum)"(
     A:ESlateVisibility Enum=GetVisibility!(self:Widget Object Reference=@n9) -> (ReturnValue:ESlateVisibility Enum=Visible)
     B:ESlateVisibility Enum=Hidden
    )
   )
   > SetVisibility(self=Get EntryDetails(), InVisibility=Visible)
   n10 = SetEvent(execute=@n6.else, EventId=Get TournamentSeriesId(), EventWindowId=Get EventWindowId())
  ]}
]
variables [
 CurrentDisplayInfo {
  type=FortTournamentDisplayInfo
  default {
   alertText=""
   alertType=Warning
   backgroundTitle=""
   detailsDescription=""
   flavorDescription=""
   longFormatTitle=""
   pinEarnedText=""
   pinScoreRequirement=0
   scheduleInfo=""
   seriesPointLeaderboardName=""
   shortFormatTitle=""
   titleLine1=""
   titleLine2=""
   backgroundLeftColor=<rgba 0 0 0 0>
   backgroundRightColor=<rgba 0 0 0 0>
   backgroundTextColor=<rgba 0 0 0 0>
   baseColor=<rgba 0 0 0 0>
   highlightColor=<rgba 0 0 0 0>
   posterFadeColor=<rgba 0 0 0 0>
   primaryColor=<rgba 0 0 0 0>
   roundDisplayNames []
   secondaryColor=<rgba 0 0 0 0>
   shadowColor=<rgba 0 0 0 0>
   titleColor=<rgba 0 0 0 0>
  }
 }
 DebugLeaderboardEntries {type="array<FortEventLeaderboardEntryData>", default []}
 DebugLeaderboardEntry1 {type=FortEventLeaderboardEntryData, default=null}
 DebugLeaderboardEntry2 {type=FortEventLeaderboardEntryData, default=null}
 DebugLeaderboardEntry3 {type=FortEventLeaderboardEntryData, default=null}
 DebugLeaderboardEntry4 {type=FortEventLeaderboardEntryData, default=null}
 EventCloseButton {}
 EventViewLiveGames {}
]
widgets {
 info {inheritedWidgetCount=10, parentClass=/Script/FortniteUI.FortEventLeaderboardScreen, rootWidgetClass=&SafeZone, widgetCount=26}
 widgets [
  SafeZone_0 {widgetClassPath=&SafeZone, bIsVariable=true, widget=SafeZone_0}
  Overlay_1 {widgetClassPath=&Overlay, parent=SafeZone_0, widget=Overlay_1, slot {class=SafeZoneSlot}}
  HorizontalBox_0 {
   widgetClassPath=&HorizontalBox
   parent=Overlay_1
   widget=HorizontalBox_0
   slot {class=OverlaySlot, properties {horizontalAlignment=HAlign_Fill, verticalAlignment=VAlign_Fill}}
  }
  VerticalBox_0 {
   widgetClassPath=&VerticalBox
   parent=HorizontalBox_0
   widget=VerticalBox_0
   slot {class=HorizontalBoxSlot, properties {padding=<ltrb 0 0 50 0>, size {sizeRule=Fill, value=1}}}
  }
  EventLeaderboardEventDetails {
   widgetClassPath=/Game/EventLeaderboardEventDetails
   bIsVariable=true
   parent=VerticalBox_0
   widget=EventLeaderboardEventDetails
   slot {class=VerticalBoxSlot, properties {padding=<ltrb 40 35 -200 10>}}
  }
  Button_ViewLiveGames {
   widgetClassPath=/CompeteOldTab/Frontend/Showdown/EventsSimpleButton
   bInherited=true
   bIsVariable=true
   parent=VerticalBox_0
   widget=Button_ViewLiveGames
   properties {
    buttonNameText=Watch Live Matches
    onButtonBaseClicked=(&self:WidgetTree.Button_ViewLiveGames.HandleOnButtonClicked)
    onButtonBaseDoubleClicked=(&self:WidgetTree.Button_ViewLiveGames.HandleOnButtonDoubleClicked)
    onButtonBaseHovered=(&self:WidgetTree.Button_ViewLiveGames.HandleOnButtonHovered)
    onButtonBaseUnhovered=(&self:WidgetTree.Button_ViewLiveGames.HandleOnButtonUnhovered)
    onSelectedChangedBase=(&self:WidgetTree.Button_ViewLiveGames.HandleOnSelectedChanged)
    xL=false
    triggeringInputAction {dataTable=&InputActionDataTable, rowName=GenericFaceButtonTop}
   }
   slot {class=VerticalBoxSlot, properties {horizontalAlignment=HAlign_Left, verticalAlignment=VAlign_Bottom, padding=<ltrb 40 5 20 5>}}
  }
  Widget_AlternativeLeaderboardsTabsParent {
   widgetClassPath=&SizeBox
   bInherited=true
   parent=VerticalBox_0
   widget=Widget_AlternativeLeaderboardsTabsParent
   properties {bOverride_MaxDesiredWidth=true, bOverride_MinDesiredHeight=true, heightOverride=45.5, maxDesiredWidth=1000, minDesiredHeight=46}
   slot {class=VerticalBoxSlot, properties {horizontalAlignment=HAlign_Center, verticalAlignment=VAlign_Center, padding=<ltrb 0 5 0 5>}}
  }
  SizeBox_AlternativeLeaderboardsParent {
   widgetClassPath=&HorizontalBox
   bIsVariable=true
   parent=Widget_AlternativeLeaderboardsTabsParent
   widget=SizeBox_AlternativeLeaderboardsParent
   slot {class=SizeBoxSlot}
  }
  ActionWidget_AlternativeLeaderboards_TabLeft {
   widgetClassPath=&CommonActionWidget
   bInherited=true
   parent=SizeBox_AlternativeLeaderboardsParent
   widget=ActionWidget_AlternativeLeaderboards_TabLeft
   properties {inputActions [{dataTable=&InputActionDataTable, rowName=SecondaryTabLeft}]}
   slot {class=HorizontalBoxSlot, properties {padding=<ltrb 0 0 8 0>}}
  }
  SizeBox_AlternativeLeaderboardsCenter {
   widgetClassPath=&SizeBox
   parent=SizeBox_AlternativeLeaderboardsParent
   widget=SizeBox_AlternativeLeaderboardsCenter
   properties {bOverride_MaxDesiredWidth=true, maxDesiredWidth=900}
   slot {class=HorizontalBoxSlot}
  }
  ScrollBox_AlternativeLeaderboardsTabs {
   widgetClassPath=&ScrollBox
   bInherited=true
   parent=SizeBox_AlternativeLeaderboardsCenter
   widget=ScrollBox_AlternativeLeaderboardsTabs
   properties {orientation=Orient_Horizontal}
   slot {class=SizeBoxSlot}
  }
  ActionWidget_AlternativeLeaderboards_TabRight {
   widgetClassPath=&CommonActionWidget
   bInherited=true
   parent=SizeBox_AlternativeLeaderboardsParent
   widget=ActionWidget_AlternativeLeaderboards_TabRight
   properties {inputActions [{dataTable=&InputActionDataTable, rowName=SecondaryTabRight}]}
   slot {class=HorizontalBoxSlot, properties {padding=<ltrb 8 0 0 0>}}
  }
  Widget_FriendsToggleTabsParent {
   widgetClassPath=&SizeBox
   bInherited=true
   parent=VerticalBox_0
   widget=Widget_FriendsToggleTabsParent
   properties {bOverride_MaxDesiredWidth=true, bOverride_MinDesiredHeight=true, heightOverride=45.5, maxDesiredWidth=1000, minDesiredHeight=46}
   slot {class=VerticalBoxSlot, properties {horizontalAlignment=HAlign_Center, verticalAlignment=VAlign_Center, padding=<ltrb 0 5 0 5>}}
  }
  HorizontalBox_FriendsToggle {
   widgetClassPath=&HorizontalBox
   parent=Widget_FriendsToggleTabsParent
   widget=HorizontalBox_FriendsToggle
   slot {class=SizeBoxSlot}
  }
  ActionWidget_FriendsToggle_TabLeft {
   widgetClassPath=&CommonActionWidget
   bInherited=true
   parent=HorizontalBox_FriendsToggle
   widget=ActionWidget_FriendsToggle_TabLeft
   properties {inputActions [{dataTable=&InputActionDataTable, rowName=MainTabLeft}]}
   slot {class=HorizontalBoxSlot, properties {padding=<ltrb 0 0 8 0>}}
  }
  SizeBox_FriendsToggleCenter {
   widgetClassPath=&SizeBox
   parent=HorizontalBox_FriendsToggle
   widget=SizeBox_FriendsToggleCenter
   properties {bOverride_MaxDesiredWidth=true, maxDesiredWidth=900}
   slot {class=HorizontalBoxSlot}
  }
  ScrollBox_FriendsToggleTabs {
   widgetClassPath=&ScrollBox
   bInherited=true
   parent=SizeBox_FriendsToggleCenter
   widget=ScrollBox_FriendsToggleTabs
   properties {orientation=Orient_Horizontal}
   slot {class=SizeBoxSlot}
  }
  ActionWidget_FriendsToggle_TabRight {
   widgetClassPath=&CommonActionWidget
   bInherited=true
   parent=HorizontalBox_FriendsToggle
   widget=ActionWidget_FriendsToggle_TabRight
   properties {inputActions [{dataTable=&InputActionDataTable, rowName=MainTabRight}]}
   slot {class=HorizontalBoxSlot, properties {padding=<ltrb 8 0 0 0>}}
  }
  ActivatableContent_LeaderboardTab {
   widgetClassPath=/Game/UI/Events/Leaderboards/LeaderboardTab
   bInherited=true
   bIsVariable=true
   parent=VerticalBox_0
   widget=ActivatableContent_LeaderboardTab
   properties {bAutoActivate=true}
   slot {class=VerticalBoxSlot, properties {padding=<ltrb 0 0 0 25>, size {sizeRule=Fill, value=1}}}
  }
  Overlay_2 {
   widgetClassPath=&Overlay
   parent=HorizontalBox_0
   widget=Overlay_2
   slot {class=HorizontalBoxSlot, properties {horizontalAlignment=HAlign_Right, size {sizeRule=Fill, value=1}}}
  }
  Overlay_EntryDetailsVisibility {
   widgetClassPath=&Overlay
   bIsVariable=true
   parent=Overlay_2
   widget=Overlay_EntryDetailsVisibility
   slot {class=OverlaySlot, properties {horizontalAlignment=HAlign_Right, verticalAlignment=VAlign_Fill}}
  }
  EntryDetails {
   widgetClassPath=&/EventLeaderboardEntryDetails
   bIsVariable=true
   parent=Overlay_EntryDetailsVisibility
   widget=EntryDetails
   properties {visibility=Hidden}
   slot {class=OverlaySlot, properties {verticalAlignment=VAlign_Fill}}
  }
  VisibilityBorder_NonTouch {
   widgetClassPath=&CommonHardwareVisibilityBorder
   parent=Overlay_2
   widget=VisibilityBorder_NonTouch
   properties {visibilityQuery {
     autoDescription=" NONE( Input.Touch )"
     tokenStreamVersion=0
     userDescription=""
     queryTokenStream [0,1,3,1,0]
     tagDictionary [{tagName=Input.Touch}]
    }}
   slot {class=OverlaySlot, properties {horizontalAlignment=HAlign_Right, verticalAlignment=VAlign_Bottom, padding=<ltrb 20 20 35 35>}}
  }
  Button_Back {
   widgetClassPath=&EventsSimpleButton
   bIsVariable=true
   parent=VisibilityBorder_NonTouch
   widget=Button_Back
   properties {
    buttonNameText=Back
    onButtonBaseClicked=(&self:WidgetTree.Button_Back.HandleOnButtonClicked)
    onButtonBaseDoubleClicked=(&self:WidgetTree.Button_Back.HandleOnButtonDoubleClicked)
    onButtonBaseHovered=(&self:WidgetTree.Button_Back.HandleOnButtonHovered)
    onButtonBaseUnhovered=(&self:WidgetTree.Button_Back.HandleOnButtonUnhovered)
    onSelectedChangedBase=(&self:WidgetTree.Button_Back.HandleOnSelectedChanged)
    triggeringInputAction {dataTable=&InputActionDataTable, rowName=GenericBack}
   }
   slot {class=BorderSlot, properties {padding=<ltrb 0 0 0 0>}}
  }
  VisibilityBorder_Touch {
   widgetClassPath=&CommonHardwareVisibilityBorder
   parent=Overlay_2
   widget=VisibilityBorder_Touch
   properties {visibilityQuery {
     autoDescription=" ANY( Input.Touch )"
     tokenStreamVersion=0
     userDescription=""
     queryTokenStream [0,1,1,1,0]
     tagDictionary [{tagName=Input.Touch}]
    }}
   slot {class=OverlaySlot, properties {horizontalAlignment=HAlign_Right, padding=<ltrb 0 25 50 20>}}
  }
  CloseButton {
   widgetClassPath=/FortUILibrary/Foundation/Buttons/CloseButton
   bIsVariable=true
   parent=VisibilityBorder_Touch
   widget=CloseButton
   properties {
    onButtonBaseClicked=(&self:WidgetTree.CloseButton.HandleOnButtonClicked)
    onButtonBaseDoubleClicked=(&self:WidgetTree.CloseButton.HandleOnButtonDoubleClicked)
    onButtonBaseHovered=(&self:WidgetTree.CloseButton.HandleOnButtonHovered)
    onButtonBaseUnhovered=(&self:WidgetTree.CloseButton.HandleOnButtonUnhovered)
    onSelectedChangedBase=(&self:WidgetTree.CloseButton.HandleOnSelectedChanged)
   }
   slot {class=BorderSlot, properties {padding=<ltrb 0 0 0 0>}}
  }
 ]
}
