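# Event leaderboard summary widget, a FortShowdownDetailView child showing a "Top Scoring Players" panel with two columns of EventLeaderboardEntry rows, the last flagged as the player.
# View-all button animates its text on hover and, if a leaderboard is available, broadcasts EventViewFullLeaderboard with tournament and event IDs.
# EventColorize tints backgrounds and title from tournament display info.
# Refresh Data always hides the content; the availability-based visibility Select is unconnected.
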WidgetBlueprint /CompeteOldTab/UI/Events/Leaderboards/EventLeaderboardSummary
animations [ViewAllHover {
  binding_count=1
  length_seconds=0.2383
  object=&self:ViewAllHover
  sequence {bindings [Text_ViewFullLeaderboard {
     id=#1
     possessed_class=CommonTextBlock
     tracks [{
       type=MovieSceneColorTrack
       display_name=Color and Opacity
       sections [{
         type=MovieSceneColorSection
         end_frame=4
         start_frame=0
         channels [
          Color.R {type=MovieSceneScriptingFloatChannel, keys [[0,1],[4.7647,0]]}
          Color.G {type=MovieSceneScriptingFloatChannel, keys [[0,1],[4.7647,0]]}
          Color.B {type=MovieSceneScriptingFloatChannel, keys [[0,1],[4.7647,0]]}
          Color.A {type=MovieSceneScriptingFloatChannel, const [1,2]}
         ]
        }]
      }]
    }]}
 }]
graphs [
 EventGraph {nodes [
   "On Button Clicked (Button_ViewAll)"!() -> (OutputDelegate, then, Button:Common Button Legacy Object Reference) {is=title K2Node_ComponentBoundEvent}
   > Branch(Condition=IsEventLeaderboardAvailable!(self:Fort Showdown Detail View Object Reference) -> (ReturnValue:bool))
   > Call Event View Full Leaderboard!(
    execute:Exec
    self:Event Leaderboard Summary Object Reference
    TournamentId:String=GetTournamentId!(self:Fort Showdown Detail View Object Reference) -> (ReturnValue:String)
    EventId:String=GetEventId!(self:Fort Showdown Detail View Object Reference) -> (ReturnValue:String)
   ) -> (then:Exec) {is=title K2Node_CallDelegate}
   "On Button Hovered (Button_ViewAll)"!() -> (OutputDelegate, then, Button:Common Button Legacy Object Reference) {is=title K2Node_ComponentBoundEvent}
   n1 = Get ViewAllHover() -> (ViewAllHover:Widget Animation Object Reference)
   > Anim Transition!(Exec:Exec, In:bool=true, Animation:Widget Animation Object Reference=@n1, PlaySpeed:double=1, Widget:User Widget Object Reference) -> (ExecOut:Exec) {is=title K2Node_MacroInstance}
   "On Button Unhovered (Button_ViewAll)"!() -> (OutputDelegate, then, Button:Common Button Legacy Object Reference) {is=title K2Node_ComponentBoundEvent}
   > Anim Transition(Animation=@n1, PlaySpeed=1)
   n2 = EventColorize!() -> (OutputDelegate, then, TournamentDisplayInfo:Fort Tournament Display Info Structure) {is=title K2Node_CustomEvent}
   n3 = Break Fort Tournament Display Info!(FortTournamentDisplayInfo:"Fort Tournament Display Info Structure (by ref)"=@n2.TournamentDisplayInfo) -> (
    TitleLine1:Text
    TitleLine2:Text
    ScheduleInfo:Text
    FlavorDescription:Text
    DetailsDescription:Text
    ShortFormatTitle:Text
    LongFormatTitle:Text
    BackgroundTitle:Text
    PinScoreRequirement:int
    PinEarnedText:Text
    BaseColor:Linear Color Structure
    PrimaryColor:Linear Color Structure
    SecondaryColor:Linear Color Structure
    HighlightColor:Linear Color Structure
    TitleColor:Linear Color Structure
    ShadowColor:Linear Color Structure
    BackgroundLeftColor:Linear Color Structure
    BackgroundRightColor:Linear Color Structure
    BackgroundTextColor:Linear Color Structure
    PosterFadeColor:Linear Color Structure
    AlertText:Text
    SeriesPointLeaderboardName:Text
    AlertType:EFortTournamentAlertType Enum
    RoundDisplayNames:Array of Texts
   ) {is=title K2Node_BreakStruct}
   > SetBrushColor(self=Get CommonBorder_ScoringTitleBG() -> (CommonBorder_ScoringTitleBG:Common Border Object Reference), InBrushColor=@n3.PrimaryColor)
   > SetColorAndOpacity(
    self=Get CommonTextBlock_Title() -> (CommonTextBlock_Title:Common Text Object Reference)
    InColorAndOpacity=Make Slate Color(SpecifiedColor=@n3.ShadowColor)
   )
   n4 = Break Linear Color!(LinearColor:"Linear Color Structure (by ref)"=@n3.BackgroundLeftColor) -> (R:float=~, G:float=~, B:float=~, A:float=~) {is=title K2Node_BreakStruct}
   n5 = Make Linear Color!(R:float=@n4, G:float=@n4.G, B:float=@n4.B, A:float=0.1) -> (LinearColor:Linear Color Structure) {is=title K2Node_MakeStruct}
   > SetBrushColor(self=Get CommonBorder_ContainerBG() -> (CommonBorder_ContainerBG:Common Border Object Reference), InBrushColor=@n5)
   > SetBrushColor(self=Get CommonBorder_ViewAllBG() -> (CommonBorder_ViewAllBG:Common Border Object Reference), InBrushColor=@n5)
   Event Refresh Data() {is=title K2Node_Event}
   > SetVisibility(self=Get Overlay_SummaryContent() -> (Overlay_SummaryContent:Overlay Object Reference), InVisibility=Hidden)
   Event PreConstruct()
   Event Construct()
   Event Tick()
   Select(Option 0=Hidden, Option 1=SelfHitTestInvisible, Index=IsEventLeaderboardAvailable())
  ]}
 EventViewFullLeaderboard {nodes [EventViewFullLeaderboard!() -> (then, TournamentId:String, EventId:String) {is=title K2Node_FunctionEntry}]}
]
variables [EventViewFullLeaderboard {}]
widgets {
 info {
  inheritedWidgetCount=6
  namedSlotCount=1
  parentClass=/Script/FortniteUI.FortShowdownDetailView
  rootWidgetClass=&SizeBox
  widgetCount=34
 }
 widgets [
  SizeBox_0 {
   widgetClassPath=&SizeBox
   widget=SizeBox_0
   properties {bOverride_MaxDesiredHeight=true, bOverride_WidthOverride=true, maxDesiredHeight=460, widthOverride=1100}
  }
  Overlay_SummaryContent {widgetClassPath=&Overlay, bIsVariable=true, parent=SizeBox_0, widget=Overlay_SummaryContent, slot {class=SizeBoxSlot}}
  BackgroundBlur_0 {
   widgetClassPath=&BackgroundBlur
   parent=Overlay_SummaryContent
   widget=BackgroundBlur_0
   properties {blurStrength=50}
   slot {class=OverlaySlot, properties {horizontalAlignment=HAlign_Fill, padding=<ltrb 15 40 0 0>}}
  }
  CommonBorder_ContainerBG {
   widgetClassPath=&CommonBorder
   bIsVariable=true
   parent=BackgroundBlur_0
   widget=CommonBorder_ContainerBG
   properties {
    background {
     bIsDynamicallyLoaded=false
     drawAs=Image
     imageType=NoImage
     mirroring=NoMirror
     resourceName=None
     resourceObject=null
     tiling=NoTile
     imageSize=<32 32>
     margin=<ltrb 0 0 0 0>
     outlineSettings {
      bUseBrushTransparency=false
      roundingType=HalfHeightRadius
      width=0
      color {colorUseRule=UseColor_Specified, specifiedColor=<rgba 0 0 0 0>}
      cornerRadii=<0 0 0 0>
     }
     tintColor {colorUseRule=UseColor_Specified, specifiedColor=<rgba 1 1 1 1>}
     uVRegion {bIsValid=false, max=<0 0>, min=<0 0>}
    }
    brushColor=<rgba 0 0.0762 0.4793 0.5>
    padding=<ltrb 0 0 0 0>
   }
   slot {class=BackgroundBlurSlot, properties {padding=<ltrb 0 0 0 0>}}
  }
  VerticalBox_1 {
   widgetClassPath=&VerticalBox
   parent=CommonBorder_ContainerBG
   widget=VerticalBox_1
   slot {class=BorderSlot, properties {padding=<ltrb 0 0 0 0>}}
  }
  ScaleBox_0 {
   widgetClassPath=&ScaleBox
   parent=VerticalBox_1
   widget=ScaleBox_0
   properties {stretch=UserSpecified, userSpecifiedScale=0.8}
   slot {class=VerticalBoxSlot, properties {padding=<ltrb 12 32 12 8>}}
  }
  HorizontalBox_Entries {
   widgetClassPath=&HorizontalBox
   parent=ScaleBox_0
   widget=HorizontalBox_Entries
   slot {class=ScaleBoxSlot, properties {horizontalAlignment=HAlign_Fill}}
  }
  VerticalBox_0 {
   widgetClassPath=&VerticalBox
   parent=HorizontalBox_Entries
   widget=VerticalBox_0
   slot {class=HorizontalBoxSlot, properties {size {sizeRule=Fill, value=1}}}
  }
  EventLeaderboardEntry {
   widgetClassPath=/Game/UI/Events/Leaderboards/EventLeaderboardEntry
   bIsVariable=true
   parent=VerticalBox_0
   widget=EventLeaderboardEntry
   properties {
    onButtonBaseClicked=(&self:WidgetTree.EventLeaderboardEntry.HandleOnButtonClicked)
    onButtonBaseDoubleClicked=(&self:WidgetTree.EventLeaderboardEntry.HandleOnButtonDoubleClicked)
    onButtonBaseHovered=(&self:WidgetTree.EventLeaderboardEntry.HandleOnButtonHovered)
    onButtonBaseUnhovered=(&self:WidgetTree.EventLeaderboardEntry.HandleOnButtonUnhovered)
    onSelectedChangedBase=(&self:WidgetTree.EventLeaderboardEntry.HandleOnSelectedChanged)
    visibility=HitTestInvisible
   }
   slot {class=VerticalBoxSlot}
  }
  EventLeaderboardEntry_0 {
   widgetClassPath=&EventLeaderboardEntry
   bIsVariable=true
   parent=VerticalBox_0
   widget=EventLeaderboardEntry_0
   properties {
    onButtonBaseClicked=(&self:WidgetTree.EventLeaderboardEntry_0.HandleOnButtonClicked)
    onButtonBaseDoubleClicked=(&self:WidgetTree.EventLeaderboardEntry_0.HandleOnButtonDoubleClicked)
    onButtonBaseHovered=(&self:WidgetTree.EventLeaderboardEntry_0.HandleOnButtonHovered)
    onButtonBaseUnhovered=(&self:WidgetTree.EventLeaderboardEntry_0.HandleOnButtonUnhovered)
    onSelectedChangedBase=(&self:WidgetTree.EventLeaderboardEntry_0.HandleOnSelectedChanged)
    visibility=HitTestInvisible
   }
   slot {class=VerticalBoxSlot}
  }
  EventLeaderboardEntry_1 {
   widgetClassPath=&EventLeaderboardEntry
   bIsVariable=true
   parent=VerticalBox_0
   widget=EventLeaderboardEntry_1
   properties {
    onButtonBaseClicked=(&self:WidgetTree.EventLeaderboardEntry_1.HandleOnButtonClicked)
    onButtonBaseDoubleClicked=(&self:WidgetTree.EventLeaderboardEntry_1.HandleOnButtonDoubleClicked)
    onButtonBaseHovered=(&self:WidgetTree.EventLeaderboardEntry_1.HandleOnButtonHovered)
    onButtonBaseUnhovered=(&self:WidgetTree.EventLeaderboardEntry_1.HandleOnButtonUnhovered)
    onSelectedChangedBase=(&self:WidgetTree.EventLeaderboardEntry_1.HandleOnSelectedChanged)
    visibility=HitTestInvisible
   }
   slot {class=VerticalBoxSlot}
  }
  EventLeaderboardEntry_2 {
   widgetClassPath=&EventLeaderboardEntry
   bIsVariable=true
   parent=VerticalBox_0
   widget=EventLeaderboardEntry_2
   properties {
    onButtonBaseClicked=(&self:WidgetTree.EventLeaderboardEntry_2.HandleOnButtonClicked)
    onButtonBaseDoubleClicked=(&self:WidgetTree.EventLeaderboardEntry_2.HandleOnButtonDoubleClicked)
    onButtonBaseHovered=(&self:WidgetTree.EventLeaderboardEntry_2.HandleOnButtonHovered)
    onButtonBaseUnhovered=(&self:WidgetTree.EventLeaderboardEntry_2.HandleOnButtonUnhovered)
    onSelectedChangedBase=(&self:WidgetTree.EventLeaderboardEntry_2.HandleOnSelectedChanged)
    visibility=HitTestInvisible
   }
   slot {class=VerticalBoxSlot}
  }
  EventLeaderboardEntry_3 {
   widgetClassPath=&EventLeaderboardEntry
   bIsVariable=true
   parent=VerticalBox_0
   widget=EventLeaderboardEntry_3
   properties {
    onButtonBaseClicked=(&self:WidgetTree.EventLeaderboardEntry_3.HandleOnButtonClicked)
    onButtonBaseDoubleClicked=(&self:WidgetTree.EventLeaderboardEntry_3.HandleOnButtonDoubleClicked)
    onButtonBaseHovered=(&self:WidgetTree.EventLeaderboardEntry_3.HandleOnButtonHovered)
    onButtonBaseUnhovered=(&self:WidgetTree.EventLeaderboardEntry_3.HandleOnButtonUnhovered)
    onSelectedChangedBase=(&self:WidgetTree.EventLeaderboardEntry_3.HandleOnSelectedChanged)
    visibility=HitTestInvisible
   }
   slot {class=VerticalBoxSlot}
  }
  Spacer_0 {widgetClassPath=&Spacer, parent=HorizontalBox_Entries, widget=Spacer_0, properties {size=<20 1>}, slot {class=HorizontalBoxSlot}}
  VerticalBox_2 {
   widgetClassPath=&VerticalBox
   parent=HorizontalBox_Entries
   widget=VerticalBox_2
   slot {class=HorizontalBoxSlot, properties {size {sizeRule=Fill, value=1}}}
  }
  EventLeaderboardEntry_4 {
   widgetClassPath=&EventLeaderboardEntry
   bIsVariable=true
   parent=VerticalBox_2
   widget=EventLeaderboardEntry_4
   properties {
    onButtonBaseClicked=(&self:WidgetTree.EventLeaderboardEntry_4.HandleOnButtonClicked)
    onButtonBaseDoubleClicked=(&self:WidgetTree.EventLeaderboardEntry_4.HandleOnButtonDoubleClicked)
    onButtonBaseHovered=(&self:WidgetTree.EventLeaderboardEntry_4.HandleOnButtonHovered)
    onButtonBaseUnhovered=(&self:WidgetTree.EventLeaderboardEntry_4.HandleOnButtonUnhovered)
    onSelectedChangedBase=(&self:WidgetTree.EventLeaderboardEntry_4.HandleOnSelectedChanged)
    visibility=HitTestInvisible
   }
   slot {class=VerticalBoxSlot}
  }
  EventLeaderboardEntry_5 {
   widgetClassPath=&EventLeaderboardEntry
   bIsVariable=true
   parent=VerticalBox_2
   widget=EventLeaderboardEntry_5
   properties {
    onButtonBaseClicked=(&self:WidgetTree.EventLeaderboardEntry_5.HandleOnButtonClicked)
    onButtonBaseDoubleClicked=(&self:WidgetTree.EventLeaderboardEntry_5.HandleOnButtonDoubleClicked)
    onButtonBaseHovered=(&self:WidgetTree.EventLeaderboardEntry_5.HandleOnButtonHovered)
    onButtonBaseUnhovered=(&self:WidgetTree.EventLeaderboardEntry_5.HandleOnButtonUnhovered)
    onSelectedChangedBase=(&self:WidgetTree.EventLeaderboardEntry_5.HandleOnSelectedChanged)
    visibility=HitTestInvisible
   }
   slot {class=VerticalBoxSlot}
  }
  EventLeaderboardEntry_6 {
   widgetClassPath=&EventLeaderboardEntry
   bIsVariable=true
   parent=VerticalBox_2
   widget=EventLeaderboardEntry_6
   properties {
    onButtonBaseClicked=(&self:WidgetTree.EventLeaderboardEntry_6.HandleOnButtonClicked)
    onButtonBaseDoubleClicked=(&self:WidgetTree.EventLeaderboardEntry_6.HandleOnButtonDoubleClicked)
    onButtonBaseHovered=(&self:WidgetTree.EventLeaderboardEntry_6.HandleOnButtonHovered)
    onButtonBaseUnhovered=(&self:WidgetTree.EventLeaderboardEntry_6.HandleOnButtonUnhovered)
    onSelectedChangedBase=(&self:WidgetTree.EventLeaderboardEntry_6.HandleOnSelectedChanged)
    visibility=HitTestInvisible
   }
   slot {class=VerticalBoxSlot}
  }
  EventLeaderboardEntry_7 {
   widgetClassPath=&EventLeaderboardEntry
   bIsVariable=true
   parent=VerticalBox_2
   widget=EventLeaderboardEntry_7
   properties {
    onButtonBaseClicked=(&self:WidgetTree.EventLeaderboardEntry_7.HandleOnButtonClicked)
    onButtonBaseDoubleClicked=(&self:WidgetTree.EventLeaderboardEntry_7.HandleOnButtonDoubleClicked)
    onButtonBaseHovered=(&self:WidgetTree.EventLeaderboardEntry_7.HandleOnButtonHovered)
    onButtonBaseUnhovered=(&self:WidgetTree.EventLeaderboardEntry_7.HandleOnButtonUnhovered)
    onSelectedChangedBase=(&self:WidgetTree.EventLeaderboardEntry_7.HandleOnSelectedChanged)
    visibility=HitTestInvisible
   }
   slot {class=VerticalBoxSlot}
  }
  EventLeaderboardEntry_8 {
   widgetClassPath=&EventLeaderboardEntry
   bIsVariable=true
   parent=VerticalBox_2
   widget=EventLeaderboardEntry_8
   properties {
    isPlayer=true
    onButtonBaseClicked=(&self:WidgetTree.EventLeaderboardEntry_8.HandleOnButtonClicked)
    onButtonBaseDoubleClicked=(&self:WidgetTree.EventLeaderboardEntry_8.HandleOnButtonDoubleClicked)
    onButtonBaseHovered=(&self:WidgetTree.EventLeaderboardEntry_8.HandleOnButtonHovered)
    onButtonBaseUnhovered=(&self:WidgetTree.EventLeaderboardEntry_8.HandleOnButtonUnhovered)
    onSelectedChangedBase=(&self:WidgetTree.EventLeaderboardEntry_8.HandleOnSelectedChanged)
    visibility=HitTestInvisible
   }
   slot {class=VerticalBoxSlot}
  }
  Overlay_1 {
   widgetClassPath=&Overlay
   parent=VerticalBox_1
   widget=Overlay_1
   slot {class=VerticalBoxSlot, properties {padding=<ltrb 0 10 0 0>}}
  }
  CommonBorder_ViewAllBG {
   widgetClassPath=&CommonBorder
   bIsVariable=true
   parent=Overlay_1
   widget=CommonBorder_ViewAllBG
   properties {
    background {
     bIsDynamicallyLoaded=false
     drawAs=Image
     imageType=NoImage
     mirroring=NoMirror
     resourceName=None
     resourceObject=null
     tiling=NoTile
     imageSize=<32 32>
     margin=<ltrb 0 0 0 0>
     outlineSettings {
      bUseBrushTransparency=false
      roundingType=HalfHeightRadius
      width=0
      color {colorUseRule=UseColor_Specified, specifiedColor=<rgba 0 0 0 0>}
      cornerRadii=<0 0 0 0>
     }
     tintColor {colorUseRule=UseColor_Specified, specifiedColor=<rgba 1 1 1 1>}
     uVRegion {bIsValid=false, max=<0 0>, min=<0 0>}
    }
    brushColor=<rgba 0 0.0762 0.4793 1>
    padding=<ltrb 0 0 0 0>
   }
   slot {class=OverlaySlot, properties {horizontalAlignment=HAlign_Fill, verticalAlignment=VAlign_Fill}}
  }
  Button_ViewAll {
   widgetClassPath=&IconTextButton
   bIsVariable=true
   parent=CommonBorder_ViewAllBG
   widget=Button_ViewAll
   properties {
    buttonText=""
    contentAlignment=HAlign_Center
    minWidth=300
    onButtonBaseClicked=(&self:WidgetTree.Button_ViewAll.HandleOnButtonClicked)
    onButtonBaseDoubleClicked=(&self:WidgetTree.Button_ViewAll.HandleOnButtonDoubleClicked)
    onButtonBaseHovered=(&self:WidgetTree.Button_ViewAll.HandleOnButtonHovered)
    onButtonBaseUnhovered=(&self:WidgetTree.Button_ViewAll.HandleOnButtonUnhovered)
    onSelectedChangedBase=(&self:WidgetTree.Button_ViewAll.HandleOnSelectedChanged)
    priority=1
    style=/FortUILibrary/Foundation/Buttons/ButtonStyles/ButtonStyle-TextOnlyBase_EmptyFocus.ButtonStyle-TextOnlyBase_EmptyFocus_C
    triggeringInputAction {dataTable=&InputActionDataTable, rowName=EventLeaderboardViewAll}
   }
   slot {class=BorderSlot, properties {padding=<ltrb 0 0 0 0>}}
  }
  Overlay_2 {widgetClassPath=&Overlay, namedSlotHost=Button_ViewAll, widget=Overlay_2}
  Text_ViewFullLeaderboard {
   widgetClassPath=&CommonTextBlock
   bIsVariable=true
   parent=Overlay_2
   widget=Text_ViewFullLeaderboard
   properties {
    justification=Right
    text=VIEW FULL LEADERBOARD
    font {
     bForceMonospaced=false
     bMaterialIsStencil=false
     fontMaterial=null
     fontObject=&Burbank
     letterSpacing=0
     monospacedWidth=1
     size=20
     skewAmount=0
     typefaceFontName=Burbank-PlayerPower
     outlineSettings {
      bApplyOutlineToDropShadows=false
      bMiteredCorners=false
      bSeparateFillAlpha=false
      outlineMaterial=null
      outlineSize=0
      outlineColor=<rgba 0 0 0 1>
     }
    }
   }
   slot {class=OverlaySlot, properties {verticalAlignment=VAlign_Center}}
  }
  SizeBox_6 {
   widgetClassPath=&SizeBox
   parent=Overlay_SummaryContent
   widget=SizeBox_6
   properties {bOverride_MinDesiredHeight=true, bOverride_MinDesiredWidth=true, minDesiredHeight=61, minDesiredWidth=260}
   slot {class=OverlaySlot}
  }
  CommonBorder_ScoringTitleBG {
   widgetClassPath=&CommonBorder
   bIsVariable=true
   parent=SizeBox_6
   widget=CommonBorder_ScoringTitleBG
   properties {
    verticalAlignment=VAlign_Center
    background {
     bIsDynamicallyLoaded=false
     drawAs=Box
     imageType=NoImage
     mirroring=NoMirror
     resourceName=None
     resourceObject=/Game/UI/Frontend/Showdown/Materials/M_UI_SkewBox_TitleCut
     tiling=NoTile
     imageSize=<256 64>
     margin=<ltrb 0.4 0.2 0.4 0.2>
     outlineSettings {
      bUseBrushTransparency=false
      roundingType=HalfHeightRadius
      width=0
      color {colorUseRule=UseColor_Specified, specifiedColor=<rgba 0 0 0 0>}
      cornerRadii=<0 0 0 0>
     }
     tintColor {colorUseRule=UseColor_Specified, specifiedColor=<rgba 1 1 1 1>}
     uVRegion {bIsValid=false, max=<0 0>, min=<0 0>}
    }
    brushColor=<rgba 0 0.2086 1 1>
    padding=<ltrb 33 12 33 6>
   }
   slot {class=SizeBoxSlot}
  }
  CommonTextBlock_Title {
   widgetClassPath=&CommonTextBlock
   bIsVariable=true
   parent=CommonBorder_ScoringTitleBG
   widget=CommonTextBlock_Title
   properties {
    text=TOP SCORING PLAYERS
    font {
     bForceMonospaced=false
     bMaterialIsStencil=false
     fontMaterial=null
     fontObject=&Burbank
     letterSpacing=0
     monospacedWidth=1
     size=24
     skewAmount=0
     typefaceFontName=BurbankRegular-Black
     outlineSettings {
      bApplyOutlineToDropShadows=false
      bMiteredCorners=false
      bSeparateFillAlpha=false
      outlineMaterial=null
      outlineSize=0
      outlineColor=<rgba 0 0 0 1>
     }
    }
   }
   slot {class=BorderSlot, properties {verticalAlignment=VAlign_Center, padding=<ltrb 33 12 33 6>}}
  }
  Text_DebugId {
   widgetClassPath=&TextBlock
   bInherited=true
   parent=Overlay_SummaryContent
   widget=Text_DebugId
   properties {
    autoWrapText=true
    justification=Center
    text=EventId: Event EventWindowId: EventWindow
    wrapTextAt=210
    wrappingPolicy=AllowPerCharacterWrapping
    colorAndOpacity {colorUseRule=UseColor_Specified, specifiedColor=<rgba 0 0 0 1>}
    font {
     bForceMonospaced=false
     bMaterialIsStencil=false
     fontMaterial=null
     fontObject=&Roboto
     letterSpacing=0
     monospacedWidth=1
     size=20
     skewAmount=0
     typefaceFontName=Bold
     outlineSettings {
      bApplyOutlineToDropShadows=false
      bMiteredCorners=false
      bSeparateFillAlpha=false
      outlineMaterial=null
      outlineSize=2
      outlineColor=<rgba 1 1 1 1>
     }
    }
   }
   slot {class=OverlaySlot, properties {horizontalAlignment=HAlign_Fill, verticalAlignment=VAlign_Center}}
  }
  Text_TimeRemaining {widgetClassPath=&CommonTextBlock, bInherited=true, widget=null}
  Image_PosterFront {widgetClassPath=/Script/EpicCMSUIFramework.EpicCMSImage, bInherited=true, widget=null}
  Image_PosterBack {widgetClassPath=&EpicCMSImage, bInherited=true, widget=null}
  Image_LoadingScreen {widgetClassPath=&EpicCMSImage, bInherited=true, widget=null}
  Image_PlaylistTile {widgetClassPath=&EpicCMSImage, bInherited=true, widget=null}
 ]
}
