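# Tournament screen that lists live games, built on FortTournamentLiveGames with an event details header and a collapsed Watch Now button.
# Init passes tournament and event IDs to the parent and header, and Colorize tints the backgrounds from tournament display info.
# Parent notifications toggle the loading row or switch the list to a no-games message.
# Back and Close, shown by non-touch or touch input, both fire CallEventLiveGamesBack.
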WidgetBlueprint /CompeteOldTab/UI/Events/TournamentLiveGames
animations [Intro {
  binding_count=2
  length_seconds=0.2526
  object=&self:Intro
  sequence {bindings [
    SizeBox_TopContainer {
     id=#1
     possessed_class=SizeBox
     tracks [{
       type=MovieScene2DTransformTrack
       sections [{
         type=MovieScene2DTransformSection
         end_frame=5
         start_frame=0
         channels [
          Translation.X {type=MovieSceneScriptingFloatChannel}
          Translation.Y {type=MovieSceneScriptingFloatChannel}
          Angle {type=MovieSceneScriptingFloatChannel}
          Scale.X {type=MovieSceneScriptingFloatChannel, keys [[0,0.8563],[4.0213,1.0254],[4.6747,0.9945],[5.0527,1]]}
          Scale.Y {type=MovieSceneScriptingFloatChannel, keys [[0,0.8563],[4.0213,1.0254],[4.6747,0.9945],[5.0527,1]]}
          Shear.X {type=MovieSceneScriptingFloatChannel}
          Shear.Y {type=MovieSceneScriptingFloatChannel}
         ]
        }]
      }]
    }
    BackgroundBlur_1 {
     id=#2
     possessed_class=BackgroundBlur
     tracks [{
       type=MovieSceneFloatTrack
       display_name=Blur Strength
       sections [{type=MovieSceneFloatSection, end_frame=4, start_frame=0, channels [None {type=MovieSceneScriptingFloatChannel, keys [[0,0],[4.3837,25]]}]}]
      }]
    }
   ]}
 }]
graphs [
 CallEventLiveGamesBack {nodes [CallEventLiveGamesBack() {is=title K2Node_FunctionEntry}]}
 EventGraph {nodes [
   n1 = "On Button Clicked (BackButton)"!() -> (OutputDelegate, then, Button:Common Button Legacy Object Reference) {is=title K2Node_ComponentBoundEvent}
   Call Call Event Live Games Back!(execute:Exec=[@n1.then, @n2.then], self:Tournament Live Games Object Reference) -> (then:Exec) {is=title K2Node_CallDelegate}
   n2 = "On Button Clicked (CloseButton)"!() -> (OutputDelegate, then, Button:Common Button Legacy Object Reference) {is=title K2Node_ComponentBoundEvent}
   n3 = EventColorize!() -> (OutputDelegate, then, TournamentDisplayInfo:Fort Tournament Display Info Structure) {is=title K2Node_CustomEvent}
   n4 = Break Fort Tournament Display Info!(FortTournamentDisplayInfo:"Fort Tournament Display Info Structure (by ref)"=@n3.TournamentDisplayInfo) -> (
    TitleLine1:Text
    TitleLine2:Text
    ScheduleInfo:Text
    FlavorDescription:Text
    DetailsDescription:Text
    ShortFormatTitle:Text
    LongFormatTitle:Text
    BackgroundTitle:Text
    PinScoreRequirement:int
    PinEarnedText:Text
    BaseColor:Linear Color Structure
    PrimaryColor:Linear Color Structure
    SecondaryColor:Linear Color Structure
    HighlightColor:Linear Color Structure
    TitleColor:Linear Color Structure
    ShadowColor:Linear Color Structure
    BackgroundLeftColor:Linear Color Structure
    BackgroundRightColor:Linear Color Structure
    BackgroundTextColor:Linear Color Structure
    PosterFadeColor:Linear Color Structure
    AlertText:Text
    SeriesPointLeaderboardName:Text
    AlertType:EFortTournamentAlertType Enum
    RoundDisplayNames:Array of Texts
   ) {is=title K2Node_BreakStruct}
   n5 = Break Linear Color!(LinearColor:"Linear Color Structure (by ref)"=@n4.BackgroundLeftColor) -> (R:float=~, G:float=~, B:float=~, A:float=~) {is=title K2Node_BreakStruct}
   > SetBrushColor(
    self=Get CommonBorder_ContainerBG() -> (CommonBorder_ContainerBG:Common Border Object Reference)
    InBrushColor=Make Linear Color!(R:float=@n5, G:float=@n5.G, B:float=@n5.B, A:float=0.5) -> (LinearColor:Linear Color Structure) {is=title K2Node_MakeStruct}
   )
   > SetBrushColor(self=Get CommonBorder_OptionsBarBG() -> (CommonBorder_OptionsBarBG:Common Border Object Reference), InBrushColor=@n4.BackgroundLeftColor)
   n6 = Init!() -> (OutputDelegate, then, TournamentId:String, EventId:String) {is=title K2Node_CustomEvent}
   > SetActiveWidgetIndex(self=Get Switcher_LeaderboardEntries() -> (Switcher_LeaderboardEntries:Common Widget Switcher Legacy Object Reference), Index=0)
   > SetTournamentIds!(execute, self:Fort Tournament Live Games Object Reference, InTournamentId:String=@n6.TournamentId, InEventWindowId:String=@n6.EventId) -> (then)
   > SetTournament!(
    execute
    self:Fort Showdown Detail View Object Reference=Get EventLeaderboardEventDetails() -> (EventLeaderboardEventDetails:Event Leaderboard Event Details Object Reference)
    TournamentId:String=@n6.TournamentId
    EventId:String=@n6.EventId
    bForceRefreshUI:bool=true
   ) -> (then)
   HandleViewLiveGames() {is=title K2Node_CustomEvent}
   > OnViewLiveGames(self:Fort Tournament Live Games Object Reference)
   n7 = Event NotifyLoadingMoreGames!() -> (OutputDelegate, then, bIsLoading:bool) {is=title K2Node_Event}
   > n8 = Branch(Condition=@n7.bIsLoading)
   n9 = Get Content_LoadingEntries() -> (Content_LoadingEntries:Horizontal Box Object Reference)
   > SetVisibility(self=@n9, InVisibility=Visible)
   SetVisibility(execute=@n8.else, self=@n9, InVisibility=Collapsed)
   Event NotifyNoLiveGames() {is=title K2Node_Event}
   > SetActiveWidgetIndex(self=Get Switcher_LeaderboardEntries(), Index=1)
  ]}
]
variables [CallEventLiveGamesBack {}]
widgets {
 info {inheritedWidgetCount=2, parentClass=/Script/FortniteUI.FortTournamentLiveGames, rootWidgetClass=&SafeZone, widgetCount=34}
 widgets [
  SafeZone_0 {widgetClassPath=&SafeZone, bIsVariable=true, widget=SafeZone_0}
  Overlay_2 {widgetClassPath=&Overlay, parent=SafeZone_0, widget=Overlay_2, slot {class=SafeZoneSlot}}
  VerticalBox_0 {
   widgetClassPath=&VerticalBox
   parent=Overlay_2
   widget=VerticalBox_0
   slot {class=OverlaySlot, properties {horizontalAlignment=HAlign_Fill, verticalAlignment=VAlign_Fill}}
  }
  EventLeaderboardEventDetails {
   widgetClassPath=/Game/EventLeaderboardEventDetails
   bIsVariable=true
   parent=VerticalBox_0
   widget=EventLeaderboardEventDetails
   slot {class=VerticalBoxSlot, properties {padding=<ltrb 40 35 -200 0>}}
  }
  Overlay_0 {
   widgetClassPath=&Overlay
   parent=VerticalBox_0
   widget=Overlay_0
   slot {
    class=VerticalBoxSlot
    properties {horizontalAlignment=HAlign_Left, verticalAlignment=VAlign_Top, padding=<ltrb 0 0 0 50>, size {sizeRule=Fill, value=1}}
   }
  }
  SizeBox_0 {
   widgetClassPath=&SizeBox
   parent=Overlay_0
   widget=SizeBox_0
   properties {bOverride_WidthOverride=true, widthOverride=1100}
   slot {class=OverlaySlot, properties {padding=<ltrb 0 40 0 0>}}
  }
  Overlay_LeaderboardContainer {widgetClassPath=&Overlay, parent=SizeBox_0, widget=Overlay_LeaderboardContainer, slot {class=SizeBoxSlot}}
  BackgroundBlur_1 {
   widgetClassPath=&BackgroundBlur
   parent=Overlay_LeaderboardContainer
   widget=BackgroundBlur_1
   properties {blurStrength=50}
   slot {class=OverlaySlot, properties {horizontalAlignment=HAlign_Fill, padding=<ltrb 50 40 20 0>}}
  }
  CommonBorder_ContainerBG {
   widgetClassPath=&CommonBorder
   bIsVariable=true
   parent=BackgroundBlur_1
   widget=CommonBorder_ContainerBG
   properties {
    background {
     bIsDynamicallyLoaded=false
     drawAs=Image
     imageType=NoImage
     mirroring=NoMirror
     resourceName=None
     resourceObject=null
     tiling=NoTile
     imageSize=<32 32>
     margin=<ltrb 0 0 0 0>
     outlineSettings {
      bUseBrushTransparency=false
      roundingType=HalfHeightRadius
      width=0
      color {colorUseRule=UseColor_Specified, specifiedColor=<rgba 0 0 0 0>}
      cornerRadii=<0 0 0 0>
     }
     tintColor {colorUseRule=UseColor_Specified, specifiedColor=<rgba 1 1 1 1>}
     uVRegion {bIsValid=false, max=<0 0>, min=<0 0>}
    }
    brushColor=<rgba 0.0052 0.0723 0.4179 0.5>
    padding=<ltrb 0 20 0 0>
   }
   slot {class=BackgroundBlurSlot, properties {padding=<ltrb 0 0 0 0>}}
  }
  Switcher_LeaderboardEntries {
   widgetClassPath=&CommonWidgetSwitcherLegacy
   bIsVariable=true
   parent=CommonBorder_ContainerBG
   widget=Switcher_LeaderboardEntries
   slot {class=BorderSlot, properties {padding=<ltrb 0 20 0 0>}}
  }
  Content_LeaderboardEntries {
   widgetClassPath=&VerticalBox
   bIsVariable=true
   parent=Switcher_LeaderboardEntries
   widget=Content_LeaderboardEntries
   slot {class=WidgetSwitcherSlot}
  }
  ListView_SessionEntries {
   widgetClassPath=&CommonListView
   bInherited=true
   bIsVariable=true
   parent=Content_LeaderboardEntries
   widget=ListView_SessionEntries
   properties {entryWidgetClass=/Game/UI/Events/TournamentLiveGamesEntry.TournamentLiveGamesEntry_C, numDesignerPreviewEntries=20}
   slot {class=VerticalBoxSlot, properties {size {sizeRule=Fill, value=1}}}
  }
  CommonBorder_OptionsBarBG {
   widgetClassPath=&CommonBorder
   bIsVariable=true
   parent=Content_LeaderboardEntries
   widget=CommonBorder_OptionsBarBG
   properties {
    horizontalAlignment=HAlign_Center
    background {
     bIsDynamicallyLoaded=false
     drawAs=Image
     imageType=NoImage
     mirroring=NoMirror
     resourceName=None
     resourceObject=null
     tiling=NoTile
     imageSize=<32 32>
     margin=<ltrb 0 0 0 0>
     outlineSettings {
      bUseBrushTransparency=false
      roundingType=HalfHeightRadius
      width=0
      color {colorUseRule=UseColor_Specified, specifiedColor=<rgba 0 0 0 0>}
      cornerRadii=<0 0 0 0>
     }
     tintColor {colorUseRule=UseColor_Specified, specifiedColor=<rgba 1 1 1 1>}
     uVRegion {bIsValid=false, max=<0 0>, min=<0 0>}
    }
    brushColor=<rgba 0.006 0.0742 0.4969 1>
    padding=<ltrb 0 0 0 0>
   }
   slot {class=VerticalBoxSlot}
  }
  Content_LoadingEntries {
   widgetClassPath=&HorizontalBox
   bIsVariable=true
   parent=Content_LeaderboardEntries
   widget=Content_LoadingEntries
   properties {visibility=Collapsed}
   slot {class=VerticalBoxSlot, properties {horizontalAlignment=HAlign_Center, padding=<ltrb 20 0 20 20>}}
  }
  ScaleBox_0 {
   widgetClassPath=&ScaleBox
   parent=Content_LoadingEntries
   widget=ScaleBox_0
   properties {stretchDirection=DownOnly}
   slot {class=HorizontalBoxSlot, properties {verticalAlignment=VAlign_Center, padding=<ltrb 25 0 25 0>}}
  }
  CommonTextBlock_0 {
   widgetClassPath=&CommonTextBlock
   parent=ScaleBox_0
   widget=CommonTextBlock_0
   properties {
    text=Loading Live Games
    textTransformPolicy=ToUpper
    font {
     bForceMonospaced=false
     bMaterialIsStencil=false
     fontMaterial=null
     fontObject=&BurbankSmall
     letterSpacing=0
     monospacedWidth=1
     size=18
     skewAmount=0
     typefaceFontName=Bold
     outlineSettings {
      bApplyOutlineToDropShadows=false
      bMiteredCorners=false
      bSeparateFillAlpha=false
      outlineMaterial=null
      outlineSize=0
      outlineColor=<rgba 0 0 0 1>
     }
    }
    shadowOffset=<0 0>
   }
   slot {class=ScaleBoxSlot}
  }
  Image_0 {
   widgetClassPath=&Image
   bIsVariable=true
   parent=Content_LoadingEntries
   widget=Image_0
   properties {brush {
     bIsDynamicallyLoaded=false
     drawAs=Image
     imageType=NoImage
     mirroring=NoMirror
     resourceName=None
     resourceObject=/FortUILibrary/UIKit/Throbber/MI_UIKit_Monochrome_Throbber
     tiling=NoTile
     imageSize=<32 32>
     margin=<ltrb 0 0 0 0>
     outlineSettings {
      bUseBrushTransparency=false
      roundingType=HalfHeightRadius
      width=0
      color {colorUseRule=UseColor_Specified, specifiedColor=<rgba 0 0 0 0>}
      cornerRadii=<0 0 0 0>
     }
     tintColor {colorUseRule=UseColor_Specified, specifiedColor=<rgba 1 1 1 1>}
     uVRegion {bIsValid=false, max=<0 0>, min=<0 0>}
    }}
   slot {class=HorizontalBoxSlot, properties {horizontalAlignment=HAlign_Center, verticalAlignment=VAlign_Center}}
  }
  Content_NoLeaderboardEntries {
   widgetClassPath=&CommonBorder
   bIsVariable=true
   parent=Switcher_LeaderboardEntries
   widget=Content_NoLeaderboardEntries
   properties {padding=<ltrb 10 40 10 25>}
   slot {class=WidgetSwitcherSlot}
  }
  Text_NoLeaderboardEntries {
   widgetClassPath=&CommonTextBlock
   bIsVariable=true
   parent=Content_NoLeaderboardEntries
   widget=Text_NoLeaderboardEntries
   properties {
    autoWrapText=true
    justification=Center
    text="Live matches for top players will be shown here. Check back soon!"
    font {
     bForceMonospaced=false
     bMaterialIsStencil=false
     fontMaterial=null
     fontObject=&Burbank
     letterSpacing=0
     monospacedWidth=1
     size=24
     skewAmount=0
     typefaceFontName=Burbank-PlayerPower
     outlineSettings {
      bApplyOutlineToDropShadows=false
      bMiteredCorners=false
      bSeparateFillAlpha=false
      outlineMaterial=null
      outlineSize=0
      outlineColor=<rgba 0 0 0 1>
     }
    }
    shadowOffset=<0 0>
   }
   slot {class=BorderSlot, properties {padding=<ltrb 10 40 10 25>}}
  }
  HorizontalBox_286 {
   widgetClassPath=&HorizontalBox
   parent=Overlay_LeaderboardContainer
   widget=HorizontalBox_286
   slot {class=OverlaySlot, properties {horizontalAlignment=HAlign_Fill}}
  }
  SizeBox_6 {
   widgetClassPath=&SizeBox
   parent=HorizontalBox_286
   widget=SizeBox_6
   properties {bOverride_MinDesiredHeight=true, bOverride_MinDesiredWidth=true, minDesiredHeight=61, minDesiredWidth=260}
   slot {class=HorizontalBoxSlot, properties {horizontalAlignment=HAlign_Left, verticalAlignment=VAlign_Top}}
  }
  CommonBorder_ScoringTitleBG {
   widgetClassPath=&CommonBorder
   bIsVariable=true
   parent=SizeBox_6
   widget=CommonBorder_ScoringTitleBG
   properties {
    verticalAlignment=VAlign_Center
    background {
     bIsDynamicallyLoaded=false
     drawAs=Box
     imageType=NoImage
     mirroring=NoMirror
     resourceName=None
     resourceObject=/Game/UI/Frontend/Showdown/Materials/M_UI_SkewBox_TitleCut
     tiling=NoTile
     imageSize=<256 64>
     margin=<ltrb 0.4 0.2 0.4 0.2>
     outlineSettings {
      bUseBrushTransparency=false
      roundingType=HalfHeightRadius
      width=0
      color {colorUseRule=UseColor_Specified, specifiedColor=<rgba 0 0 0 0>}
      cornerRadii=<0 0 0 0>
     }
     tintColor {colorUseRule=UseColor_Specified, specifiedColor=<rgba 1 1 1 1>}
     uVRegion {bIsValid=false, max=<0 0>, min=<0 0>}
    }
    brushColor=<rgba 0 0.2086 1 1>
    padding=<ltrb 33 8 33 6>
   }
   slot {class=SizeBoxSlot}
  }
  HorizontalBox_106 {
   widgetClassPath=&HorizontalBox
   parent=CommonBorder_ScoringTitleBG
   widget=HorizontalBox_106
   slot {class=BorderSlot, properties {verticalAlignment=VAlign_Center, padding=<ltrb 33 8 33 6>}}
  }
  Overlay_OnlineStatus {
   widgetClassPath=&Overlay
   bIsVariable=true
   parent=HorizontalBox_106
   widget=Overlay_OnlineStatus
   slot {class=HorizontalBoxSlot, properties {horizontalAlignment=HAlign_Center, verticalAlignment=VAlign_Center}}
  }
  Image_ping {
   widgetClassPath=&Image
   parent=Overlay_OnlineStatus
   widget=Image_ping
   properties {brush {
     bIsDynamicallyLoaded=false
     drawAs=Image
     imageType=NoImage
     mirroring=NoMirror
     resourceName=None
     resourceObject=/Game/UI/Events/Leaderboards/Materials/M_UI_LiveIcon_Inst
     tiling=NoTile
     imageSize=<60 60>
     margin=<ltrb 0 0 0 0>
     outlineSettings {
      bUseBrushTransparency=false
      roundingType=HalfHeightRadius
      width=0
      color {colorUseRule=UseColor_Specified, specifiedColor=<rgba 0 0 0 0>}
      cornerRadii=<0 0 0 0>
     }
     tintColor {colorUseRule=UseColor_Specified, specifiedColor=<rgba 1 1 1 1>}
     uVRegion {bIsValid=false, max=<0 0>, min=<0 0>}
    }}
   slot {class=OverlaySlot, properties {padding=<ltrb 0 -10 0 -10>}}
  }
  Image_OnlineStatus {
   widgetClassPath=&Image
   bIsVariable=true
   parent=Overlay_OnlineStatus
   widget=Image_OnlineStatus
   properties {
    visibility=SelfHitTestInvisible
    brush {
     bIsDynamicallyLoaded=false
     drawAs=Image
     imageType=NoImage
     mirroring=NoMirror
     resourceName=None
     resourceObject=/Game/UI/Events/Leaderboards/Materials/M_UI_LiveIconCenter_Inst
     tiling=NoTile
     imageSize=<30 30>
     margin=<ltrb 0 0 0 0>
     outlineSettings {
      bUseBrushTransparency=false
      roundingType=HalfHeightRadius
      width=0
      color {colorUseRule=UseColor_Specified, specifiedColor=<rgba 0 0 0 0>}
      cornerRadii=<0 0 0 0>
     }
     tintColor {colorUseRule=UseColor_Specified, specifiedColor=<rgba 1 1 1 1>}
     uVRegion {bIsValid=false, max=<0 0>, min=<0 0>}
    }
    colorAndOpacity=<rgba 1 0 0 1>
   }
   slot {class=OverlaySlot, properties {horizontalAlignment=HAlign_Center, verticalAlignment=VAlign_Center}}
  }
  Text_RoundTitle {
   widgetClassPath=&CommonTextBlock
   bIsVariable=true
   parent=HorizontalBox_106
   widget=Text_RoundTitle
   properties {
    text=Live Games
    font {
     bForceMonospaced=false
     bMaterialIsStencil=false
     fontMaterial=null
     fontObject=&Burbank
     letterSpacing=0
     monospacedWidth=1
     size=24
     skewAmount=0
     typefaceFontName=BurbankRegular-Black
     outlineSettings {
      bApplyOutlineToDropShadows=false
      bMiteredCorners=false
      bSeparateFillAlpha=false
      outlineMaterial=null
      outlineSize=0
      outlineColor=<rgba 0 0 0 1>
     }
    }
   }
   slot {class=HorizontalBoxSlot, properties {verticalAlignment=VAlign_Center, padding=<ltrb 33 8 33 6>}}
  }
  SizeBox_129 {
   widgetClassPath=&SizeBox
   parent=HorizontalBox_286
   widget=SizeBox_129
   slot {
    class=HorizontalBoxSlot
    properties {horizontalAlignment=HAlign_Right, verticalAlignment=VAlign_Top, padding=<ltrb 0 -20 0 0>, size {sizeRule=Fill, value=1}}
   }
  }
  Button_WatchGame {
   widgetClassPath=/CompeteOldTab/Frontend/Showdown/EventsSimpleButton
   bInherited=true
   bIsVariable=true
   parent=SizeBox_129
   widget=Button_WatchGame
   properties {
    buttonNameText=Watch NOw
    onButtonBaseClicked=(&self:WidgetTree.Button_WatchGame.HandleOnButtonClicked)
    onButtonBaseDoubleClicked=(&self:WidgetTree.Button_WatchGame.HandleOnButtonDoubleClicked)
    onButtonBaseHovered=(&self:WidgetTree.Button_WatchGame.HandleOnButtonHovered)
    onButtonBaseUnhovered=(&self:WidgetTree.Button_WatchGame.HandleOnButtonUnhovered)
    onSelectedChangedBase=(&self:WidgetTree.Button_WatchGame.HandleOnSelectedChanged)
    visibility=Collapsed
    triggeringInputAction {dataTable=&InputActionDataTable, rowName=GenericFaceButtonBottom}
   }
   slot {class=SizeBoxSlot, properties {horizontalAlignment=HAlign_Center, verticalAlignment=VAlign_Top}}
  }
  HorizontalBox_0 {
   widgetClassPath=&HorizontalBox
   parent=Overlay_2
   widget=HorizontalBox_0
   slot {class=OverlaySlot, properties {horizontalAlignment=HAlign_Right, verticalAlignment=VAlign_Bottom}}
  }
  VisibilityBorder_NonTouch {
   widgetClassPath=&CommonHardwareVisibilityBorder
   parent=HorizontalBox_0
   widget=VisibilityBorder_NonTouch
   properties {visibilityQuery {
     autoDescription=" NONE( Input.Touch )"
     tokenStreamVersion=0
     userDescription=""
     queryTokenStream [0,1,3,1,0]
     tagDictionary [{tagName=Input.Touch}]
    }}
   slot {class=HorizontalBoxSlot, properties {padding=<ltrb 20 20 20 20>}}
  }
  BackButton {
   widgetClassPath=&EventsSimpleButton
   bIsVariable=true
   parent=VisibilityBorder_NonTouch
   widget=BackButton
   properties {
    buttonNameText=Back
    onButtonBaseClicked=(&self:WidgetTree.BackButton.HandleOnButtonClicked)
    onButtonBaseDoubleClicked=(&self:WidgetTree.BackButton.HandleOnButtonDoubleClicked)
    onButtonBaseHovered=(&self:WidgetTree.BackButton.HandleOnButtonHovered)
    onButtonBaseUnhovered=(&self:WidgetTree.BackButton.HandleOnButtonUnhovered)
    onSelectedChangedBase=(&self:WidgetTree.BackButton.HandleOnSelectedChanged)
    triggeringInputAction {dataTable=&InputActionDataTable, rowName=GenericBack}
   }
   slot {class=BorderSlot, properties {padding=<ltrb 0 0 0 0>}}
  }
  VisibilityBorder_Touch {
   widgetClassPath=&CommonHardwareVisibilityBorder
   parent=Overlay_2
   widget=VisibilityBorder_Touch
   properties {visibilityQuery {
     autoDescription=" ANY( Input.Touch )"
     tokenStreamVersion=0
     userDescription=""
     queryTokenStream [0,1,1,1,0]
     tagDictionary [{tagName=Input.Touch}]
    }}
   slot {class=OverlaySlot, properties {horizontalAlignment=HAlign_Right, padding=<ltrb 0 25 50 20>}}
  }
  CloseButton {
   widgetClassPath=/FortUILibrary/Foundation/Buttons/CloseButton
   bIsVariable=true
   parent=VisibilityBorder_Touch
   widget=CloseButton
   properties {
    onButtonBaseClicked=(&self:WidgetTree.CloseButton.HandleOnButtonClicked)
    onButtonBaseDoubleClicked=(&self:WidgetTree.CloseButton.HandleOnButtonDoubleClicked)
    onButtonBaseHovered=(&self:WidgetTree.CloseButton.HandleOnButtonHovered)
    onButtonBaseUnhovered=(&self:WidgetTree.CloseButton.HandleOnButtonUnhovered)
    onSelectedChangedBase=(&self:WidgetTree.CloseButton.HandleOnSelectedChanged)
   }
   slot {class=BorderSlot, properties {padding=<ltrb 0 0 0 0>}}
  }
 ]
}
