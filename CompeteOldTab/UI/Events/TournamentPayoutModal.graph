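# Tournament payout modal widget on FortTournamentPayoutModal, showing event details, view tabs and a scrolling list of payout thresholds.
# Init feeds tournament and window IDs to detail and header widgets, then resets scrolling.
# EventColorize stores display info and tints backgrounds; tab initialization colors buttons from it.
# NotifyNoPayouts switches a default-collapsed panel to its message and fires NoPayout; input-specific Back or Close buttons fire CallEventBack.
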WidgetBlueprint /CompeteOldTab/UI/Events/TournamentPayoutModal
animations [Intro {
  binding_count=2
  length_seconds=0.2526
  object=&self:Intro
  sequence {bindings [
    SizeBox_TopContainer {
     id=#1
     possessed_class=SizeBox
     tracks [{
       type=MovieScene2DTransformTrack
       sections [{
         type=MovieScene2DTransformSection
         end_frame=5
         start_frame=0
         channels [
          Translation.X {type=MovieSceneScriptingFloatChannel}
          Translation.Y {type=MovieSceneScriptingFloatChannel}
          Angle {type=MovieSceneScriptingFloatChannel}
          Scale.X {type=MovieSceneScriptingFloatChannel, keys [[0,0.8563],[4.0213,1.0254],[4.6747,0.9945],[5.0527,1]]}
          Scale.Y {type=MovieSceneScriptingFloatChannel, keys [[0,0.8563],[4.0213,1.0254],[4.6747,0.9945],[5.0527,1]]}
          Shear.X {type=MovieSceneScriptingFloatChannel}
          Shear.Y {type=MovieSceneScriptingFloatChannel}
         ]
        }]
      }]
    }
    BackgroundBlur_1 {
     id=#2
     possessed_class=BackgroundBlur
     tracks [{
       type=MovieSceneFloatTrack
       display_name=Blur Strength
       sections [{type=MovieSceneFloatSection, end_frame=4, start_frame=0, channels [None {type=MovieSceneScriptingFloatChannel, keys [[0,0],[4.3837,25]]}]}]
      }]
    }
   ]}
 }]
graphs [
 CallEventBack {nodes [CallEventBack() {is=title K2Node_FunctionEntry}]}
 EmptyFunction {nodes [
   EmptyFunction() {is=title K2Node_FunctionEntry}
   > Return Node!(execute, Handled:bool=true)
  ]}
 EventGraph {nodes [
   n1 = "On Button Clicked (BackButton)"!() -> (OutputDelegate, then, Button:Common Button Legacy Object Reference) {is=title K2Node_ComponentBoundEvent}
   Call Call Event Back!(execute:Exec=[@n1.then, @n2.then], self:Tournament Payout Modal Object Reference) -> (then:Exec) {is=title K2Node_CallDelegate}
   n2 = "On Button Clicked (CloseButton)"!() -> (OutputDelegate, then, Button:Common Button Legacy Object Reference) {is=title K2Node_ComponentBoundEvent}
   n3 = EventColorize!() -> (OutputDelegate, then, TournamentDisplayInfo:Fort Tournament Display Info Structure) {is=title K2Node_CustomEvent}
   > Set CurrentDisplayInfo(CurrentDisplayInfo:Fort Tournament Display Info Structure=@n3.TournamentDisplayInfo) -> (Output_Get:Fort Tournament Display Info Structure)
   n4 = Break Fort Tournament Display Info!(FortTournamentDisplayInfo:"Fort Tournament Display Info Structure (by ref)"=@n3.TournamentDisplayInfo) -> (
    TitleLine1:Text
    TitleLine2:Text
    ScheduleInfo:Text
    FlavorDescription:Text
    DetailsDescription:Text
    ShortFormatTitle:Text
    LongFormatTitle:Text
    BackgroundTitle:Text
    PinScoreRequirement:int
    PinEarnedText:Text
    BaseColor:Linear Color Structure
    PrimaryColor:Linear Color Structure
    SecondaryColor:Linear Color Structure
    HighlightColor:Linear Color Structure
    TitleColor:Linear Color Structure
    ShadowColor:Linear Color Structure
    BackgroundLeftColor:Linear Color Structure
    BackgroundRightColor:Linear Color Structure
    BackgroundTextColor:Linear Color Structure
    PosterFadeColor:Linear Color Structure
    AlertText:Text
    SeriesPointLeaderboardName:Text
    AlertType:EFortTournamentAlertType Enum
    RoundDisplayNames:Array of Texts
   ) {is=title K2Node_BreakStruct}
   n5 = Break Linear Color!(LinearColor:"Linear Color Structure (by ref)"=@n4.BackgroundLeftColor) -> (R:float=~, G:float=~, B:float=~, A:float=~) {is=title K2Node_BreakStruct}
   > SetBrushColor(
    self=Get CommonBorder_ContainerBG() -> (CommonBorder_ContainerBG:Common Border Object Reference)
    InBrushColor=Make Linear Color!(R:float=@n5, G:float=@n5.G, B:float=@n5.B, A:float=0.5) -> (LinearColor:Linear Color Structure) {is=title K2Node_MakeStruct}
   )
   > SetBrushColor(self=Get CommonBorder_OptionsBarBG() -> (CommonBorder_OptionsBarBG:Common Border Object Reference), InBrushColor=@n4.BackgroundLeftColor)
   n6 = Init!() -> (OutputDelegate, then, EventWindowId:String, TournamentId:String) {is=title K2Node_CustomEvent}
   > SetTournament!(
    execute
    self:Fort Showdown Detail View Object Reference=Get EventLeaderboardEventDetails() -> (EventLeaderboardEventDetails:Event Leaderboard Event Details Object Reference)
    TournamentId:String=@n6.TournamentId
    EventId:String=@n6.EventWindowId
    bForceRefreshUI:bool=true
   ) -> (then)
   > EventIntro()
   > SetTournament(
    self=Get EventLeaderboardHeaderDetails() -> (EventLeaderboardHeaderDetails:Event Leaderboard Header Object Reference)
    TournamentId=@n6.TournamentId
    EventId=@n6.EventWindowId
    bForceRefreshUI=true
   )
   > SetScrollOffset!(
    execute
    self:Scroll Box Object Reference=Get CommonHierarchicalScrollBox_0() -> (CommonHierarchicalScrollBox_0:Common Hierarchical Scroll Box Object Reference)
    NewScrollOffset:float=0
   ) -> (then)
   Event NotifyNoPayouts() {is=title K2Node_Event}
   > SetActiveWidgetIndex(self=Get CommonWidgetSwitcher_476() -> (CommonWidgetSwitcher_476:Common Widget Switcher Legacy Object Reference), Index=1)
   > Call No Payout!(execute:Exec, self:Tournament Payout Modal Object Reference) -> (then:Exec) {is=title K2Node_CallDelegate}
   n7 = Event OnTabVisibilityChange!() -> (OutputDelegate, then, bIsVisible:bool) {is=title K2Node_Event}
   > SetVisibility(
    self=Get HorizontalBox_0() -> (HorizontalBox_0:Horizontal Box Object Reference)
    InVisibility=Select(Option 0=Collapsed, Option 1=Visible, Index=@n7.bIsVisible)
   )
   n8 = Event InitializeTabButton!() -> (OutputDelegate, then, TabButton:Common Button Legacy Object Reference, TabText:"Text (by ref)") {is=title K2Node_Event}
   > n9 = Cast To EventsSimpleTab!(execute, Object=@n8.TabButton) -> (then, CastFailed, AsEvents Simple Tab:Events Simple Tab Object Reference) {is=title K2Node_DynamicCast}
   > Set ButtonNameText!(execute, ButtonNameText:Text=@n8.TabText, self:Events Simple Tab Object Reference=@n9) -> (then, Output_Get:Text)
   n10 = Get!() -> (
    CurrentDisplayInfo_TitleLine1:Text
    CurrentDisplayInfo_TitleLine2:Text
    CurrentDisplayInfo_ScheduleInfo:Text
    CurrentDisplayInfo_FlavorDescription:Text
    CurrentDisplayInfo_DetailsDescription:Text
    CurrentDisplayInfo_ShortFormatTitle:Text
    CurrentDisplayInfo_LongFormatTitle:Text
    CurrentDisplayInfo_BackgroundTitle:Text
    CurrentDisplayInfo_PinScoreRequirement:int
    CurrentDisplayInfo_PinEarnedText:Text
    CurrentDisplayInfo_BaseColor:Linear Color Structure
    CurrentDisplayInfo_PrimaryColor:Linear Color Structure
    CurrentDisplayInfo_SecondaryColor:Linear Color Structure
    CurrentDisplayInfo_HighlightColor:Linear Color Structure
    CurrentDisplayInfo_TitleColor:Linear Color Structure
    CurrentDisplayInfo_ShadowColor:Linear Color Structure
    CurrentDisplayInfo_BackgroundLeftColor:Linear Color Structure
    CurrentDisplayInfo_BackgroundRightColor:Linear Color Structure
    CurrentDisplayInfo_BackgroundTextColor:Linear Color Structure
    CurrentDisplayInfo_PosterFadeColor:Linear Color Structure
    CurrentDisplayInfo_AlertText:Text
    CurrentDisplayInfo_SeriesPointLeaderboardName:Text
    CurrentDisplayInfo_AlertType:EFortTournamentAlertType Enum
    CurrentDisplayInfo_RoundDisplayNames:Array of Texts
   )
   > UpdateColor!(
    execute
    self:Events Simple Tab Object Reference=@n9
    ParameterName:String=ButtonColor1
    New Color:Linear Color Structure=@n10.CurrentDisplayInfo_PrimaryColor
   ) -> (then)
   > UpdateColor(self=@n9, ParameterName=ButtonColor2, New Color=@n10.CurrentDisplayInfo_PrimaryColor)
   > UpdateColor(self=@n9, ParameterName=TextColor, New Color=@n10.CurrentDisplayInfo_ShadowColor)
   EventIntro() {is=title K2Node_CustomEvent}
  ]}
 NoPayout {nodes [NoPayout() {is=title K2Node_FunctionEntry}]}
]
variables [
 CallEventBack {}
 CurrentDisplayInfo {
  type=FortTournamentDisplayInfo
  default {
   alertText=""
   alertType=Warning
   backgroundTitle=""
   detailsDescription=""
   flavorDescription=""
   longFormatTitle=""
   pinEarnedText=""
   pinScoreRequirement=0
   scheduleInfo=""
   seriesPointLeaderboardName=""
   shortFormatTitle=""
   titleLine1=""
   titleLine2=""
   backgroundLeftColor=<rgba 0 0 0 1>
   backgroundRightColor=<rgba 0 0 0 1>
   backgroundTextColor=<rgba 0 0 0 1>
   baseColor=<rgba 0 0 0 1>
   highlightColor=<rgba 0 0 0 1>
   posterFadeColor=<rgba 0 0 0 1>
   primaryColor=<rgba 0 0 0 1>
   roundDisplayNames []
   secondaryColor=<rgba 0 0 0 1>
   shadowColor=<rgba 0 0 0 1>
   titleColor=<rgba 0 0 0 1>
  }
 }
 NoPayout {}
]
widgets {
 info {inheritedWidgetCount=7, parentClass=/Script/FortniteUI.FortTournamentPayoutModal, rootWidgetClass=&SafeZone, widgetCount=37}
 widgets [
  SafeZone_0 {widgetClassPath=&SafeZone, bIsVariable=true, widget=SafeZone_0}
  Overlay_2 {widgetClassPath=&Overlay, parent=SafeZone_0, widget=Overlay_2, slot {class=SafeZoneSlot}}
  SizeBox_TopContainer {
   widgetClassPath=&SizeBox
   bIsVariable=true
   parent=Overlay_2
   widget=SizeBox_TopContainer
   properties {heightOverride=814, visibility=Collapsed, widthOverride=1743}
   slot {class=OverlaySlot}
  }
  Overlay_1 {widgetClassPath=&Overlay, parent=SizeBox_TopContainer, widget=Overlay_1, slot {class=SizeBoxSlot}}
  CommonWidgetSwitcher_476 {
   widgetClassPath=&CommonWidgetSwitcherLegacy
   bIsVariable=true
   parent=Overlay_1
   widget=CommonWidgetSwitcher_476
   slot {class=OverlaySlot, properties {horizontalAlignment=HAlign_Fill, verticalAlignment=VAlign_Fill}}
  }
  Overlay_3 {widgetClassPath=&Overlay, parent=CommonWidgetSwitcher_476, widget=Overlay_3, slot {class=WidgetSwitcherSlot}}
  VerticalBox_2 {
   widgetClassPath=&VerticalBox
   parent=Overlay_3
   widget=VerticalBox_2
   slot {class=OverlaySlot, properties {horizontalAlignment=HAlign_Fill, verticalAlignment=VAlign_Fill, padding=<ltrb 20 20 20 20>}}
  }
  Text_TournamentName {
   widgetClassPath=&CommonTextBlock
   bInherited=true
   bIsVariable=true
   parent=VerticalBox_2
   widget=Text_TournamentName
   properties {
    justification=Center
    text=Tournament Name
    textTransformPolicy=ToUpper
    font {
     bForceMonospaced=false
     bMaterialIsStencil=false
     fontMaterial=null
     fontObject=&Burbank
     letterSpacing=0
     monospacedWidth=1
     size=60
     skewAmount=0
     typefaceFontName=Burbank-Black
     outlineSettings {
      bApplyOutlineToDropShadows=false
      bMiteredCorners=false
      bSeparateFillAlpha=false
      outlineMaterial=null
      outlineSize=0
      outlineColor=<rgba 0 0 0 1>
     }
    }
    shadowColorAndOpacity=<rgba 0.0273 0.0012 0.1221 1>
    shadowOffset=<8 8>
   }
   slot {class=VerticalBoxSlot, properties {horizontalAlignment=HAlign_Center, padding=<ltrb 0 0 0 -25>}}
  }
  Text_SessionName {
   widgetClassPath=&CommonTextBlock
   bInherited=true
   bIsVariable=true
   parent=VerticalBox_2
   widget=Text_SessionName
   properties {
    justification=Center
    text=Session Name
    textTransformPolicy=ToUpper
    font {
     bForceMonospaced=false
     bMaterialIsStencil=false
     fontMaterial=null
     fontObject=&Burbank
     letterSpacing=0
     monospacedWidth=1
     size=25
     skewAmount=0
     typefaceFontName=Burbank-Black
     outlineSettings {
      bApplyOutlineToDropShadows=false
      bMiteredCorners=false
      bSeparateFillAlpha=false
      outlineMaterial=null
      outlineSize=0
      outlineColor=<rgba 0 0 0 1>
     }
    }
    shadowColorAndOpacity=<rgba 0.0273 0.0012 0.1221 1>
    shadowOffset=<8 8>
   }
   slot {class=VerticalBoxSlot, properties {horizontalAlignment=HAlign_Center}}
  }
  VerticalBox_68 {
   widgetClassPath=&VerticalBox
   parent=Overlay_3
   widget=VerticalBox_68
   slot {class=OverlaySlot, properties {horizontalAlignment=HAlign_Center, padding=<ltrb 0 100 0 0>}}
  }
  Text_ThresholdTypeLabel {
   widgetClassPath=&CommonTextBlock
   bInherited=true
   bIsVariable=true
   parent=VerticalBox_68
   widget=Text_ThresholdTypeLabel
   properties {
    justification=Center
    text="Prizes && Rewards"
    textTransformPolicy=ToUpper
    font {
     bForceMonospaced=false
     bMaterialIsStencil=false
     fontMaterial=null
     fontObject=&Burbank
     letterSpacing=0
     monospacedWidth=1
     size=60
     skewAmount=0
     typefaceFontName=Burbank-Black
     outlineSettings {
      bApplyOutlineToDropShadows=false
      bMiteredCorners=false
      bSeparateFillAlpha=false
      outlineMaterial=null
      outlineSize=0
      outlineColor=<rgba 0 0 0 1>
     }
    }
    shadowColorAndOpacity=<rgba 0.0273 0.0012 0.1221 1>
    shadowOffset=<8 8>
   }
   slot {class=VerticalBoxSlot, properties {horizontalAlignment=HAlign_Left}}
  }
  Overlay_330 {widgetClassPath=&Overlay, parent=CommonWidgetSwitcher_476, widget=Overlay_330, slot {class=WidgetSwitcherSlot}}
  CommonTextBlock_1 {
   widgetClassPath=&CommonTextBlock
   bIsVariable=true
   parent=Overlay_330
   widget=CommonTextBlock_1
   properties {
    justification=Center
    text=There are no payouts for this tournament
    textTransformPolicy=ToUpper
    font {
     bForceMonospaced=false
     bMaterialIsStencil=false
     fontMaterial=null
     fontObject=&Burbank
     letterSpacing=0
     monospacedWidth=1
     size=60
     skewAmount=0
     typefaceFontName=Burbank-Black
     outlineSettings {
      bApplyOutlineToDropShadows=false
      bMiteredCorners=false
      bSeparateFillAlpha=false
      outlineMaterial=null
      outlineSize=0
      outlineColor=<rgba 0 0 0 1>
     }
    }
    shadowColorAndOpacity=<rgba 0.0273 0.0012 0.1221 1>
    shadowOffset=<8 8>
   }
   slot {class=OverlaySlot, properties {horizontalAlignment=HAlign_Center, verticalAlignment=VAlign_Center}}
  }
  VerticalBox_0 {
   widgetClassPath=&VerticalBox
   parent=Overlay_2
   widget=VerticalBox_0
   slot {class=OverlaySlot, properties {horizontalAlignment=HAlign_Fill, verticalAlignment=VAlign_Fill, padding=<ltrb 0 0 50 0>}}
  }
  EventLeaderboardEventDetails {
   widgetClassPath=/Game/EventLeaderboardEventDetails
   bIsVariable=true
   parent=VerticalBox_0
   widget=EventLeaderboardEventDetails
   slot {class=VerticalBoxSlot, properties {padding=<ltrb 40 35 -200 0>}}
  }
  SizeBox_1 {
   widgetClassPath=&SizeBox
   parent=VerticalBox_0
   widget=SizeBox_1
   slot {
    class=VerticalBoxSlot
    properties {horizontalAlignment=HAlign_Center, verticalAlignment=VAlign_Center, padding=<ltrb 0 20 450 0>, size {sizeRule=Automatic, value=0.1}}
   }
  }
  HorizontalBox_0 {widgetClassPath=&HorizontalBox, bIsVariable=true, parent=SizeBox_1, widget=HorizontalBox_0, slot {class=SizeBoxSlot}}
  ActionWidget_TabLeft {
   widgetClassPath=&CommonActionWidget
   bInherited=true
   bIsVariable=true
   parent=HorizontalBox_0
   widget=ActionWidget_TabLeft
   properties {inputActions [{dataTable=&InputActionDataTable, rowName=SecondaryTabLeft}]}
   slot {class=HorizontalBoxSlot, properties {padding=<ltrb 0 0 8 0>}}
  }
  EntryBox_ViewButtons {
   widgetClassPath=&DynamicEntryBox
   bInherited=true
   bIsVariable=true
   parent=HorizontalBox_0
   widget=EntryBox_ViewButtons
   properties {entryWidgetClass=/CompeteOldTab/Frontend/Showdown/EventsSimpleTab.EventsSimpleTab_C, numDesignerPreviewEntries=2, visibility=Visible}
   slot {class=HorizontalBoxSlot}
  }
  ActionWidget_TabRight {
   widgetClassPath=&CommonActionWidget
   bInherited=true
   bIsVariable=true
   parent=HorizontalBox_0
   widget=ActionWidget_TabRight
   properties {inputActions [{dataTable=&InputActionDataTable, rowName=SecondaryTabRight}]}
   slot {class=HorizontalBoxSlot, properties {padding=<ltrb 8 0 0 0>}}
  }
  SizeBox_0 {
   widgetClassPath=&SizeBox
   parent=VerticalBox_0
   widget=SizeBox_0
   properties {bOverride_WidthOverride=true, widthOverride=949.396}
   slot {class=VerticalBoxSlot, properties {horizontalAlignment=HAlign_Left, padding=<ltrb 0 40 0 60>, size {sizeRule=Fill, value=1}}}
  }
  Overlay_LeaderboardContainer {widgetClassPath=&Overlay, parent=SizeBox_0, widget=Overlay_LeaderboardContainer, slot {class=SizeBoxSlot}}
  BackgroundBlur_1 {
   widgetClassPath=&BackgroundBlur
   parent=Overlay_LeaderboardContainer
   widget=BackgroundBlur_1
   properties {blurStrength=50}
   slot {class=OverlaySlot, properties {horizontalAlignment=HAlign_Fill, padding=<ltrb 50 40 20 0>}}
  }
  CommonBorder_ContainerBG {
   widgetClassPath=&CommonBorder
   bIsVariable=true
   parent=BackgroundBlur_1
   widget=CommonBorder_ContainerBG
   properties {
    background {
     bIsDynamicallyLoaded=false
     drawAs=Image
     imageType=NoImage
     mirroring=NoMirror
     resourceName=None
     resourceObject=null
     tiling=NoTile
     imageSize=<32 32>
     margin=<ltrb 0 0 0 0>
     outlineSettings {
      bUseBrushTransparency=false
      roundingType=HalfHeightRadius
      width=0
      color {colorUseRule=UseColor_Specified, specifiedColor=<rgba 0 0 0 0>}
      cornerRadii=<0 0 0 0>
     }
     tintColor {colorUseRule=UseColor_Specified, specifiedColor=<rgba 1 1 1 1>}
     uVRegion {bIsValid=false, max=<0 0>, min=<0 0>}
    }
    brushColor=<rgba 0.0052 0.0723 0.4179 0.5>
    padding=<ltrb 0 20 0 0>
   }
   slot {class=BackgroundBlurSlot, properties {padding=<ltrb 0 0 0 0>}}
  }
  Switcher_LeaderboardEntries {
   widgetClassPath=&CommonWidgetSwitcherLegacy
   bIsVariable=true
   parent=CommonBorder_ContainerBG
   widget=Switcher_LeaderboardEntries
   slot {class=BorderSlot, properties {padding=<ltrb 0 20 0 0>}}
  }
  Content_LeaderboardEntries {
   widgetClassPath=&VerticalBox
   bIsVariable=true
   parent=Switcher_LeaderboardEntries
   widget=Content_LeaderboardEntries
   properties {clipping=ClipToBounds}
   slot {class=WidgetSwitcherSlot}
  }
  SizeBox_55 {
   widgetClassPath=&SizeBox
   parent=Content_LeaderboardEntries
   widget=SizeBox_55
   properties {heightOverride=800}
   slot {class=VerticalBoxSlot, properties {size {sizeRule=Fill, value=1}}}
  }
  CommonHierarchicalScrollBox_0 {
   widgetClassPath=/Script/CommonUI.CommonHierarchicalScrollBox
   bIsVariable=true
   parent=SizeBox_55
   widget=CommonHierarchicalScrollBox_0
   properties {
    navigationDestination=Center
    widgetStyle {
     barThickness=9
     bottomShadowBrush {
      bIsDynamicallyLoaded=false
      drawAs=NoDrawType
      imageType=FullColor
      mirroring=NoMirror
      resourceName=../../../Engine/Content/Slate/Common/ScrollBoxShadowBottom.png
      resourceObject=null
      tiling=NoTile
      imageSize=<16 8>
      margin=<ltrb 0.5 0 0.5 1>
      outlineSettings {
       bUseBrushTransparency=false
       roundingType=HalfHeightRadius
       width=0
       color {colorUseRule=UseColor_Specified, specifiedColor=<rgba 0 0 0 0>}
       cornerRadii=<0 0 0 0>
      }
      tintColor {colorUseRule=UseColor_Specified, specifiedColor=<rgba 1 1 1 1>}
      uVRegion {bIsValid=false, max=<0 0>, min=<0 0>}
     }
     horizontalScrolledContentPadding=<ltrb 0 0 1 0>
     leftShadowBrush {
      bIsDynamicallyLoaded=false
      drawAs=Box
      imageType=FullColor
      mirroring=NoMirror
      resourceName=../../../Engine/Content/Slate/Common/ScrollBoxShadowLeft.png
      resourceObject=null
      tiling=NoTile
      imageSize=<8 16>
      margin=<ltrb 1 0.5 0 0.5>
      outlineSettings {
       bUseBrushTransparency=false
       roundingType=HalfHeightRadius
       width=0
       color {colorUseRule=UseColor_Specified, specifiedColor=<rgba 0 0 0 0>}
       cornerRadii=<0 0 0 0>
      }
      tintColor {colorUseRule=UseColor_Specified, specifiedColor=<rgba 1 1 1 1>}
      uVRegion {bIsValid=false, max=<0 0>, min=<0 0>}
     }
     rightShadowBrush {
      bIsDynamicallyLoaded=false
      drawAs=Box
      imageType=FullColor
      mirroring=NoMirror
      resourceName=../../../Engine/Content/Slate/Common/ScrollBoxShadowRight.png
      resourceObject=null
      tiling=NoTile
      imageSize=<8 16>
      margin=<ltrb 0 0.5 1 0.5>
      outlineSettings {
       bUseBrushTransparency=false
       roundingType=HalfHeightRadius
       width=0
       color {colorUseRule=UseColor_Specified, specifiedColor=<rgba 0 0 0 0>}
       cornerRadii=<0 0 0 0>
      }
      tintColor {colorUseRule=UseColor_Specified, specifiedColor=<rgba 1 1 1 1>}
      uVRegion {bIsValid=false, max=<0 0>, min=<0 0>}
     }
     topShadowBrush {
      bIsDynamicallyLoaded=false
      drawAs=NoDrawType
      imageType=FullColor
      mirroring=NoMirror
      resourceName=../../../Engine/Content/Slate/Common/ScrollBoxShadowTop.png
      resourceObject=null
      tiling=NoTile
      imageSize=<16 8>
      margin=<ltrb 0.5 1 0.5 0>
      outlineSettings {
       bUseBrushTransparency=false
       roundingType=HalfHeightRadius
       width=0
       color {colorUseRule=UseColor_Specified, specifiedColor=<rgba 0 0 0 0>}
       cornerRadii=<0 0 0 0>
      }
      tintColor {colorUseRule=UseColor_Specified, specifiedColor=<rgba 1 1 1 1>}
      uVRegion {bIsValid=false, max=<0 0>, min=<0 0>}
     }
     verticalScrolledContentPadding=<ltrb 0 0 0 1>
    }
   }
   slot {class=SizeBoxSlot}
  }
  DynamicEntryBox_PayoutThresholds {
   widgetClassPath=&DynamicEntryBox
   bInherited=true
   bIsVariable=true
   parent=CommonHierarchicalScrollBox_0
   widget=DynamicEntryBox_PayoutThresholds
   properties {
    entryBoxType=Vertical
    entryWidgetClass=/Game/UI/Events/TournamentPayoutThresholdEntry.TournamentPayoutThresholdEntry_C
    numDesignerPreviewEntries=14
    visibility=Visible
   }
   slot {class=ScrollBoxSlot}
  }
  CommonBorder_OptionsBarBG {
   widgetClassPath=&CommonBorder
   bIsVariable=true
   parent=Content_LeaderboardEntries
   widget=CommonBorder_OptionsBarBG
   properties {
    horizontalAlignment=HAlign_Center
    background {
     bIsDynamicallyLoaded=false
     drawAs=Image
     imageType=NoImage
     mirroring=NoMirror
     resourceName=None
     resourceObject=null
     tiling=NoTile
     imageSize=<32 32>
     margin=<ltrb 0 0 0 0>
     outlineSettings {
      bUseBrushTransparency=false
      roundingType=HalfHeightRadius
      width=0
      color {colorUseRule=UseColor_Specified, specifiedColor=<rgba 0 0 0 0>}
      cornerRadii=<0 0 0 0>
     }
     tintColor {colorUseRule=UseColor_Specified, specifiedColor=<rgba 1 1 1 1>}
     uVRegion {bIsValid=false, max=<0 0>, min=<0 0>}
    }
    brushColor=<rgba 0.006 0.0742 0.4969 1>
    padding=<ltrb 0 0 0 0>
   }
   slot {class=VerticalBoxSlot}
  }
  Content_NoLeaderboardEntries {
   widgetClassPath=&CommonBorder
   bIsVariable=true
   parent=Switcher_LeaderboardEntries
   widget=Content_NoLeaderboardEntries
   properties {padding=<ltrb 10 40 10 25>}
   slot {class=WidgetSwitcherSlot}
  }
  Text_NoLeaderboardEntries {
   widgetClassPath=&CommonTextBlock
   bIsVariable=true
   parent=Content_NoLeaderboardEntries
   widget=Text_NoLeaderboardEntries
   properties {
    autoWrapText=true
    justification=Center
    text="No Leaderboard Entries Available\r\nCheck Back Later"
    font {
     bForceMonospaced=false
     bMaterialIsStencil=false
     fontMaterial=null
     fontObject=&Burbank
     letterSpacing=0
     monospacedWidth=1
     size=24
     skewAmount=0
     typefaceFontName=Burbank-PlayerPower
     outlineSettings {
      bApplyOutlineToDropShadows=false
      bMiteredCorners=false
      bSeparateFillAlpha=false
      outlineMaterial=null
      outlineSize=0
      outlineColor=<rgba 0 0 0 1>
     }
    }
    shadowOffset=<0 0>
   }
   slot {class=BorderSlot, properties {padding=<ltrb 10 40 10 25>}}
  }
  EventLeaderboardHeaderDetails {
   widgetClassPath=/Game/EventLeaderboardHeader
   bIsVariable=true
   parent=Overlay_LeaderboardContainer
   widget=EventLeaderboardHeaderDetails
   slot {class=OverlaySlot, properties {padding=<ltrb 20 0 0 0>}}
  }
  VisibilityBorder_Touch {
   widgetClassPath=&CommonHardwareVisibilityBorder
   parent=Overlay_2
   widget=VisibilityBorder_Touch
   properties {visibilityQuery {
     autoDescription=" ANY( Input.Touch )"
     tokenStreamVersion=0
     userDescription=""
     queryTokenStream [0,1,1,1,0]
     tagDictionary [{tagName=Input.Touch}]
    }}
   slot {class=OverlaySlot, properties {horizontalAlignment=HAlign_Right, padding=<ltrb 0 25 50 20>}}
  }
  CloseButton {
   widgetClassPath=/FortUILibrary/Foundation/Buttons/CloseButton
   bIsVariable=true
   parent=VisibilityBorder_Touch
   widget=CloseButton
   properties {
    onButtonBaseClicked=(&self:WidgetTree.CloseButton.HandleOnButtonClicked)
    onButtonBaseDoubleClicked=(&self:WidgetTree.CloseButton.HandleOnButtonDoubleClicked)
    onButtonBaseHovered=(&self:WidgetTree.CloseButton.HandleOnButtonHovered)
    onButtonBaseUnhovered=(&self:WidgetTree.CloseButton.HandleOnButtonUnhovered)
    onSelectedChangedBase=(&self:WidgetTree.CloseButton.HandleOnSelectedChanged)
   }
   slot {class=BorderSlot, properties {padding=<ltrb 0 0 0 0>}}
  }
  VisibilityBorder_NonTouch {
   widgetClassPath=&CommonHardwareVisibilityBorder
   parent=Overlay_2
   widget=VisibilityBorder_NonTouch
   properties {visibilityQuery {
     autoDescription=" NONE( Input.Touch )"
     tokenStreamVersion=0
     userDescription=""
     queryTokenStream [0,1,3,1,0]
     tagDictionary [{tagName=Input.Touch}]
    }}
   slot {class=OverlaySlot, properties {horizontalAlignment=HAlign_Right, verticalAlignment=VAlign_Bottom, padding=<ltrb 20 20 20 20>}}
  }
  BackButton {
   widgetClassPath=/CompeteOldTab/Frontend/Showdown/EventsSimpleButton
   bIsVariable=true
   parent=VisibilityBorder_NonTouch
   widget=BackButton
   properties {
    buttonNameText=Back
    onButtonBaseClicked=(&self:WidgetTree.BackButton.HandleOnButtonClicked)
    onButtonBaseDoubleClicked=(&self:WidgetTree.BackButton.HandleOnButtonDoubleClicked)
    onButtonBaseHovered=(&self:WidgetTree.BackButton.HandleOnButtonHovered)
    onButtonBaseUnhovered=(&self:WidgetTree.BackButton.HandleOnButtonUnhovered)
    onSelectedChangedBase=(&self:WidgetTree.BackButton.HandleOnSelectedChanged)
    triggeringInputAction {dataTable=&InputActionDataTable, rowName=GenericBack}
   }
   slot {class=BorderSlot, properties {padding=<ltrb 0 0 0 0>}}
  }
 ]
}
