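# Translucent UI material for the 64 external-link icon in the login UI art of CompeteOldTab.
# The red channel of the icon texture serves as a gradient for the shared AAThreshold function, which by its name gives an antialiased cutoff.
# The IconThreshold scalar parameter sets the cutoff, and a constant white vector accompanies the resulting mask.
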Material /CompeteOldTab/UI/Login/Art/M-Icon-ExternalLink-64
graph {
 properties {blendMode=BLEND_Translucent, materialDomain=MD_UI, pixelDepthOffsetMode=PDOM_Legacy, usedShadingModels=Default Lit}
 nodes [
  Constant3Vector() {constant=<rgba 1 1 1 0>}
  MaterialFunctionCall!(
   Threshold=ScalarParameter() {defaultValue=0.19, parameterName=IconThreshold}
   Gradient=TextureSample() {samplerSource=SSM_Clamp_WorldGroupSettings, texture=&/T-Icon-ExternalLink-64}.R
   Filter Width Override
   Filter Width Fudge Factor
  ) {materialFunction=/Game/UI/Materials/AAThreshold}
 ]
}
