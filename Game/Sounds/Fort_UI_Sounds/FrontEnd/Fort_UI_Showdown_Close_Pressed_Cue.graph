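# UI sound cue for pressing Close in the front-end Showdown screen, routed to the Fort_UI sound class.
# A Mixer layers a Squads tab pop-up wave at full level over a quieter back sound that a Modulator pitches up.
# An Enveloper fades the mix out over its curve, and the whole cue is further pitched up and attenuated.
# Overridden concurrency allows one instance and blocks retriggers while it plays.
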SoundCue /Game/Sounds/Fort_UI_Sounds/FrontEnd/Fort_UI_Showdown_Close_Pressed_Cue
root=Enveloper_0
nodes [
 {
  id=Enveloper_0
  children [Mixer_1]
  properties {
   pitchCurve {editorCurveData {keys [{time=0}]}}
   volumeCurve {editorCurveData {keys [{interpMode=RCIM_Cubic, tangentMode=RCTM_Break, time=0, value=1}, <at 1 0>]}}
  }
 }
 {id=Mixer_1, children [Modulator_1,WavePlayer_0], properties {inputVolume [0.3,1]}}
 {id=Modulator_1, children [WavePlayer_5], properties {pitchMax=1.5, pitchMin=1.5}}
 {id=WavePlayer_0, properties {soundWaveAssetPtr=/Game/Sounds/Fort_UI_Sounds/FrontEnd/New_FrontEnd/UI_FN_Squads_Tab_PopUps_01}}
 {id=WavePlayer_5, properties {soundWaveAssetPtr=&/fort_ui_back_01}}
]
properties {
 bOverrideConcurrency=true
 maxDistance=2097152
 pitchMultiplier=1.5
 soundClassObject=&Fort_UI
 volumeMultiplier=0.55
 concurrencyOverrides {maxCount=1, resolutionRule=PreventNew}
}
