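# Front-end UI sound cue, named for the Showdown details screen, in the Fort_UI sound class.
# It plays the Card_Summary wave from the Store sound folder through a Modulator that randomizes pitch slightly upward, at reduced cue volume.
# A concurrency override allows only one instance at a time, stopping the oldest when a new one starts.
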SoundCue /Game/Sounds/Fort_UI_Sounds/FrontEnd/Fort_UI_Showdown_Details_Cue
root=Modulator_0
nodes [
 {id=Modulator_0, children [WavePlayer_13], properties {pitchMax=1.2, pitchMin=1.1, volumeMin=1}}
 {id=WavePlayer_13, properties {soundWaveAssetPtr=/Game/Sounds/Fort_UI_Sounds/FrontEnd/Store/wav/Card_Summary}}
]
properties {
 bOverrideConcurrency=true
 maxDistance=2097152
 soundClassObject=&Fort_UI
 volumeMultiplier=0.4
 concurrencyOverrides {maxCount=1, resolutionRule=StopOldest}
}
