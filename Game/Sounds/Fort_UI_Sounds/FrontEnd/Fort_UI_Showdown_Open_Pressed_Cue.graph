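# Sound cue for the front-end UI press that opens Showdown, in the Fort_UI sound class.
# A single wave player reuses a Squads tab pop-up sound, fed through a modulator that applies a fixed raised pitch.
# A concurrency override allows one instance at a time and blocks new plays while it is sounding, with a slight volume boost on the cue.
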SoundCue /Game/Sounds/Fort_UI_Sounds/FrontEnd/Fort_UI_Showdown_Open_Pressed_Cue
root=Modulator_0
nodes [
 {id=Modulator_0, children [WavePlayer_0], properties {pitchMax=1.1, pitchMin=1.1, volumeMax=1, volumeMin=1}}
 {id=WavePlayer_0, properties {soundWaveAssetPtr=/Game/Sounds/Fort_UI_Sounds/FrontEnd/New_FrontEnd/UI_FN_Squads_Tab_PopUps_01}}
]
properties {
 bOverrideConcurrency=true
 maxDistance=2097152
 soundClassObject=&Fort_UI
 volumeMultiplier=1.2
 concurrencyOverrides {maxCount=1, resolutionRule=PreventNew}
}
