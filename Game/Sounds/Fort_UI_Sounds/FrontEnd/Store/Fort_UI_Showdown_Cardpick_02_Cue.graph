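# Sound cue for a card pick sound in the front-end store Showdown UI, routed through the Fort_UI sound class at reduced volume.
# A single wave player plays the CardPutAway_02 wav into a modulator that randomizes pitch well above normal on each play.
# A concurrency override prevents new instances from starting once the limit is reached.
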SoundCue /Game/Sounds/Fort_UI_Sounds/FrontEnd/Store/Fort_UI_Showdown_Cardpick_02_Cue
root=Modulator_0
nodes [
 {id=Modulator_0, children [WavePlayer_12], properties {pitchMax=1.6, pitchMin=1.5}}
 {id=WavePlayer_12, properties {soundWaveAssetPtr=/Game/Sounds/Fort_UI_Sounds/FrontEnd/Store/wav/CardPutAway_02}}
]
properties {
 bOverrideConcurrency=true
 maxDistance=2097152
 soundClassObject=&Fort_UI
 volumeMultiplier=0.5
 concurrencyOverrides {resolutionRule=PreventNew}
}
