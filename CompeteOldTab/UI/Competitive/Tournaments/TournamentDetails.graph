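# Tournament details screen widget for the competitive UI, showing poster, rules, event windows, leaderboard, payouts and live games.
# Selecting an event tile feeds its IDs to the window details and payout widgets, then centers it.
# Child-widget delegates switch a full-screen switcher between details, leaderboard, payout and watch list.
# AnimTransition resumes or reverses poster, blur and panel animations mid-play; scoring rules fill a two-column grid.
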WidgetBlueprint /CompeteOldTab/UI/Competitive/Tournaments/TournamentDetails
animations [
 PosterFlip {
  binding_count=3
  display_label=""
  length_seconds=0.4635
  object=&self:PosterFlip
  sequence {bindings [
    CommonWidgetSwitcher_PosterSides {
     id=#1
     display_name=OverlayPosterSides
     possessed_class=CommonWidgetSwitcherLegacy
     tracks [{
       type=MovieScene2DTransformTrack
       sections [{
         type=MovieScene2DTransformSection
         end_frame=9
         start_frame=0
         channels [
          Translation.X {type=MovieSceneScriptingFloatChannel, const [-25.0851,1]}
          Translation.Y {type=MovieSceneScriptingFloatChannel, keys [[4.366,0],[8.0443,-90]]}
          Angle {type=MovieSceneScriptingFloatChannel}
          Scale.X {type=MovieSceneScriptingFloatChannel, keys [[0,1],[4.0803,0],[4.556,0],[8.0443,0.95]]}
          Scale.Y {type=MovieSceneScriptingFloatChannel, keys [[0,1],[4.0803,0.9898],[4.556,0.9898],[8.0443,0.95]]}
          Shear.X {type=MovieSceneScriptingFloatChannel, const [0,4]}
          Shear.Y {type=MovieSceneScriptingFloatChannel, keys [[0,0],[4.0803,-8.7946],[4.556,6.9265],[8.0443,0]]}
         ]
        }]
      }]
    }
    TournamentPosterBack {
     id=#2
     possessed_class=ShowdownTournamentPosterBack_C
     tracks [{
       type=MovieSceneFloatTrack
       display_name=Render Opacity
       sections [{
         type=MovieSceneFloatSection
         end_frame=9
         start_frame=0
         channels [None {type=MovieSceneScriptingFloatChannel, keys [[0,0],[4.2883,1],[8.0443,1]]}]
        }]
      }]
    }
    Image_PosterFade {
     id=#3
     possessed_class=Image
     tracks [{
       type=MovieSceneWidgetMaterialTrack
       sections [{
         type=MovieSceneParameterSection
         end_frame=9
         start_frame=0
         channels [
          CenterR {type=MovieSceneScriptingFloatChannel, keys [[0,0.9],[9.2697,1.0897]]}
          CenterG {type=MovieSceneScriptingFloatChannel, keys [[0,1.08],[9.2697,1.7384]]}
          CenterB {type=MovieSceneScriptingFloatChannel}
          CenterA {type=MovieSceneScriptingFloatChannel}
         ]
        }]
      }]
    }
   ]}
 }
 ShowDetails {
  binding_count=2
  display_label=""
  length_seconds=0.4444
  object=&self:ShowDetails
  sequence {bindings [
    CommonBorder_TournamentTitle {
     id=#4
     possessed_class=CommonBorder
     tracks [
      {
       type=MovieSceneFloatVectorTrack
       display_name=Desired Size Scale
       sections [{
         type=MovieSceneFloatVectorSection
         end_frame=8
         start_frame=0
         channels [
          Vector.X {type=MovieSceneScriptingFloatChannel}
          Vector.Y {type=MovieSceneScriptingFloatChannel, keys [[0,0],[8.8883,1]]}
         ]
        }]
      }
      {
       type=MovieScene2DTransformTrack
       sections [{
         type=MovieScene2DTransformSection
         end_frame=8
         start_frame=0
         channels [
          Translation.X {type=MovieSceneScriptingFloatChannel}
          Translation.Y {type=MovieSceneScriptingFloatChannel}
          Angle {type=MovieSceneScriptingFloatChannel}
          Scale.X {type=MovieSceneScriptingFloatChannel}
          Scale.Y {type=MovieSceneScriptingFloatChannel}
          Shear.X {type=MovieSceneScriptingFloatChannel}
          Shear.Y {type=MovieSceneScriptingFloatChannel}
         ]
        }]
      }
     ]
    }
    CommonBorderTournamentDescription {
     id=#5
     possessed_class=CommonBorder
     tracks [{
       type=MovieSceneFloatVectorTrack
       display_name=Desired Size Scale
       sections [{
         type=MovieSceneFloatVectorSection
         end_frame=8
         start_frame=0
         channels [
          Vector.X {type=MovieSceneScriptingFloatChannel}
          Vector.Y {type=MovieSceneScriptingFloatChannel, keys [[0,0],[0.8303,0],[8.8883,1]]}
         ]
        }]
      }]
    }
   ]}
 }
 ShowEventWindowDetails {
  binding_count=3
  display_label=""
  length_seconds=0.45
  object=&self:ShowEventWindowDetails
  sequence {bindings [
    SizeBox_EventWindowContainer {
     id=#6
     possessed_class=SizeBox
     tracks [{
       type=MovieSceneFloatTrack
       display_name=Height Override
       sections [{type=MovieSceneFloatSection, end_frame=9, start_frame=0, channels [None {type=MovieSceneScriptingFloatChannel, keys [[0,215],[9,114]]}]}]
      }]
    }
    CommonBorder_TournamentTitle {
     id=#7
     possessed_class=CommonBorder
     tracks [{
       type=MovieSceneFloatVectorTrack
       display_name=Desired Size Scale
       sections [{
         type=MovieSceneFloatVectorSection
         end_frame=9
         start_frame=0
         channels [
          Vector.X {type=MovieSceneScriptingFloatChannel}
          Vector.Y {type=MovieSceneScriptingFloatChannel, keys [[0,1],[9,0]]}
         ]
        }]
      }]
    }
    CommonBorderEventDetails {
     id=#8
     possessed_class=CommonBorder
     tracks [{
       type=MovieSceneFloatVectorTrack
       display_name=Desired Size Scale
       sections [{
         type=MovieSceneFloatVectorSection
         end_frame=9
         start_frame=0
         channels [
          Vector.X {type=MovieSceneScriptingFloatChannel}
          Vector.Y {type=MovieSceneScriptingFloatChannel, keys [[0,0],[9,1]]}
         ]
        }]
      }]
    }
   ]}
 }
 BlurBG {
  binding_count=2
  display_label=""
  length_seconds=0.7273
  object=&self:BlurBG
  sequence {bindings [
    RetainerBox_BGText {
     id=#9
     possessed_class=RetainerBox
     tracks [{
       type=MovieScene2DTransformTrack
       sections [{
         type=MovieScene2DTransformSection
         end_frame=14
         start_frame=0
         channels [
          Translation.X {type=MovieSceneScriptingFloatChannel}
          Translation.Y {type=MovieSceneScriptingFloatChannel, keys [[0,0],[14.5453,-188.3544]]}
          Angle {type=MovieSceneScriptingFloatChannel}
          Scale.X {type=MovieSceneScriptingFloatChannel}
          Scale.Y {type=MovieSceneScriptingFloatChannel}
          Shear.X {type=MovieSceneScriptingFloatChannel}
          Shear.Y {type=MovieSceneScriptingFloatChannel}
         ]
        }]
      }]
    }
    BackgroundBlur_Background {
     id=#10
     possessed_class=BackgroundBlur
     tracks [
      {
       type=MovieSceneFloatTrack
       display_name=Blur Strength
       sections [{type=MovieSceneFloatSection, end_frame=14, start_frame=0, channels [None {type=MovieSceneScriptingFloatChannel}]}]
      }
      {
       type=MovieSceneFloatTrack
       display_name=Render Opacity
       sections [{type=MovieSceneFloatSection, end_frame=14, start_frame=0, channels [None {type=MovieSceneScriptingFloatChannel, keys [[0,0],[14.5453,1]]}]}]
      }
     ]
    }
   ]}
 }
 BlurPoster {
  binding_count=3
  display_label=""
  length_seconds=0.5352
  object=&self:BlurPoster
  sequence {bindings [
    BackgroundBlur_PosterLayer {
     id=#11
     possessed_class=BackgroundBlur
     tracks [{
       type=MovieSceneFloatTrack
       display_name=Render Opacity
       sections [{type=MovieSceneFloatSection, end_frame=10, start_frame=0, channels [None {type=MovieSceneScriptingFloatChannel, keys [[0,0],[10.7033,1]]}]}]
      }]
    }
    Image_PosterFade {
     id=#12
     possessed_class=Image
     tracks [{
       type=MovieSceneWidgetMaterialTrack
       sections [{
         type=MovieSceneParameterSection
         end_frame=10
         start_frame=0
         channels [
          StretchR {type=MovieSceneScriptingFloatChannel, keys [[0,0.3],[10.704,0.2]]}
          StretchG {type=MovieSceneScriptingFloatChannel, keys [[0,0.5],[10.704,1]]}
          StretchB {type=MovieSceneScriptingFloatChannel, const [0,2]}
          StretchA {type=MovieSceneScriptingFloatChannel, const [1,2]}
         ]
        }]
      }]
    }
    SizeBox_PosterContainer {
     id=#13
     possessed_class=SizeBox
     tracks [{
       type=MovieScene2DTransformTrack
       sections [{
         type=MovieScene2DTransformSection
         end_frame=10
         start_frame=0
         channels [
          Translation.X {type=MovieSceneScriptingFloatChannel}
          Translation.Y {type=MovieSceneScriptingFloatChannel, keys [[0,20],[10.7037,-351.8636]]}
          Angle {type=MovieSceneScriptingFloatChannel}
          Scale.X {type=MovieSceneScriptingFloatChannel}
          Scale.Y {type=MovieSceneScriptingFloatChannel}
          Shear.X {type=MovieSceneScriptingFloatChannel}
          Shear.Y {type=MovieSceneScriptingFloatChannel}
         ]
        }]
      }]
    }
   ]}
 }
 Intro {
  binding_count=7
  display_label=""
  length_seconds=0.5122
  object=&self:Intro
  sequence {
   bindings [
    CommonBorder_BGText {
     id=#14
     possessed_class=CommonBorder
     tracks [
      {
       type=MovieSceneFloatTrack
       display_name=Render Opacity
       sections [{type=MovieSceneFloatSection, channels [None {type=MovieSceneScriptingFloatChannel, keys [[0,0],[30.402,1]]}]}]
      }
      {
       type=MovieScene2DTransformTrack
       sections [{
         type=MovieScene2DTransformSection
         channels [
          Translation.X {type=MovieSceneScriptingFloatChannel}
          Translation.Y {type=MovieSceneScriptingFloatChannel}
          Angle {type=MovieSceneScriptingFloatChannel}
          Scale.X {type=MovieSceneScriptingFloatChannel, keys [[0,0.875],[30.731,1]]}
          Scale.Y {type=MovieSceneScriptingFloatChannel, keys [[0,0.875],[30.731,1]]}
          Shear.X {type=MovieSceneScriptingFloatChannel}
          Shear.Y {type=MovieSceneScriptingFloatChannel}
         ]
        }]
      }
     ]
    }
    TournamentRoundText {
     id=#15
     possessed_class=CommonTextBlock
     tracks [
      {
       type=MovieScene2DTransformTrack
       sections [{
         type=MovieScene2DTransformSection
         channels [
          Translation.X {type=MovieSceneScriptingFloatChannel, keys [[4.063,-486.5504],[18.052,0]]}
          Translation.Y {type=MovieSceneScriptingFloatChannel}
          Angle {type=MovieSceneScriptingFloatChannel}
          Scale.X {type=MovieSceneScriptingFloatChannel}
          Scale.Y {type=MovieSceneScriptingFloatChannel}
          Shear.X {type=MovieSceneScriptingFloatChannel}
          Shear.Y {type=MovieSceneScriptingFloatChannel}
         ]
        }]
      }
      {
       type=MovieSceneFloatTrack
       display_name=Render Opacity
       sections [{type=MovieSceneFloatSection, channels [None {type=MovieSceneScriptingFloatChannel, keys [[4.063,0],[18.053,1]]}]}]
      }
     ]
    }
    VerticalBoxEventTitle {
     id=#16
     possessed_class=VerticalBox
     tracks [
      {
       type=MovieScene2DTransformTrack
       sections [{
         type=MovieScene2DTransformSection
         channels [
          Translation.X {type=MovieSceneScriptingFloatChannel, keys [[8.673,-113.3773],[30,0]]}
          Translation.Y {type=MovieSceneScriptingFloatChannel}
          Angle {type=MovieSceneScriptingFloatChannel}
          Scale.X {type=MovieSceneScriptingFloatChannel}
          Scale.Y {type=MovieSceneScriptingFloatChannel}
          Shear.X {type=MovieSceneScriptingFloatChannel}
          Shear.Y {type=MovieSceneScriptingFloatChannel}
         ]
        }]
      }
      {
       type=MovieSceneFloatTrack
       display_name=Render Opacity
       sections [{type=MovieSceneFloatSection, channels [None {type=MovieSceneScriptingFloatChannel, keys [[8.673,0],[30,1]]}]}]
      }
     ]
    }
    SizeBox_PosterContainer {
     id=#17
     possessed_class=SizeBox
     tracks [{
       type=MovieScene2DTransformTrack
       sections [{
         type=MovieScene2DTransformSection
         channels [
          Translation.X {type=MovieSceneScriptingFloatChannel, keys [[0,-570.9688],[21.279,-20]]}
          Translation.Y {type=MovieSceneScriptingFloatChannel, keys [[0,38],[21.279,20]]}
          Angle {type=MovieSceneScriptingFloatChannel}
          Scale.X {type=MovieSceneScriptingFloatChannel, keys [[0,0.75],[20.13,0.995],[29.027,0.9992]]}
          Scale.Y {type=MovieSceneScriptingFloatChannel, keys [[0,0.75],[20.13,0.995],[29.027,0.9992]]}
          Shear.X {type=MovieSceneScriptingFloatChannel}
          Shear.Y {type=MovieSceneScriptingFloatChannel}
         ]
        }]
      }]
    }
    VerticalBoxEventContent {
     id=#18
     possessed_class=VerticalBox
     tracks [{
       type=MovieSceneFloatTrack
       display_name=Render Opacity
       sections [{type=MovieSceneFloatSection, channels [None {type=MovieSceneScriptingFloatChannel, keys [[0,0],[18.186,1]]}]}]
      }]
    }
    TournamentMatchType {
     id=#19
     possessed_class=CommonTextBlock
     tracks [
      {
       type=MovieScene2DTransformTrack
       sections [{
         type=MovieScene2DTransformSection
         channels [
          Translation.X {type=MovieSceneScriptingFloatChannel, keys [[4.063,-385.3535],[4.849,-75.0502],[18.839,0]]}
          Translation.Y {type=MovieSceneScriptingFloatChannel}
          Angle {type=MovieSceneScriptingFloatChannel}
          Scale.X {type=MovieSceneScriptingFloatChannel}
          Scale.Y {type=MovieSceneScriptingFloatChannel}
          Shear.X {type=MovieSceneScriptingFloatChannel}
          Shear.Y {type=MovieSceneScriptingFloatChannel}
         ]
        }]
      }
      {
       type=MovieSceneFloatTrack
       display_name=Render Opacity
       sections [{type=MovieSceneFloatSection, channels [None {type=MovieSceneScriptingFloatChannel, keys [[4.849,0],[18.839,1]]}]}]
      }
     ]
    }
    Image_PosterFade {
     id=#20
     possessed_class=Image
     tracks [{
       type=MovieSceneWidgetMaterialTrack
       sections [{
         type=MovieSceneParameterSection
         channels [
          CenterR {type=MovieSceneScriptingFloatChannel, keys [[0,1.0116],[21.279,0.9]]}
          CenterG {type=MovieSceneScriptingFloatChannel, keys [[0,2.1493],[21.279,1.08]]}
          CenterB {type=MovieSceneScriptingFloatChannel}
          CenterA {type=MovieSceneScriptingFloatChannel}
         ]
        }]
      }]
    }
   ]
   display_rate {numerator=60}
  }
 }
 FullScreenContent {
  binding_count=1
  length_seconds=0.2346
  object=&self:FullScreenContent
  sequence {
   bindings [OverlayPoster {
     id=#21
     possessed_class=Overlay
     tracks [
      {
       type=MovieSceneFloatTrack
       display_name=Render Opacity
       sections [{type=MovieSceneFloatSection, end_frame=14, start_frame=0, channels [None {type=MovieSceneScriptingFloatChannel, keys [[0,1],[14.076,0]]}]}]
      }
      {
       type=MovieScene2DTransformTrack
       sections [{
         type=MovieScene2DTransformSection
         end_frame=14
         start_frame=0
         channels [
          Translation.X {type=MovieSceneScriptingFloatChannel}
          Translation.Y {type=MovieSceneScriptingFloatChannel, keys [[0,0],[14.076,30]]}
          Angle {type=MovieSceneScriptingFloatChannel}
          Scale.X {type=MovieSceneScriptingFloatChannel}
          Scale.Y {type=MovieSceneScriptingFloatChannel}
          Shear.X {type=MovieSceneScriptingFloatChannel}
          Shear.Y {type=MovieSceneScriptingFloatChannel}
         ]
        }]
      }
     ]
    }]
   display_rate {numerator=60}
  }
 }
 CollapseLeaderboardSummary {
  binding_count=1
  length_seconds=0.3877
  object=&self:CollapseLeaderboardSummary
  sequence {bindings [CommonBorder_Leaderboard {
     id=#22
     possessed_class=CommonBorder
     tracks [{
       type=MovieSceneFloatVectorTrack
       display_name=Desired Size Scale
       sections [{
         type=MovieSceneFloatVectorSection
         end_frame=7
         start_frame=0
         channels [
          Vector.X {type=MovieSceneScriptingFloatChannel}
          Vector.Y {type=MovieSceneScriptingFloatChannel, keys [[0,1],[7.7527,0]]}
         ]
        }]
      }]
    }]}
 }
 HidePosterFade {
  binding_count=1
  length_seconds=0.4635
  object=&self:HidePosterFade
  sequence {bindings [Image_PosterFade {
     id=#3
     possessed_class=Image
     tracks [{
       type=MovieSceneWidgetMaterialTrack
       sections [{
         type=MovieSceneParameterSection
         end_frame=9
         start_frame=0
         channels [
          CenterR {type=MovieSceneScriptingFloatChannel, keys [[0,0.9],[9.2697,1.0897]]}
          CenterG {type=MovieSceneScriptingFloatChannel, keys [[0,1.08],[9.2697,1.7384]]}
          CenterB {type=MovieSceneScriptingFloatChannel}
          CenterA {type=MovieSceneScriptingFloatChannel}
         ]
        }]
      }]
    }]}
 }
]
graphs [
 Animate Event Window Details {nodes [
   n1 = Animate Event Window Details!() -> (then, Play Speed:double=1) {is=title K2Node_FunctionEntry}
   n2 = Get EventWindowDetailsShown() -> (EventWindowDetailsShown:bool)
   > AnimTransition!(
    execute
    self
    In:bool=@n2
    Animation:Widget Animation Object Reference=Get ShowEventWindowDetails() -> (ShowEventWindowDetails:Widget Animation Object Reference)
    PlaySpeed:double=@n1
   ) -> (then)
   > AnimTransition(In=@n2, Animation=Get BlurPoster() -> (BlurPoster:Widget Animation Object Reference), PlaySpeed=@n1)
   > AnimTransition(In=@n2, Animation=Get BlurBG() -> (BlurBG:Widget Animation Object Reference), PlaySpeed=@n1)
   > AnimTransition(In=@n2, Animation=Get CollapseLeaderboardSummary() -> (CollapseLeaderboardSummary:Widget Animation Object Reference), PlaySpeed=@n1)
  ]}
 BackActionSelected {nodes [BackActionSelected() {is=title K2Node_FunctionEntry}]}
 CreateScoringWidgets {nodes [
   n1 = CreateScoringWidgets!() -> (then, TournamentSeriesId:String, TournamentId:String) {is=title K2Node_FunctionEntry}
   > n2 = FindTournamentScoringRules!(execute, self:Fort Tournament Details Page Object Reference, TournamentSeriesId:String=@n1, TournamentId:String=@n1.TournamentId) -> (then, ReturnValue:Array of Fort Showdown Scoring Rule Info Structures)
   > n3 = RemoveInvalidScoringRules!(execute, self, UnfilteredScoringRules:Array of Fort Showdown Scoring Rule Info Structures=@n2) -> (then, FilteredScoringRules:Array of Fort Showdown Scoring Rule Info Structures)
   > n4 = For Each Loop(Array:Array of Fort Showdown Scoring Rule Info Structures=@n3) -> (Array Element:Fort Showdown Scoring Rule Info Structure)
   > n5 = Create Widget!(
    execute:Exec
    Class:User Widget Class Reference=/CompeteOldTab/Frontend/Showdown/ShowdownTournamentDetailsScoringRule.ShowdownTournamentDetailsScoringRule_C
    OwningPlayer:Player Controller Object Reference
   ) -> (then:Exec, ReturnValue:Showdown Tournament Details Scoring Rule Object Reference)
   > n6 = AddChildToUniformGrid!(
    execute
    self:Uniform Grid Panel Object Reference=Get UniformGridPanel_TournamentScoringRules() -> (UniformGridPanel_TournamentScoringRules:Uniform Grid Panel Object Reference)
    Content:Widget Object Reference=@n5
    InRow:int=0
    InColumn:int=0
   ) -> (then, ReturnValue:Uniform Grid Slot Object Reference)
   n7 = int / int!(A:int=Length(TargetArray:Array of Fort Showdown Scoring Rule Info Structures=@n3), B:int=2) -> (ReturnValue:int=0) {is=title K2Node_CallFunction}
   n8 = "integer >= integer"!(A:int=@n4.Array Index, B:int=@n7) -> (ReturnValue:bool) {is=title K2Node_CallFunction}
   > SetRow!(
    execute
    self:Uniform Grid Slot Object Reference=@n6
    InRow:int=Select(Option 0:int=@n4.Array Index, Option 1:int=int - int(A=@n4.Array Index, B=@n7) {is=title K2Node_CallFunction}, Index=@n8) -> (ReturnValue:int)
   ) -> (then)
   > SetColumn!(execute, self:Uniform Grid Slot Object Reference=@n6, InColumn:int=Select(Option 0=0, Option 1=1, Index=@n8)) -> (then)
   > SetHorizontalAlignment(self:Uniform Grid Slot Object Reference=@n6, InHorizontalAlignment=HAlign_Fill)
   > SetVerticalAlignment(self:Uniform Grid Slot Object Reference=@n6, InVerticalAlignment=VAlign_Fill)
   > SetData!(
    execute
    self:Showdown Tournament Details Scoring Rule Object Reference=@n5
    ScoreRule:Fort Showdown Scoring Rule Info Structure=@n4
    TournamentDisplayInfo:Fort Tournament Display Info Structure=Get TournamentDisplayInfo() -> (TournamentDisplayInfo:Fort Tournament Display Info Structure)
   ) -> (then)
  ]}
 EventGraph {nodes [
   "On Button Clicked (SeriesPointLeaderboardButton)"!() -> (OutputDelegate, then, Button:Common Button Legacy Object Reference) {is=title K2Node_ComponentBoundEvent}
   > Branch(Condition=IsSeriesPointLeaderboardAvailable!(self:Fort Showdown Detail View Object Reference) -> (ReturnValue:bool))
   > n1 = Branch(Condition=Get EventWindowDetailsShown())
   > n2 = CloseEventWindowDetails()
   Init!(
    execute=[@n2.then, @n1.else]
    self:Event Leaderboard Screen Object Reference=Get EventLeaderboardScreen() -> (EventLeaderboardScreen:Event Leaderboard Screen Object Reference)
    EventWindowId:String
    TournamentSeriesId:String=GetTournamentId!(self:Fort Showdown Detail View Object Reference) -> (ReturnValue:String)
   ) -> (then)
   > SetActiveWidgetIndex(
    self=Get CommonWidgetSwitcher_FullScreenSwitch() -> (CommonWidgetSwitcher_FullScreenSwitch:Common Widget Switcher Legacy Object Reference)
    Index=1
   )
   n3 = "On Button Clicked (Button_Back)"!() -> (OutputDelegate, then, Button:Common Button Legacy Object Reference) {is=title K2Node_ComponentBoundEvent}
   HandleBackClicked(execute=[@n3.then, @n15.then, @n17.then], self:Fort Tournament Details Page Object Reference)
   "On Button Clicked (DetailsPanelButton)"!() -> (OutputDelegate, then, Button:Common Button Legacy Object Reference) {is=title K2Node_ComponentBoundEvent}
   > n4 = Branch(Condition=Get EventWindowDetailsShown())
   > n5 = CloseEventWindowDetails()
   Set TournamentDetailsShown(execute=[@n5.then, @n4.else], TournamentDetailsShown:bool=true) -> (Output_Get:bool)
   n6 = Get ShowDetails() -> (ShowDetails:Widget Animation Object Reference)
   > AnimTransition(In=true, Animation=@n6, PlaySpeed=1)
   n7 = Get CommonWidgetSwitcher_DetailsButtons() -> (CommonWidgetSwitcher_DetailsButtons:Common Widget Switcher Legacy Object Reference)
   n8 = Get CommonWidgetSwitcher_WindowsDetails() -> (CommonWidgetSwitcher_WindowsDetails:Common Widget Switcher Legacy Object Reference)
   > SetActiveWidgetIndex(self=[@n7, @n8], Index=1)
   n9 = Get PosterFlip() -> (PosterFlip:Widget Animation Object Reference)
   > AnimTransition(In=true, Animation=@n9, PlaySpeed=1)
   > PlaySound2D(Sound=/Game/Sounds/Fort_UI_Sounds/FrontEnd/Fort_UI_Showdown_Details_Cue)
   n10 = Get EventWindowDetailsShown()
   > n11 = Branch(Condition=@n10)
   n12 = Get BlurPoster()
   > AnimTransition(In=false, Animation=@n12, PlaySpeed=1)
   n13 = Get BlurBG()
   AnimTransition(execute=@n11.else, In=true, Animation=@n13, PlaySpeed=1)
   n14 = Get CollapseLeaderboardSummary()
   > AnimTransition(In=true, Animation=@n14, PlaySpeed=1)
   n15 = "On Button Clicked (OverviewPanelButton)"!() -> (OutputDelegate, then, Button:Common Button Legacy Object Reference) {is=title K2Node_ComponentBoundEvent}
   "On Button Clicked (Button_EnableMFA)"!() -> (OutputDelegate, then, Button:Common Button Legacy Object Reference) {is=title K2Node_ComponentBoundEvent}
   > n16 = Create Widget(Class=/CompeteOldTab/UI/Login/MFA/EnableMultiFactorAuthModalBP.EnableMultiFactorAuthModalBP_C, OwningPlayer=GetOwningPlayer()) -> (ReturnValue:Enable Multi Factor Auth Modal BP Object Reference)
   > "SetText (Text)"(
    self=Get CT_MFA_Description!(self:Enable Multi Factor Modal Object Reference=@n16) -> (CT_MFA_Description:Common Text Object Reference)
    InText="LOCTEXT(\"Two-Factor Authentication (2FA) is required to compete for cash prizes!\")"
   )
   > QueueModal!(execute, self:Enable Multi Factor Modal Object Reference=@n16, bAllowPermanentOptOut:bool=false, bTournamentQueue:bool=true) -> (then, ReturnValue:bool)
   n17 = "On Button Clicked (CloseButton)"!() -> (OutputDelegate, then, Button:Common Button Legacy Object Reference) {is=title K2Node_ComponentBoundEvent}
   n18 = EventShowWindowDetails!() -> (OutputDelegate, then, Button:Common Button Legacy Object Reference) {is=title K2Node_CustomEvent}
   > n19 = Set lastFocusedEvent(lastFocusedEvent:Widget Object Reference=@n18.Button) -> (Output_Get:Widget Object Reference)
   > n20 = Cast To ShowdownEventTile!(execute, Object=@n18.Button) -> (then, CastFailed, AsShowdown Event Tile:Showdown Event Tile Object Reference) {is=title K2Node_DynamicCast}
   > SetVisibility(
    self=Get Button_Payout!(self:Tournament Event Window Details Object Reference=Get TournamentEventWindowDetails() -> (TournamentEventWindowDetails:Tournament Event Window Details Object Reference)) -> (Button_Payout:Events Simple Button Object Reference)
    InVisibility=Visible
   )
   > n21 = UpdateEventWindow!(
    execute
    self:Fort Tournament Payout Modal Object Reference=Get TournamentPayoutModal() -> (TournamentPayoutModal:Tournament Payout Modal Object Reference)
    EventId:String=Get TournamentId!(self:Showdown Event Tile Object Reference=@n20) -> (TournamentId:String)
    EventWindowId:String=Get EventId!(self:Showdown Event Tile Object Reference=@n20) -> (EventId:String)
   ) -> (then)
   RefreshViolators(execute=[@n21.then, @n20.CastFailed])
   > ScrollWidgetIntoView!(
    execute
    self:Scroll Box Object Reference=Get EventsScrollBox() -> (EventsScrollBox:Scroll Box Object Reference)
    WidgetToFind:Widget Object Reference=@n19
    AnimateScroll:bool=true
    ScrollDestination:EDescendantScrollDestination Enum=Center
    Padding:float=0
   ) -> (then)
   > EventDeselectWindows!(execute, self, SelectedWidget:Widget Object Reference=@n19) -> (then)
   > n22 = Cast To ShowdownEventTile(Object=@n19)
   > SetTournament!(
    execute
    self:Fort Showdown Detail View Object Reference=Get TournamentEventWindowDetails()
    TournamentId:String=Get TournamentId(self=@n22)
    EventId:String=Get EventId(self=@n22)
    bForceRefreshUI:bool=true
   ) -> (then)
   > RefreshScoringRules()
   > n23 = Branch(Condition=Get EventWindowDetailsShown())
   Set EventWindowDetailsShown(execute=@n23.else, EventWindowDetailsShown:bool=true) -> (Output_Get:bool)
   > Animate Event Window Details!(execute, self, Play Speed:double=1) -> (then) {is=title K2Node_CallFunction}
   n24 = GetAllChildren(self=Get EventsScrollBox())
   > n25 = Reverse for Each Loop!(Exec:Exec, Array:Array of Widget Object References=@n24) -> (LoopBody:Exec, ArrayIndex:int, ArrayElement:Widget Object Reference, Completed:Exec) {is=title K2Node_MacroInstance}
   > n26 = Cast To ShowdownEventTile(Object=@n25.ArrayElement)
   > EventResize!(execute, self:Showdown Event Tile Object Reference=@n26, Out:bool=false) -> (then)
   > PlaySound2D(Sound=/Game/Sounds/Fort_UI_Sounds/FrontEnd/Fort_UI_Showdown_Open_Pressed_Cue)
   n27 = AnimTransition!() -> (OutputDelegate, then, In:bool=~, Animation:Widget Animation Object Reference, PlaySpeed:double=1) {is=title K2Node_CustomEvent}
   > n28 = Branch(Condition=@n27.In)
   > n29 = PauseAnimation!(execute, self:User Widget Object Reference, InAnimation:Widget Animation Object Reference=@n27.Animation) -> (then, ReturnValue:float)
   n30 = GetStartTime!(self:Widget Animation Object Reference=@n27.Animation) -> (ReturnValue:float)
   > PlayAnimation(
    InAnimation=@n27.Animation
    StartAtTime=float + float(A=@n29, B=@n30) {is=title K2Node_CommutativeAssociativeBinaryOperator}
    PlaybackSpeed=@n27.PlaySpeed
   )
   n31 = PauseAnimation(execute=@n28.else, InAnimation=@n27.Animation)
   n32 = GetEndTime(self=@n27.Animation)
   > PlayAnimation(
    InAnimation=@n27.Animation
    StartAtTime=Select(
     Option 0:double=float - float(A=@n32, B=@n31) {is=title K2Node_CallFunction}
     Option 1:double=0
     Index="float <= float"!(A:double=@n31, B:double=@n30) -> (ReturnValue:bool) {is=title K2Node_CallFunction}
    ) -> (ReturnValue:double)
    PlayMode=Reverse
    PlaybackSpeed=@n27.PlaySpeed
   )
   n33 = EventToggleFullScreenContent!() -> (OutputDelegate, then, In:bool=~) {is=title K2Node_CustomEvent}
   n34 = Get FullScreenContent() -> (FullScreenContent:Widget Animation Object Reference)
   > AnimTransition(In=@n33.In, Animation=@n34, PlaySpeed=1) {is=title K2Node_CallFunction}
   > n35 = Branch(Condition=@n33.In)
   n36 = EventFullScreenAnimOutFinished() {is=title K2Node_CustomEvent}
   BindToAnimationFinished(execute=@n35.else, Animation=@n34, Delegate=@n36)
   n37 = EventViewPayout!() -> (OutputDelegate, then, EventWindowId:String, EventId:String) {is=title K2Node_CustomEvent}
   n38 = Get TournamentPayoutModal()
   > SetEventWindowId!(execute, self:Fort Tournament Payout Modal Object Reference=@n38, EventId:String=@n37.EventId, EventWindowId:String=@n37.EventWindowId) -> (then)
   > Init!(execute, self:Tournament Payout Modal Object Reference=@n38, EventWindowId=@n37.EventWindowId, TournamentId:String=@n37.EventId) -> (then)
   > n39 = SetActiveWidgetIndex(self=Get CommonWidgetSwitcher_FullScreenSwitch(), Index=2)
   n40 = Get BlurPoster()
   AnimTransition(execute=[@n39.then, @n47.then], In=false, Animation=@n40, PlaySpeed=1)
   n41 = Get BlurBG()
   > AnimTransition(In=false, Animation=@n41, PlaySpeed=1)
   n42 = Get HidePosterFade() -> (HidePosterFade:Widget Animation Object Reference)
   > AnimTransition(In=true, Animation=@n42, PlaySpeed=1)
   n43 = EventclosePayout() {is=title K2Node_CustomEvent}
   > n44 = Sequence()
   AnimTransition(execute=[@n49.then, @n44.then_0], In=true, Animation=@n40, PlaySpeed=1)
   > AnimTransition(In=true, Animation=@n41, PlaySpeed=1)
   > AnimTransition(In=false, Animation=@n42, PlaySpeed=1)
   SetActiveWidgetIndex(execute=@n44.then_1, self=Get CommonWidgetSwitcher_FullScreenSwitch(), Index=0)
   > EventToggleFullScreenContent!(execute, self, In:bool=false) -> (then) {is=title K2Node_CallFunction}
   n45 = No payouts() {is=title K2Node_CustomEvent}
   > SetVisibility(self=Get Button_Payout(self=Get TournamentEventWindowDetails()), InVisibility=Collapsed)
   n46 = EventviewLiveGames() {is=title K2Node_CustomEvent}
   > OnViewLiveGames(self:Fort Tournament Live Games Object Reference=Get TournamentWatchList() -> (TournamentWatchList:Tournament Live Games Object Reference))
   > SetActiveWidgetIndex(self=Get CommonWidgetSwitcher_FullScreenSwitch(), Index=3)
   > n47 = EventToggleFullScreenContent(In=false)
   n48 = EventBackoutLiveGames() {is=title K2Node_CustomEvent}
   > Event Leaderboard Screen On Activated(self:Event Leaderboard Screen Object Reference=Get EventLeaderboardScreen())
   > SetActiveWidgetIndex(self=Get CommonWidgetSwitcher_FullScreenSwitch(), Index=1)
   > n49 = EventToggleFullScreenContent(In=true)
   CloseTournamentDetails() {is=title K2Node_CustomEvent}
   > Set TournamentDetailsShown(TournamentDetailsShown=false)
   > AnimTransition(In=false, Animation=@n6, PlaySpeed=1)
   > SetActiveWidgetIndex(self=[@n7, @n8], Index=0)
   > AnimTransition(In=false, Animation=@n9, PlaySpeed=1)
   > PlaySound2D(Sound=&Fort_UI_Showdown_Details_Cue)
   > n50 = Branch(Condition=@n10)
   > AnimTransition(In=true, Animation=@n12, PlaySpeed=1)
   AnimTransition(execute=@n50.else, In=false, Animation=@n13, PlaySpeed=1)
   > AnimTransition(In=false, Animation=@n14, PlaySpeed=1)
   CloseEventWindowDetails() {is=title K2Node_CustomEvent}
   n51 = Get lastFocusedEvent() -> (lastFocusedEvent:Widget Object Reference)
   > n52 = Is Valid(InputObject=@n51)
   Set lastFocusedEvent(execute=[@n56.then, @n52.Is Not Valid])
   n53 = Get TournamentEventWindowDetails()
   > SetVisibility(
    self=[
     Get Button_Payout(self=@n53)
     Get Button_Leaderboard!(self:Tournament Event Window Details Object Reference=@n53) -> (Button_Leaderboard:Events Simple Button Object Reference)
    ]
    InVisibility=Collapsed
   )
   > RefreshViolators()
   > EventDeselectWindows()
   > Set EventWindowDetailsShown(EventWindowDetailsShown=false)
   > Animate Event Window Details(Play Speed=1)
   > n54 = Reverse for Each Loop(Array=@n24)
   > n55 = Cast To ShowdownEventTile(Object=@n54.ArrayElement)
   > EventResize(self=@n55, Out=true)
   > PlaySound2D(Sound=/Game/Sounds/Fort_UI_Sounds/FrontEnd/Fort_UI_Showdown_Close_Pressed_Cue)
   n56 = SetFocus(execute=@n52.Is Valid, self=@n51)
   n57 = EventDeselectWindows!() -> (OutputDelegate, then, SelectedWidget:Widget Object Reference) {is=title K2Node_CustomEvent}
   n58 = Get EventsScrollBox()
   > n59 = For Loop(LastIndex=GetChildrenCount!(self:Panel Widget Object Reference=@n58) -> (ReturnValue:int=0))
   > Branch(Condition="Not Equal (Integer)"!(
     A:int=@n59
     B:int=GetChildIndex!(self:Panel Widget Object Reference=@n58, Content:Widget Object Reference=@n57.SelectedWidget) -> (ReturnValue:int=0)
    ) -> (ReturnValue:bool) {is=title K2Node_CallFunction})
   > n60 = Cast To ShowdownEventTile(Object=GetChildAt!(self:Panel Widget Object Reference=@n58, Index:int=@n59) -> (ReturnValue:Widget Object Reference))
   > Branch(Condition=GetSelected!(self:Common Button Base Object Reference=@n60) -> (ReturnValue:bool))
   > ClearSelection(self:Common Button Base Object Reference=@n60)
   n61 = EventCenterEventWindow!() -> (OutputDelegate, then, Widget:Widget Object Reference) {is=title K2Node_CustomEvent}
   > n62 = Cast To CommonButtonLegacy!(execute, Object=@n61.Widget) -> (then, CastFailed, AsCommon Button Legacy:Common Button Legacy Object Reference) {is=title K2Node_DynamicCast}
   > ScrollWidgetIntoView(self=Get EventsScrollBox(), WidgetToFind=@n62, AnimateScroll=true, ScrollDestination=Center, Padding=0)
   > CenterWidget!(execute, ToCenter:Widget Object Reference=@n62) -> (then)
   n63 = EventStartCentering() {is=title K2Node_CustomEvent}
   > Set isCenteringScrollBox(isCenteringScrollBox:bool=true) -> (Output_Get:bool)
   n64 = EventStopCentering() {is=title K2Node_CustomEvent}
   > Set isCenteringScrollBox(isCenteringScrollBox=false)
   n65 = EventViewFullLeaderboard!() -> (OutputDelegate, then, TournamentId:String, EventId:String) {is=title K2Node_CustomEvent}
   > Init!(
    execute
    self:Event Leaderboard Screen Object Reference=Get EventLeaderboardScreen()
    EventWindowId=@n65.EventId
    TournamentSeriesId:String=@n65.TournamentId
   ) -> (then)
   > SetActiveWidgetIndex(self=Get CommonWidgetSwitcher_FullScreenSwitch(), Index=1)
   > EventToggleFullScreenContent(In=true)
   > Init!(
    execute
    self:Tournament Live Games Object Reference=Get TournamentWatchList()
    TournamentId:String=@n65.TournamentId
    EventId:String=@n65.EventId
   ) -> (then)
   n66 = EventCloseLeaderboard() {is=title K2Node_CustomEvent}
   > SetActiveWidgetIndex(self=Get CommonWidgetSwitcher_FullScreenSwitch(), Index=0)
   > n67 = Branch(Condition=Get bIsMainLeaderboardNonWindow!(self:Fort Event Leaderboard Screen Object Reference=Get EventLeaderboardScreen()) -> (bIsMainLeaderboardNonWindow:bool))
   n68 = Get CommonWidgetSwitcher_FullScreenSwitch()
   > Delay(Duration="Max (Float)"(A=Get TransitionDuration!(self:Common Animated Switcher Object Reference=@n68) -> (TransitionDuration:float), B=0))
   n69 = Get firstActiveEvent() -> (firstActiveEvent:Widget Object Reference)
   > n70 = Is Valid(InputObject=@n69)
   n71 = Get EventsScrollBox()
   n72 = GetChildrenCount(self=@n71) -> (ReturnValue=0)
   SetFocus(execute=@n70.Is Not Valid, self=GetChildAt(self=@n71, Index=int - int(A=@n72, B=1)))
   SetFocus(execute=@n70.Is Valid, self=@n69)
   EventToggleFullScreenContent(execute=@n67.else, In=false)
   Event OnInitialized()
   > n73 = Branch(Condition=IsMobileGame())
   > SetVisibility(self=Get BackgroundBlur_Background() -> (BackgroundBlur_Background:Background Blur Object Reference), InVisibility=Hidden)
   > SetBrushColor!(
    execute
    self=Get CommonBorder_BGText() -> (CommonBorder_BGText:Common Border Object Reference)
    InBrushColor_R:float=0
    InBrushColor_G:float=0
    InBrushColor_B:float=0
    InBrushColor_A:float=0
   ) -> (then)
   SetVisibility(execute=@n73.else, self=Get BackgroundBlur_Background(), InVisibility=SelfHitTestInvisible)
   > SetBrushColor(self=Get CommonBorder_BGText(), InBrushColor_R=0, InBrushColor_G=0, InBrushColor_B=0, InBrushColor_A=1)
   Event Construct()
   > ClearChildren(self=Get EventsScrollBox())
   > SetScrollWidget!(execute, ToScroll:Widget Object Reference=Get ScrollBox_DetailAndRules() -> (ScrollBox_DetailAndRules:Scroll Box Object Reference)) -> (then)
   > PlayAnimationForward(InAnimation=Get Intro())
   > PlaySound2D(Sound=/Game/Sounds/Fort_UI_Sounds/FrontEnd/Store/Fort_UI_Showdown_Cardpick_02_Cue)
   > Set isCenteringScrollBox(isCenteringScrollBox=false)
   > Animate Event Window Details(Play Speed=9999)
   n74 = Get ShowEventWindowDetails()
   > BindToAnimationStarted!(execute, self:Widget Animation Object Reference=@n74, Widget:User Widget Object Reference=Self-Reference(), Delegate:Delegate=@n63) -> (then)
   > BindToAnimationFinished!(execute, self:Widget Animation Object Reference=@n74, Widget:User Widget Object Reference=Self-Reference(), Delegate=@n64) -> (then)
   n75 = Get TournamentEventWindowDetails()
   > Bind Event to View Payouts(self:Tournament Event Window Details Object Reference=@n75, Delegate=@n37) {is=title K2Node_AddDelegate}
   > Bind Event to View Leaderboard(self:Tournament Event Window Details Object Reference=@n75, Delegate=@n65) {is=title K2Node_AddDelegate}
   > Bind Event to Call Event Live Games Back(self:Tournament Live Games Object Reference=Get TournamentWatchList(), Delegate=@n48) {is=title K2Node_AddDelegate}
   n76 = Get EventLeaderboardScreen()
   > Bind Event to Event View Live Games(self:Event Leaderboard Screen Object Reference=@n76, Delegate=@n46) {is=title K2Node_AddDelegate}
   > Bind Event to Event Close Button(self:Event Leaderboard Screen Object Reference=@n76, Delegate=@n66) {is=title K2Node_AddDelegate}
   n77 = Get TournamentPayoutModal()
   > Bind Event to Call Event Back(self:Tournament Payout Modal Object Reference=@n77, Delegate=@n43) {is=title K2Node_AddDelegate}
   > Bind Event to No Payout(self:Tournament Payout Modal Object Reference=@n77, Delegate=@n45) {is=title K2Node_AddDelegate}
   n78 = Get TournamentEventWindowDetails()
   > SetVisibility(self=[Get Button_Payout(self=@n78), Get Button_Leaderboard(self=@n78)], InVisibility=Collapsed)
   > n79 = GetSafeZonePadding!(execute) -> (then, SafePadding:Vector 4 Structure, SafePaddingScale:Vector 2D Structure, SpillOverPadding:Vector 4 Structure)
   n80 = SlotAsOverlaySlot(Widget=Get Image_PosterFade() -> (Image_PosterFade:Image Object Reference))
   n81 = Break Margin!(Margin:"Margin Structure (by ref)"=Get Padding!(self:Overlay Slot Object Reference=@n80) -> (Padding:Margin Structure)) -> (Left:float=~, Top:float=~, Right:float=~, Bottom:float=~) {is=title K2Node_BreakStruct}
   n82 = BreakVector4!(InVec:"Vector 4 Structure (by ref)"=@n79) -> (X:double, Y:double, Z:double, W:double)
   > SetPadding!(
    execute
    self:Overlay Slot Object Reference=[
     SlotAsOverlaySlot(Widget=Get BackgroundBlur_PosterLayer() -> (BackgroundBlur_PosterLayer:Background Blur Object Reference))
     SlotAsOverlaySlot(Widget=Get BackgroundBlur_Background())
     @n80
    ]
    InPadding:Margin Structure=Make Margin(
     Left=float - float(A=@n81, B=@n82)
     Top=float - float(A=@n81.Top, B=@n82.Y)
     Right=float - float(A=@n81.Right, B=@n82.Z)
     Bottom=float - float(A=@n81.Bottom, B=@n82.W)
    )
   ) -> (then)
   Event Tick()
   > Branch(Condition=Get isCenteringScrollBox() -> (isCenteringScrollBox:bool))
   > EventCenterEventWindow!(execute, self, Widget:Widget Object Reference=Get lastFocusedEvent()) -> (then) {is=title K2Node_CallFunction}
   Event Refresh Data() {is=title K2Node_Event}
   > SetVisibility(
    self=Get MFAContainer() -> (MFAContainer:Vertical Box Object Reference)
    InVisibility=Select(
     Option 0:ESlateVisibility Enum=Collapsed
     Option 1:ESlateVisibility Enum=SelfHitTestInvisible
     Index=DoCalloutMFARequirement!(self:Fort Showdown Detail View Object Reference) -> (ReturnValue:bool)
    ) -> (ReturnValue:ESlateVisibility Enum)
   )
   > SetVisibility(
    self=Get SeriesPointLeaderboardButton() -> (SeriesPointLeaderboardButton:Events Simple Button Object Reference)
    InVisibility=Select(Option 0=Collapsed, Option 1=SelfHitTestInvisible, Index=IsSeriesPointLeaderboardAvailable())
   )
   > Refresh()
   n83 = Get EventsScrollBox()
   > Branch(Condition="integer > integer"(A=GetChildrenCount(self=@n83) -> (ReturnValue=0), B=1) {is=title K2Node_CallFunction})
   n84 = Get EventsScrollBox()
   n85 = GetChildrenCount(self=@n84) -> (ReturnValue=0)
   n86 = int - int(A=@n85, B=1)
   > n87 = For Loop with Break!(Execute:Exec, FirstIndex:int, LastIndex:int=@n86, Break:Exec=@n91.then) -> (LoopBody:Exec, Index:int, Completed:Exec) {is=title K2Node_MacroInstance}
   n88 = GetChildAt(self=@n84, Index=@n87)
   > n89 = Cast To ShowdownEventTile(Object=@n88)
   > n90 = Branch(Condition=NOT Boolean(A=Get isEnded!(self:Showdown Event Tile Object Reference=@n89) -> (isEnded:bool)))
   > n91 = Set firstActiveEvent(firstActiveEvent:Widget Object Reference=@n88) -> (Output_Get:Widget Object Reference)
   Branch(execute=@n90.else, Condition="Equal (Integer)"(A=@n87, B=@n86) {is=title K2Node_CallFunction})
   > Set firstActiveEvent(firstActiveEvent=@n88)
   Delay(execute=@n87.Completed, Duration=float * float(A=0.25, B=float / float(A:int=@n87, B:int=@n85)) {is=title K2Node_PromotableOperator})
   > EventCenterEventWindow(Widget=Get firstActiveEvent())
   Event OnActivateScreen() {is=title K2Node_Event}
   n92 = Get CommonWidgetSwitcher_FullScreenSwitch()
   n93 = Get EventLeaderboardScreen()
   > Branch(Condition="Equal (Object)"(A=GetActiveWidget!(self:Widget Switcher Object Reference=@n92) -> (ReturnValue:Widget Object Reference), B=@n93) {is=title K2Node_CallFunction})
   > Event Leaderboard Screen On Activated(self=@n93)
   n94 = Get lastFocusedEvent()
   SetFocus(execute=@n95.Is Valid, self=@n94)
   n95 = Is Valid(exec=@n36.then, InputObject=@n94)
  ]}
 Handle_Back {nodes [
   Handle_Back() {is=title K2Node_FunctionEntry}
   n1 = GetActiveWidgetIndex!(self:Widget Switcher Object Reference=Get CommonWidgetSwitcher_FullScreenSwitch()) -> (ReturnValue:int=0)
   > n2 = Branch(Condition="Equal (Integer)"(A=@n1, B=0))
   > n3 = Branch(Condition=Get TournamentDetailsShown() -> (TournamentDetailsShown:bool))
   > CloseTournamentDetails() {is=title K2Node_CallFunction}
   > Return Node!(execute, Passthrough:bool=false)
   n4 = Branch(execute=@n3.else, Condition=Get EventWindowDetailsShown())
   > CloseEventWindowDetails() {is=title K2Node_CallFunction}
   > Return Node(Passthrough=false)
   Return Node(execute=@n4.else, Passthrough=true)
   n5 = Branch(execute=@n2.else, Condition="Equal (Integer)"(A=@n1, B=1))
   > n6 = EventCloseLeaderboard() {is=title K2Node_CallFunction}
   Return Node(execute=[@n7.then, @n6.then], Passthrough=false)
   Branch(execute=@n5.else, Condition="Equal (Integer)"(A=@n1, B=2))
   > n7 = EventclosePayout() {is=title K2Node_CallFunction}
  ]}
 Is Scoring Rule Valid {nodes [
   n1 = Is Scoring Rule Valid!() -> (then, ScoringRuleInfo:"Fort Showdown Scoring Rule Info Structure (by ref)") {is=title K2Node_FunctionEntry}
   > Return Node!(
    execute
    IsValid:bool=HasScoringRuleDisplayInfo!(
     self:CMSContext Object Reference=Get CMSContext!() -> (ReturnValue:CMSContext Object Reference) {is=title K2Node_GetBlueprintContextEx}
     RuleName:Name=Break Fort Showdown Scoring Rule Info!(FortShowdownScoringRuleInfo:"Fort Showdown Scoring Rule Info Structure (by ref)"=@n1) -> (DisplayInfoId:Name, PointValue:int, AccumulatedPointValue:int, ScoreRequirement:int, UpperBoundRequirement:int, LowerBoundRequirement:int) {is=title K2Node_BreakStruct}
    ) -> (ReturnValue:bool)
   )
  ]}
 OnHandleBack {nodes [
   OnHandleBack() {is=title K2Node_FunctionEntry}
   > n1 = Handle_Back!(execute, self) -> (then, Passthrough:bool) {is=title K2Node_CallFunction}
   > Return Node!(execute, bResult:bool=@n1)
  ]}
 Refresh {nodes [
   Refresh() {is=title K2Node_FunctionEntry}
   > RefreshOverviewPanel()
   > RefreshDetailsPanel()
   > SetTournament(
    self=Get EventLeaderboardSummary() -> (EventLeaderboardSummary:Event Leaderboard Summary Object Reference)
    TournamentId=GetTournamentId()
    EventId=GetTournamentNextEvent!(self:Fort Showdown Detail View Object Reference) -> (ReturnValue:String)
    bForceRefreshUI=true
   )
  ]}
 RefreshDetailsPanel {nodes [
   n1 = Break Fort Tournament Display Info!(FortTournamentDisplayInfo:"Fort Tournament Display Info Structure (by ref)"=Get TournamentDisplayInfo()) -> (
    TitleLine1:Text
    TitleLine2:Text
    ScheduleInfo:Text
    FlavorDescription:Text
    DetailsDescription:Text
    ShortFormatTitle:Text
    LongFormatTitle:Text
    BackgroundTitle:Text
    PinScoreRequirement:int
    PinEarnedText:Text
    BaseColor:Linear Color Structure
    PrimaryColor:Linear Color Structure
    SecondaryColor:Linear Color Structure
    HighlightColor:Linear Color Structure
    TitleColor:Linear Color Structure
    ShadowColor:Linear Color Structure
    BackgroundLeftColor:Linear Color Structure
    BackgroundRightColor:Linear Color Structure
    BackgroundTextColor:Linear Color Structure
    PosterFadeColor:Linear Color Structure
    AlertText:Text
    SeriesPointLeaderboardName:Text
    AlertType:EFortTournamentAlertType Enum
    RoundDisplayNames:Array of Texts
   ) {is=title K2Node_BreakStruct}
   n2 = SetContentColorAndOpacity!(execute, self:Border Object Reference=Get CommonBorder_BGText(), InContentColorAndOpacity:Linear Color Structure=@n1.BackgroundTextColor) -> (then)
   SetColorAndOpacity(
    execute=[@n2.then, @n15.then]
    self=Get TournamentRoundText() -> (TournamentRoundText:Common Text Object Reference)
    InColorAndOpacity=Make Slate Color(SpecifiedColor=@n1.PrimaryColor)
   )
   n3 = BreakColor(InColor=@n1.ShadowColor)
   > SetBrushColor!(
    execute
    self=Get CommonBorder_ScoringBG() -> (CommonBorder_ScoringBG:Common Border Object Reference)
    InBrushColor:Linear Color Structure=MakeColor!(R:float=@n3, G:float=@n3.G, B:float=@n3.B, A:float=0) -> (ReturnValue:Linear Color Structure)
   ) -> (then)
   > SetVectorParameterValue(
    self=GetEffectMaterial!(self:Retainer Box Object Reference=Get RetainerBox_BGText() -> (RetainerBox_BGText:Retainer Box Object Reference)) -> (ReturnValue:Material Instance Dynamic Object Reference)
    ParameterName=Color
    Value=@n1.BackgroundTextColor
   )
   > SetBrushColor(self=Get CommonBorder_ScoringTitleBG() -> (CommonBorder_ScoringTitleBG:Common Border Object Reference), InBrushColor=@n1.SecondaryColor)
   > SetColorAndOpacity(
    self=Get CommonTextBlock_UnlockCriteria() -> (CommonTextBlock_UnlockCriteria:Common Text Object Reference)
    InColorAndOpacity=Make Slate Color(SpecifiedColor=@n1.PrimaryColor)
   )
   n4 = GetEffectMaterial(self=Get RetainerBox_TimeCallout() -> (RetainerBox_TimeCallout:Retainer Box Object Reference))
   > SetVectorParameterValue(self=@n4, ParameterName=PrimaryColor, Value=@n1.PrimaryColor)
   > SetVectorParameterValue(self=@n4, ParameterName=HighlightColor, Value=@n1.HighlightColor)
   n5 = Get EventLeaderboardScreen()
   > n6 = Cast To LeaderboardListViewWrapper!(
    execute
    Object=Get ActivatableContent_LeaderboardListWrapper!(self:Fort Leaderboard View Tab Object Reference=Get ActivatableContent_LeaderboardTab!(self:Fort Event Leaderboard Screen Object Reference=@n5) -> (ActivatableContent_LeaderboardTab:Fort Leaderboard View Tab Object Reference)) -> (ActivatableContent_LeaderboardListWrapper:Fort Leaderboard List View Wrapper Object Reference)
   ) -> (then, CastFailed, AsLeaderboard List View Wrapper:Leaderboard List View Wrapper Object Reference) {is=title K2Node_DynamicCast}
   > Set LeaderboardListViewWrapper(LeaderboardListViewWrapper:Leaderboard List View Wrapper Object Reference=@n6) -> (Output_Get:Leaderboard List View Wrapper Object Reference)
   n7 = Get EventLeaderboardScreen()
   n8 = Get TournamentPayoutModal()
   > n9 = For Each Loop(Array:Array of Events Simple Button Object References=Make Array!(
     "[0]":Events Simple Button Object Reference=Get DetailsPanelButton() -> (DetailsPanelButton:Events Simple Button Object Reference)
     "[1]":Events Simple Button Object Reference=Get Button_Back() -> (Button_Back:Events Simple Button Object Reference)
     "[2]":Events Simple Button Object Reference=Get OverviewPanelButton() -> (OverviewPanelButton:Events Simple Button Object Reference)
     "[3]":Events Simple Button Object Reference=Get Button_Back!(self:Event Leaderboard Screen Object Reference=@n7) -> (Button_Back)
     "[4]":Events Simple Button Object Reference=Get BackButton!(self:Tournament Payout Modal Object Reference=@n8) -> (BackButton:Events Simple Button Object Reference)
     "[5]":Events Simple Button Object Reference=Get SeriesPointLeaderboardButton()
    ) -> (Array:Array of Events Simple Button Object References)) -> (Array Element:Events Simple Button Object Reference)
   > Set GamepadButtonTextColor!(execute, GamepadButtonTextColor:Linear Color Structure=@n1.SecondaryColor, self:Events Simple Button Object Reference=@n9) -> (then, Output_Get:Linear Color Structure)
   > n10 = GetDynamicMaterial(self=Get Image_ButtonTop!(self:Events Simple Button Object Reference=@n9) -> (Image_ButtonTop:Image Object Reference))
   > SetVectorParameterValue(self=@n10, ParameterName=Color1, Value=@n1.SecondaryColor)
   > SetVectorParameterValue(self=@n10, ParameterName=Color2, Value=@n1.SecondaryColor)
   > Branch(Condition="float < float"!(A:double=RGB to HSV!(InColor:Linear Color Structure=@n1.SecondaryColor) -> (H:float, S:float, V:float, A:float).V, B:double=1.1) -> (ReturnValue:bool) {is=title K2Node_CallFunction})
   > n11 = GetDynamicFontMaterial!(
    execute
    self:Text Object Reference=Get CenterButtonTextWidget!(self:Events Simple Button Object Reference=@n9) -> (CenterButtonTextWidget:Common Text Object Reference)
   ) -> (then, ReturnValue:Material Instance Dynamic Object Reference)
   > SetVectorParameterValue(self=@n11, ParameterName=BaseColor, Value=(R=1,G=1,B=1,A=1))
   > SetVectorParameterValue(self=@n11, ParameterName=HoverColor, Value=(R=0.0052,G=0.0052,B=0.0056,A=1))
   > Set DefaultTextColor!(execute, DefaultTextColor:Linear Color Structure=(R=1,G=1,B=1,A=1), self:Events Simple Button Object Reference=@n9) -> (then, Output_Get:Linear Color Structure)
   n12 = Get TournamentDisplayInfo()
   EventColorize!(
    execute=@n9.Completed
    self:Event Leaderboard Summary Object Reference=Get EventLeaderboardSummary()
    TournamentDisplayInfo:Fort Tournament Display Info Structure=@n12
   ) -> (then)
   > EventColorize(self:Event Leaderboard Screen Object Reference=Get EventLeaderboardScreen(), TournamentDisplayInfo=@n12)
   > EventColorize(self:Tournament Payout Modal Object Reference=Get TournamentPayoutModal(), TournamentDisplayInfo=@n12)
   RefreshDetailsPanel() {is=title K2Node_FunctionEntry}
   > SetTournament(
    self=Get TournamentPosterBack() -> (TournamentPosterBack:Showdown Tournament Poster Back Object Reference)
    TournamentId=GetTournamentId()
    bForceRefreshUI=true
   )
   > RefreshScoringRules()
   n13 = Break Showdown Tournament Entry!(ShowdownTournamentEntry:"Showdown Tournament Entry Structure (by ref)"=Get CMSTournamentData() -> (CMSTournamentData:Showdown Tournament Entry Structure)) -> (
    TournamentDisplayId:String
    TitleLine1:String
    TitleLine2:String
    ScheduleInfo:String
    PosterFrontImage:String
    PosterBackImage:String
    FlavorDescription:String
    DetailsDescription:String
    ShortFormatTitle:String
    LongFormatTitle:String
    BackgroundTitle:String
    PinScoreRequirement:int
    PinEarnedText:String
    BaseColor:String
    PrimaryColor:String
    SecondaryColor:String
    HighlightColor:String
    TitleColor:String
    ShadowColor:String
    BackgroundLeftColor:String
    BackgroundRightColor:String
    BackgroundTextColor:String
    PosterFadeColor:String
    PlaylistTileImage:String
    LoadingScreenImage:String
    TournamentViewBackgroundImage:String
    TournamentLandingPageImage:String
    SquarePosterImage:String
    StyleInfoId:String
    AlertText:String
    SeriesPointLeaderboardName:String
    PlaylistDescription:String
    AlertType:EFortTournamentAlertType Enum
    RoundDisplayNames:Array of Texts
   ) {is=title K2Node_BreakStruct}
   > "SetText (Text)"(
    self=Get CommonTextBlock_TournamentSubtitle() -> (CommonTextBlock_TournamentSubtitle:Common Text Object Reference)
    InText=TextToUpper!(InText:"Text (by ref)"="To Text (String)"(InString=@n13.LongFormatTitle)) -> (ReturnValue:Text)
   )
   > "SetText (Text)"(
    self=Get CommonTextBlock_TournamentDescription() -> (CommonTextBlock_TournamentDescription:Common Text Object Reference)
    InText="To Text (String)"(InString=@n13.DetailsDescription)
   )
   > SetVisibility(
    self=Get UnlockPinContent() -> (UnlockPinContent:Horizontal Box Object Reference)
    InVisibility=Select(Option 0=Collapsed, Option 1=SelfHitTestInvisible, Index="integer > integer"(A=@n13.PinScoreRequirement, B=0))
   )
   > "SetText (Text)"(
    self=Get Text_PinUnlockScore() -> (Text_PinUnlockScore:Common Text Object Reference)
    InText="To Text (Integer)"(Value=@n13.PinScoreRequirement)
   )
   > n14 = GetDynamicMaterial(self=Get Image_PosterFade())
   > n15 = SetVectorParameterValue(self=@n14, ParameterName=Color, Value=@n1.PosterFadeColor)
   InputTypeChanged!(execute, self:Events Simple Button Object Reference=@n9, bUsingGamepad:bool=false) -> (then)
  ]}
 RefreshEvents {nodes [
   RefreshEvents() {is=title K2Node_FunctionEntry}
   n1 = Get EventsScrollBox()
   > ClearChildren(self=@n1)
   > n2 = Create Widget(Class=/CompeteOldTab/Frontend/Showdown/ShowdownEventEmptySpacer.ShowdownEventEmptySpacer_C) -> (ReturnValue:Showdown Event Empty Spacer Object Reference)
   > AddChild(self=@n1, Content=@n2)
   > n3 = GetTournamentEvents!(
    execute
    self:Fort Showdown Detail View Object Reference
    bGroupByRoundNumber:bool=CanGroupTournamentEventsByVisualRoundNumber!(self:Fort Showdown Detail View Object Reference) -> (ReturnValue:bool)
   ) -> (then, ReturnValue:Array of Strings)
   > n4 = For Each Loop(Array:Array of Strings=@n3) -> (Array Element:String)
   UpdateDefaultTournamentEvent!(execute=@n4.Completed, self, TournamentEvents:Array of Strings=@n3) -> (then)
   n5 = Create Widget!(
    execute=@n4.LoopBody
    Class=/CompeteOldTab/Frontend/Showdown/ShowdownEventTile.ShowdownEventTile_C
    OwningPlayer
    Style:Common Button Style Class Reference
    bApplyAlphaOnDisable:bool=True
    bLocked:bool=False
    bSelectable:bool=True
    bShouldSelectUponReceivingFocus:bool=False
    bInteractableWhenSelected:bool=False
    bToggleable:bool=False
    bTriggerClickedAfterSelection:bool=False
    bDisplayInputActionWhenNotInteractable:bool=True
    bHideInputActionWithKeyboard:bool=False
    bShouldUseFallbackDefaultInputAction:bool=True
    bRequiresHold:bool=False
    HoldData:Common UIHold Data Class Reference
    ClickMethod:EButtonClickMethod Enum=MouseDown
    InputPriority:int=0
    TriggeringInputAction:Data Table Row Handle Structure
   ) -> (then, ReturnValue:Showdown Event Tile Object Reference)
   > AddChild(self=@n1, Content=@n5)
   > SetData!(execute, self:Showdown Event Tile Object Reference=@n5, TournamentInfo:String=GetTournamentId(), EventInfo:String=@n4) -> (then)
   > Bind Event to On Button Clicked(self:Common Button Legacy Object Reference=@n5, Delegate=Create Event()) {is=title K2Node_AddDelegate}
   > EventIntro!(
    execute
    self:Showdown Event Tile Object Reference=@n5
    Index:double=float / float(A=@n4.Array Index, B=Length(TargetArray:Array of Strings=@n3))
   ) -> (then)
   PrintString!(execute, InString=Scroll Box - Event Refresh, bPrintToScreen, bPrintToLog, TextColor, Duration, Key) -> (then)
  ]}
 RefreshOverviewPanel {nodes [
   RefreshOverviewPanel() {is=title K2Node_FunctionEntry}
   n1 = Break Showdown Tournament Entry(ShowdownTournamentEntry=Get CMSTournamentData())
   > "SetText (Text)"(
    self=Get TournamentSubTitleBackground() -> (TournamentSubTitleBackground:Common Text Object Reference)
    InText=TextToUpper(InText="To Text (String)"(InString=@n1.BackgroundTitle))
   )
   > "SetText (Text)"(self=Get TournamentRoundText(), InText=TextToUpper(InText="To Text (String)"(InString=@n1.ShortFormatTitle)))
   > n2 = Branch(Condition=IsSeriesPointLeaderboardAvailable())
   n3 = "To Text (String)"(InString=@n1.SeriesPointLeaderboardName)
   > n4 = Branch(Condition=TextIsEmpty(InText=@n3))
   n5 = GetTournamentMatchType!(self:Fort Showdown Detail View Object Reference) -> (ReturnValue:EFortShowdownMatchType Enum=Unknown)
   "SetText (Text)"(
    execute=[@n12.then, @n2.else, @n4.then]
    self=Get TournamentMatchType() -> (TournamentMatchType:Common Text Object Reference)
    InText=TextToUpper(InText=Select!(
      Unknown:Text
      Solo:Text="LOCTEXT(\"Solo\")"
      Duos:Text="LOCTEXT(\"Duos\")"
      Squads:Text="LOCTEXT(\"Squads\")"
      Index:EFortShowdownMatchType Enum=@n5
     ) -> (ReturnValue:Text))
   )
   n6 = GetTournamentTimeLeft!(self:Fort Showdown Detail View Object Reference) -> (ReturnValue:Timespan Structure)
   n7 = GetEventTimespanAsText!(
    self:Fort Showdown Detail View Object Reference
    InTimeRemaining:Timespan Structure=@n6
    bShowSeconds:bool=false
    FormatType:ETimespanAsTextFormat Enum=Approximate
   ) -> (ReturnValue:Text)
   n8 = BreakTimespan!(InTimespan:Timespan Structure=@n6) -> (Days:int=0, Hours:int=0, Minutes:int=0, Seconds:int=0, Milliseconds:int=0)
   n9 = "Equal (Integer)"(A=@n8, B=0)
   n10 = "Equal (Integer)"(A=@n8, B=1)
   n11 = Select!(
    Unknown
    FutureTBD:Text
    FutureScheduled:Text=Select!(
     Option 0:Text=Select(Option 0="LOCTEXT(\"This event starts in {0}.\")", Option 1="LOCTEXT(\"<b>This event starts tomorrow!</>\")", Index=@n10)
     Option 1:Text="LOCTEXT(\"<b>This event starts today!</>\")"
     Index:bool=@n9
    ) -> (ReturnValue)
    FutureNext:Text=Select(Option 0="LOCTEXT(\"The next event is live in {0}.\")", Option 1="LOCTEXT(\"<b>The next event is live later today!</>\")", Index=@n9)
    Live:Text="LOCTEXT(\"<b>THIS EVENT IS LIVE NOW!</>\")"
    LiveParticipating:Text="LOCTEXT(\"<b>THIS EVENT IS LIVE NOW!</>\")"
    LiveNotParticipating:Text="LOCTEXT(\"<b>THIS EVENT IS LIVE NOW!</>\")"
    Completed:Text="LOCTEXT(\"This event has ended.\")"
    CompletedParticipated:Text="LOCTEXT(\"This event has ended.\")"
    CompletedNotPartipated:Text="LOCTEXT(\"This event has ended.\")"
    Cancelled:Text
    Index:EFortShowdownEventState Enum=GetTournamentTimeState!(self:Fort Showdown Detail View Object Reference) -> (ReturnValue:EFortShowdownEventState Enum=Unknown)
   ) -> (ReturnValue)
   > SetText(
    self=Get RichTextBlock_TournamentTimeCallout1() -> (RichTextBlock_TournamentTimeCallout1:Rich Text Block Object Reference)
    InText=TextToUpper(InText=Format Text!(Format=@n11, 0:Text=@n7) -> (Result))
   )
   > SetTournament(
    self=Get TournamentPosterFront() -> (TournamentPosterFront:Showdown Tournament Poster Front Object Reference)
    TournamentId=GetTournamentId()
    bForceRefreshUI=true
   )
   > RefreshEvents() {is=title K2Node_CallFunction}
   n12 = "SetText (Text)"(execute=@n4.else, self=Get CenterButtonTextWidget(self=Get SeriesPointLeaderboardButton()), InText=@n3)
   JoinStringArray(SourceArray=ParseIntoArray(SourceString=@n1.ShortFormatTitle, Delimiter=" "), Separator="\r\n")
  ]}
 RefreshScoringRules {nodes [
   RefreshScoringRules() {is=title K2Node_FunctionEntry}
   > ClearChildren(self=Get UniformGridPanel_TournamentScoringRules())
   > n1 = Cast To ShowdownEventTile(Object=Get lastFocusedEvent())
   n2 = GetTournamentId()
   > CreateScoringWidgets!(execute, self, TournamentSeriesId:String=@n2, TournamentId:String=Get EventId(self=@n1)) -> (then) {is=title K2Node_CallFunction}
   CreateScoringWidgets(
    execute=@n1.CastFailed
    TournamentSeriesId=@n2
    TournamentId=Get DefaultTournamentIdForScoringRules() -> (DefaultTournamentIdForScoringRules:String)
   )
  ]}
 RefreshViolators {nodes [
   RefreshViolators() {is=title K2Node_FunctionEntry}
   n1 = Get lastFocusedEvent()
   > Branch(Condition=IsValid(Object=@n1))
   > n2 = Cast To ShowdownEventTile(Object=@n1)
   Set FocusedEvent(execute=[@n2.CastFailed, @n2.then], FocusedEvent:Showdown Event Tile Object Reference=@n2) -> (Output_Get:Showdown Event Tile Object Reference)
   n3 = Get EventsScrollBox()
   > n4 = For Loop(LastIndex=GetChildrenCount(self=@n3) -> (ReturnValue=0))
   n5 = GetNumEventVisualRounds!(self:Fort Showdown Detail View Object Reference) -> (ReturnValue:int=0)
   n6 = Get FocusedEvent() -> (FocusedEvent:Showdown Event Tile Object Reference)
   n7 = Branch(
    execute=@n4.Completed
    Condition=OR Boolean(
     A=IsNullItem!(ItemObject:Object Reference=@n6) -> (Is Null Item:bool)
     B="integer <= integer"!(A:int=@n5, B:int=1) -> (ReturnValue:bool) {is=title K2Node_CallFunction}
    )
   )
   n8 = Break Showdown Tournament Entry(ShowdownTournamentEntry=Get CMSTournamentData())
   n9 = Get TournamentRoundText()
   > "SetText (Text)"(self=@n9, InText=TextToUpper(InText="To Text (String)"(InString=@n8.ShortFormatTitle)))
   n10 = Get Violator!(self:Showdown Event Tile Object Reference=Get FocusedEvent()) -> (Violator:Showdown Event Tile Violator Object Reference)
   "SetText (Text)"(
    execute=@n7.else
    self=@n9
    InText=TextToUpper(InText=GetWindowRoundDisplayString!(self:Fort Showdown Detail View Object Reference=@n10) -> (ReturnValue:Text))
   )
   n11 = Cast To ShowdownEventTile(execute=@n4.LoopBody, Object=GetChildAt(self=@n3, Index=@n4))
   n12 = Get Violator(self=@n11)
   n13 = GetEventWindowVisualRoundNumber!(self:Fort Showdown Detail View Object Reference=@n12) -> (ReturnValue:int=0)
   n14 = GetEventWindowVisualRoundNumber(self=Get Violator(self=Get FocusedEvent())) -> (ReturnValue=0)
   n15 = IsNullItem(ItemObject=Get FocusedEvent())
   > Set OverrideDoNotShowRound!(
    execute
    OverrideDoNotShowRound:bool=AND Boolean(A=NOT Boolean(A=@n15), B="Equal (Integer)"(A=@n14, B=@n13))
    self:Showdown Event Tile Violator Object Reference=@n12
   ) -> (then, Output_Get:bool)
   > RefreshViolatorData(self:Showdown Event Tile Violator Object Reference=@n12)
  ]}
 RemoveInvalidScoringRules {nodes [
   n1 = RemoveInvalidScoringRules!() -> (then, UnfilteredScoringRules:Array of Fort Showdown Scoring Rule Info Structures) {is=title K2Node_FunctionEntry}
   > n2 = Sequence!(execute) -> (then_0, then_1, then_2:Exec)
   > n3 = For Each Loop(Array:Array of Fort Showdown Scoring Rule Info Structures=@n1) -> (Array Element:Fort Showdown Scoring Rule Info Structure)
   > n4 = Is Scoring Rule Valid!(execute, self, ScoringRuleInfo:"Fort Showdown Scoring Rule Info Structure (by ref)"=@n3) -> (then, IsValid:bool) {is=title K2Node_CallFunction}
   > Branch(Condition=@n4)
   > Add(
    TargetArray:Array of Fort Showdown Scoring Rule Info Structures=Get FilteredTournamentRules() -> (FilteredTournamentRules:Array of Fort Showdown Scoring Rule Info Structures)
    NewItem:"Fort Showdown Scoring Rule Info Structure (by ref)"=@n3
   ) {is=title K2Node_CallArrayFunction}
   Return Node!(execute=@n2.then_1, FilteredScoringRules:Array of Fort Showdown Scoring Rule Info Structures=Get FilteredTournamentRules())
  ]}
 UpdateDefaultTournamentEvent {nodes [
   n1 = UpdateDefaultTournamentEvent!() -> (then, TournamentEvents:Array of Strings) {is=title K2Node_FunctionEntry}
   n2 = GetTournamentNextEvent()
   > n3 = Branch(Condition=IsEmpty(InString=@n2))
   n4 = Length(TargetArray=@n1)
   > Set DefaultTournamentIdForScoringRules(DefaultTournamentIdForScoringRules:String="Get (a copy)"(Array=@n1, Dimension 1=int - int(A=@n4) {is=title K2Node_PromotableOperator})) -> (Output_Get:String)
   Set DefaultTournamentIdForScoringRules(execute=@n3.else, DefaultTournamentIdForScoringRules=@n2)
  ]}
]
variables [
 BackActionSelected {}
 DefaultTournamentIdForScoringRules {type=string, default=""}
 EventWindowDetailsShown {type=boolean}
 firstActiveEvent {type=Widget, default=null}
 isCenteringScrollBox {type=boolean}
 lastFocusedEvent {type=Widget, default=null}
 LeaderboardisCollapsed {type=boolean}
 LeaderboardListViewWrapper {type=LeaderboardListViewWrapper_C, default=null}
 TournamentDetailsShown {type=boolean}
]
widgets {
 info {inheritedWidgetCount=6, parentClass=/Script/FortniteUI.FortTournamentDetailsPage, rootWidgetClass=&Overlay, widgetCount=100}
 widgets [
  Overlay_2 {widgetClassPath=&Overlay, widget=Overlay_2, properties {visibility=Visible}}
  DetailsSwitcher {
   widgetClassPath=&CommonWidgetSwitcherLegacy
   bIsVariable=true
   parent=Overlay_2
   widget=DetailsSwitcher
   slot {class=OverlaySlot, properties {horizontalAlignment=HAlign_Fill, verticalAlignment=VAlign_Fill}}
  }
  OverviewPanel {widgetClassPath=&Overlay, bIsVariable=true, parent=DetailsSwitcher, widget=OverviewPanel, slot {class=WidgetSwitcherSlot}}
  RetainerBox_BGText {
   widgetClassPath=/Script/UMG.RetainerBox
   bIsVariable=true
   parent=OverviewPanel
   widget=RetainerBox_BGText
   properties {effectMaterial=/CompeteOldTab/Frontend/Showdown/Materials/M_UI_Tournament_BGText, visibility=HitTestInvisible}
   slot {class=OverlaySlot, properties {horizontalAlignment=HAlign_Fill, verticalAlignment=VAlign_Fill}}
  }
  CommonBorder_BGText {
   widgetClassPath=&CommonBorder
   bIsVariable=true
   parent=RetainerBox_BGText
   widget=CommonBorder_BGText
   properties {
    background {
     bIsDynamicallyLoaded=false
     drawAs=Image
     imageType=NoImage
     mirroring=NoMirror
     resourceName=None
     resourceObject=null
     tiling=NoTile
     imageSize=<32 32>
     margin=<ltrb 0 0 0 0>
     outlineSettings {
      bUseBrushTransparency=false
      roundingType=HalfHeightRadius
      width=0
      color {colorUseRule=UseColor_Specified, specifiedColor=<rgba 0 0 0 0>}
      cornerRadii=<0 0 0 0>
     }
     tintColor {colorUseRule=UseColor_Specified, specifiedColor=<rgba 1 1 1 1>}
     uVRegion {bIsValid=false, max=<0 0>, min=<0 0>}
    }
    brushColor=<rgba 0 0 0 0>
    padding=<ltrb 0 0 0 0>
    renderTransformPivot=<1 0>
   }
   slot {class=PanelSlot}
  }
  ScaleBox_2 {
   widgetClassPath=&ScaleBox
   parent=CommonBorder_BGText
   widget=ScaleBox_2
   properties {renderTransform {angle=0, scale=<1 1>, shear=<0 0>, translation=<42 44>}}
   slot {class=BorderSlot, properties {padding=<ltrb 0 0 0 0>}}
  }
  TournamentSubTitleBackground {
   widgetClassPath=&CommonTextBlock
   bIsVariable=true
   parent=ScaleBox_2
   widget=TournamentSubTitleBackground
   properties {
    lineHeightPercentage=0.75
    text="Preseason\r\nTournament"
    font {
     bForceMonospaced=false
     bMaterialIsStencil=false
     fontMaterial=null
     fontObject=&Burbank
     letterSpacing=0
     monospacedWidth=1
     size=175
     skewAmount=0
     typefaceFontName=BurbankRegular-Black
     outlineSettings {
      bApplyOutlineToDropShadows=true
      bMiteredCorners=false
      bSeparateFillAlpha=false
      outlineMaterial=null
      outlineSize=3
      outlineColor=<rgba 0 0 0 0.1>
     }
    }
   }
   slot {class=ScaleBoxSlot, properties {horizontalAlignment=HAlign_Right, verticalAlignment=VAlign_Top}}
  }
  BackgroundBlur_Background {
   widgetClassPath=&BackgroundBlur
   bIsVariable=true
   parent=OverviewPanel
   widget=BackgroundBlur_Background
   properties {blurStrength=10, renderOpacity=0}
   slot {class=OverlaySlot, properties {horizontalAlignment=HAlign_Fill, verticalAlignment=VAlign_Fill}}
  }
  Overlay_0 {
   widgetClassPath=&Overlay
   parent=OverviewPanel
   widget=Overlay_0
   slot {class=OverlaySlot, properties {horizontalAlignment=HAlign_Fill, verticalAlignment=VAlign_Fill}}
  }
  SafeZone_0 {
   widgetClassPath=&SafeZone
   bIsVariable=true
   parent=Overlay_0
   widget=SafeZone_0
   slot {class=OverlaySlot, properties {horizontalAlignment=HAlign_Fill, verticalAlignment=VAlign_Fill}}
  }
  OverlayPoster {widgetClassPath=&Overlay, parent=SafeZone_0, widget=OverlayPoster, slot {class=SafeZoneSlot}}
  SizeBox_PosterContainer {
   widgetClassPath=&SizeBox
   parent=OverlayPoster
   widget=SizeBox_PosterContainer
   properties {heightOverride=1080, maxDesiredWidth=1266.55, minDesiredWidth=750, widthOverride=750}
   slot {class=OverlaySlot, properties {horizontalAlignment=HAlign_Fill, verticalAlignment=VAlign_Fill, padding=<ltrb 900 0 0 0>}}
  }
  ScaleBox_0 {widgetClassPath=&ScaleBox, parent=SizeBox_PosterContainer, widget=ScaleBox_0, slot {class=SizeBoxSlot}}
  OverlayPosterSides {
   widgetClassPath=&Overlay
   parent=ScaleBox_0
   widget=OverlayPosterSides
   properties {renderTransformPivot=<0.5 1.6641>}
   slot {class=ScaleBoxSlot, properties {horizontalAlignment=HAlign_Right, verticalAlignment=VAlign_Bottom}}
  }
  TournamentPosterFront {
   widgetClassPath=/CompeteOldTab/Frontend/Showdown/ShowdownTournamentPosterFront
   bIsVariable=true
   parent=OverlayPosterSides
   widget=TournamentPosterFront
   properties {titleColor {colorUseRule=UseColor_Specified, specifiedColor=<rgba 1 1 1 1>}}
   slot {class=OverlaySlot, properties {horizontalAlignment=HAlign_Fill, verticalAlignment=VAlign_Fill}}
  }
  TournamentPosterBack {
   widgetClassPath=/CompeteOldTab/Frontend/Showdown/ShowdownTournamentPosterBack
   bIsVariable=true
   parent=OverlayPosterSides
   widget=TournamentPosterBack
   properties {renderOpacity=0, titlePaddingMobile=<ltrb 0 0 80 0>}
   slot {class=OverlaySlot}
  }
  VisibilityBorder_Touch {
   widgetClassPath=&CommonHardwareVisibilityBorder
   parent=OverlayPoster
   widget=VisibilityBorder_Touch
   properties {visibilityQuery {
     autoDescription=" ANY( Input.Touch )"
     tokenStreamVersion=0
     userDescription=""
     queryTokenStream [0,1,1,1,0]
     tagDictionary [{tagName=Input.Touch}]
    }}
   slot {class=OverlaySlot, properties {horizontalAlignment=HAlign_Right, padding=<ltrb 0 25 50 20>}}
  }
  CloseButton {
   widgetClassPath=/FortUILibrary/Foundation/Buttons/CloseButton
   bIsVariable=true
   parent=VisibilityBorder_Touch
   widget=CloseButton
   properties {
    onButtonBaseClicked=(&self:WidgetTree.CloseButton.HandleOnButtonClicked)
    onButtonBaseDoubleClicked=(&self:WidgetTree.CloseButton.HandleOnButtonDoubleClicked)
    onButtonBaseHovered=(&self:WidgetTree.CloseButton.HandleOnButtonHovered)
    onButtonBaseUnhovered=(&self:WidgetTree.CloseButton.HandleOnButtonUnhovered)
    onSelectedChangedBase=(&self:WidgetTree.CloseButton.HandleOnSelectedChanged)
   }
   slot {class=BorderSlot, properties {padding=<ltrb 0 0 0 0>}}
  }
  BackgroundBlur_PosterLayer {
   widgetClassPath=&BackgroundBlur
   bIsVariable=true
   parent=Overlay_0
   widget=BackgroundBlur_PosterLayer
   properties {blurStrength=12, renderOpacity=0}
   slot {class=OverlaySlot, properties {horizontalAlignment=HAlign_Fill, verticalAlignment=VAlign_Fill}}
  }
  Image_PosterFade {
   widgetClassPath=&Image
   bIsVariable=true
   parent=Overlay_0
   widget=Image_PosterFade
   properties {
    visibility=HitTestInvisible
    brush {
     bIsDynamicallyLoaded=false
     drawAs=Image
     imageType=NoImage
     mirroring=NoMirror
     resourceName=None
     resourceObject=/CompeteOldTab/Frontend/Showdown/Materials/M_UI_TournamentDetails_PosterFade_Inst
     tiling=NoTile
     imageSize=<32 602>
     margin=<ltrb 0 0 0 0>
     outlineSettings {
      bUseBrushTransparency=false
      roundingType=HalfHeightRadius
      width=0
      color {colorUseRule=UseColor_Specified, specifiedColor=<rgba 0 0 0 0>}
      cornerRadii=<0 0 0 0>
     }
     tintColor {colorUseRule=UseColor_Specified, specifiedColor=<rgba 1 1 1 1>}
     uVRegion {bIsValid=false, max=<0 0>, min=<0 0>}
    }
   }
   slot {class=OverlaySlot, properties {horizontalAlignment=HAlign_Fill, verticalAlignment=VAlign_Fill}}
  }
  CommonWidgetSwitcher_FullScreenSwitch {
   widgetClassPath=&CommonWidgetSwitcherLegacy
   bIsVariable=true
   parent=Overlay_0
   widget=CommonWidgetSwitcher_FullScreenSwitch
   properties {transitionType=Zoom}
   slot {class=OverlaySlot, properties {horizontalAlignment=HAlign_Fill, verticalAlignment=VAlign_Fill}}
  }
  SafeZone_1 {
   widgetClassPath=&SafeZone
   bIsVariable=true
   parent=CommonWidgetSwitcher_FullScreenSwitch
   widget=SafeZone_1
   slot {class=WidgetSwitcherSlot}
  }
  VerticalBoxEventContent {widgetClassPath=&VerticalBox, parent=SafeZone_1, widget=VerticalBoxEventContent, slot {class=SafeZoneSlot}}
  HorizontalBox_0 {
   widgetClassPath=&HorizontalBox
   parent=VerticalBoxEventContent
   widget=HorizontalBox_0
   slot {class=VerticalBoxSlot, properties {padding=<ltrb 35 55 0 0>}}
  }
  VerticalBoxEventTitle {
   widgetClassPath=&VerticalBox
   parent=HorizontalBox_0
   widget=VerticalBoxEventTitle
   slot {class=HorizontalBoxSlot, properties {size {sizeRule=Fill, value=0.5}}}
  }
  CommonBorder_TournamentTitle {
   widgetClassPath=&CommonBorder
   parent=VerticalBoxEventTitle
   widget=CommonBorder_TournamentTitle
   properties {clipping=ClipToBounds, verticalAlignment=VAlign_Bottom, padding=<ltrb 0 0 0 0>, renderTransformPivot=<0.5 1>}
   slot {class=VerticalBoxSlot, properties {verticalAlignment=VAlign_Top, size {sizeRule=Automatic, value=0.5}}}
  }
  SizeBox_2 {
   widgetClassPath=&SizeBox
   parent=CommonBorder_TournamentTitle
   widget=SizeBox_2
   properties {heightOverride=83.2038, widthOverride=1142}
   slot {class=BorderSlot, properties {verticalAlignment=VAlign_Bottom, padding=<ltrb 0 0 0 0>}}
  }
  CommonTextBlock_TournamentSubtitle {
   widgetClassPath=&CommonTextBlock
   bIsVariable=true
   parent=SizeBox_2
   widget=CommonTextBlock_TournamentSubtitle
   properties {
    autoWrapText=true
    lineHeightPercentage=0.8
    text=NEW YORK MAJORS
    font {
     bForceMonospaced=false
     bMaterialIsStencil=false
     fontMaterial=null
     fontObject=&Burbank
     letterSpacing=0
     monospacedWidth=1
     size=59
     skewAmount=0
     typefaceFontName=BurbankRegular-Black
     outlineSettings {
      bApplyOutlineToDropShadows=false
      bMiteredCorners=false
      bSeparateFillAlpha=false
      outlineMaterial=null
      outlineSize=0
      outlineColor=<rgba 0 0 0 1>
     }
    }
    margin=<ltrb 0 0 0 -16>
    renderTransform {angle=0, scale=<0.95 1>, shear=<0 0>, translation=<0 0>}
    renderTransformPivot=<0 0.5>
   }
   slot {class=SizeBoxSlot, properties {verticalAlignment=VAlign_Top, padding=<ltrb 0 0 0 40>}}
  }
  CommonBorder_TournamentTitleDetails {
   widgetClassPath=&CommonBorder
   parent=VerticalBoxEventTitle
   widget=CommonBorder_TournamentTitleDetails
   properties {padding=<ltrb 0 0 0 0>}
   slot {class=VerticalBoxSlot, properties {horizontalAlignment=HAlign_Left}}
  }
  VerticalBox_2 {
   widgetClassPath=&VerticalBox
   parent=CommonBorder_TournamentTitleDetails
   widget=VerticalBox_2
   slot {class=BorderSlot, properties {padding=<ltrb 0 0 0 0>}}
  }
  RetainerBox_TimeCallout {
   widgetClassPath=&RetainerBox
   bIsVariable=true
   parent=VerticalBox_2
   widget=RetainerBox_TimeCallout
   properties {effectMaterial=/CompeteOldTab/Frontend/Showdown/Materials/M_TimeCalloutColorize}
   slot {class=VerticalBoxSlot}
  }
  RichTextBlock_TournamentTimeCallout1 {
   widgetClassPath=/Script/UMG.RichTextBlock
   bIsVariable=true
   parent=RetainerBox_TimeCallout
   widget=RichTextBlock_TournamentTimeCallout1
   properties {
    text="<b>THIS EVENT IS LIVE NOW!</> "
    textStyleSet=/CompeteOldTab/Frontend/Showdown/TournamentStyles
    margin=<ltrb 0 -8 0 -8>
    renderTransform {angle=0, scale=<0.95 1>, shear=<-9 0>, translation=<0 0>}
    renderTransformPivot=<0 0.5>
   }
   slot {class=PanelSlot}
  }
  CommonWidgetSwitcher_DetailsButtons {
   widgetClassPath=&CommonWidgetSwitcherLegacy
   bIsVariable=true
   parent=VerticalBox_2
   widget=CommonWidgetSwitcher_DetailsButtons
   slot {class=VerticalBoxSlot, properties {padding=<ltrb 0 10 0 0>}}
  }
  VerticalBox_1 {widgetClassPath=&VerticalBox, parent=CommonWidgetSwitcher_DetailsButtons, widget=VerticalBox_1, slot {class=WidgetSwitcherSlot}}
  HorizontalBox_3 {widgetClassPath=&HorizontalBox, parent=VerticalBox_1, widget=HorizontalBox_3, slot {class=VerticalBoxSlot}}
  DetailsPanelButton {
   widgetClassPath=/CompeteOldTab/Frontend/Showdown/EventsSimpleButton
   bIsVariable=true
   parent=HorizontalBox_3
   widget=DetailsPanelButton
   properties {
    buttonNameText=Tournament Details
    clickMethod=DownAndUp
    onButtonBaseClicked=(&self:WidgetTree.DetailsPanelButton.HandleOnButtonClicked)
    onButtonBaseDoubleClicked=(&self:WidgetTree.DetailsPanelButton.HandleOnButtonDoubleClicked)
    onButtonBaseHovered=(&self:WidgetTree.DetailsPanelButton.HandleOnButtonHovered)
    onButtonBaseUnhovered=(&self:WidgetTree.DetailsPanelButton.HandleOnButtonUnhovered)
    onSelectedChangedBase=(&self:WidgetTree.DetailsPanelButton.HandleOnSelectedChanged)
    pressMethod=DownAndUp
    xL=false
    buttonPadding=<20 6>
    color {
     ButtonColor1=<rgba 0.008 0.9823 0.2423 1>
     ButtonColor2=<rgba 0.008 0.9823 0.2423 1>
     ButtonHover=<rgba 1 1 1 1>
     TextColor=<rgba 0.0052 0.0052 0.0056 1>
     TextHover=<rgba 0 0 0 1>
     TextPressed=<rgba 0 0 0 1>
     TextPressedHover=<rgba 0 0 0 1>
    }
    triggeringInputAction {dataTable=&InputActionDataTable, rowName=TournamentDetails}
   }
   slot {class=HorizontalBoxSlot, properties {horizontalAlignment=HAlign_Left, verticalAlignment=VAlign_Center}}
  }
  VisibilityBorder_NonTouch {
   widgetClassPath=&CommonHardwareVisibilityBorder
   parent=HorizontalBox_3
   widget=VisibilityBorder_NonTouch
   properties {visibilityQuery {
     autoDescription=" NONE( Input.Touch )"
     tokenStreamVersion=0
     userDescription=""
     queryTokenStream [0,1,3,1,0]
     tagDictionary [{tagName=Input.Touch}]
    }}
   slot {class=HorizontalBoxSlot, properties {horizontalAlignment=HAlign_Left, verticalAlignment=VAlign_Center, padding=<ltrb 30 0 0 0>}}
  }
  Button_Back {
   widgetClassPath=&EventsSimpleButton
   bIsVariable=true
   parent=VisibilityBorder_NonTouch
   widget=Button_Back
   properties {
    buttonNameText=Back
    clickMethod=DownAndUp
    onButtonBaseClicked=(&self:WidgetTree.Button_Back.HandleOnButtonClicked)
    onButtonBaseDoubleClicked=(&self:WidgetTree.Button_Back.HandleOnButtonDoubleClicked)
    onButtonBaseHovered=(&self:WidgetTree.Button_Back.HandleOnButtonHovered)
    onButtonBaseUnhovered=(&self:WidgetTree.Button_Back.HandleOnButtonUnhovered)
    onSelectedChangedBase=(&self:WidgetTree.Button_Back.HandleOnSelectedChanged)
    pressMethod=DownAndUp
    xL=false
    buttonPadding=<20 6>
    color {
     ButtonColor1=<rgba 0.008 0.9823 0.2423 1>
     ButtonColor2=<rgba 0.008 0.9823 0.2423 1>
     ButtonHover=<rgba 1 1 1 1>
     TextColor=<rgba 0.0052 0.0052 0.0056 1>
     TextHover=<rgba 0 0 0 1>
     TextPressed=<rgba 0 0 0 1>
     TextPressedHover=<rgba 0 0 0 1>
    }
    triggeringInputAction {dataTable=&InputActionDataTable, rowName=Back}
   }
   slot {class=BorderSlot, properties {padding=<ltrb 0 0 0 0>}}
  }
  SeriesPointLeaderboardButton {
   widgetClassPath=&EventsSimpleButton
   bIsVariable=true
   parent=VerticalBox_1
   widget=SeriesPointLeaderboardButton
   properties {
    buttonNameText=Series Point leaderboard
    clickMethod=DownAndUp
    onButtonBaseClicked=(&self:WidgetTree.SeriesPointLeaderboardButton.HandleOnButtonClicked)
    onButtonBaseDoubleClicked=(&self:WidgetTree.SeriesPointLeaderboardButton.HandleOnButtonDoubleClicked)
    onButtonBaseHovered=(&self:WidgetTree.SeriesPointLeaderboardButton.HandleOnButtonHovered)
    onButtonBaseUnhovered=(&self:WidgetTree.SeriesPointLeaderboardButton.HandleOnButtonUnhovered)
    onSelectedChangedBase=(&self:WidgetTree.SeriesPointLeaderboardButton.HandleOnSelectedChanged)
    pressMethod=DownAndUp
    xL=false
    buttonPadding=<20 6>
    color {
     ButtonColor1=<rgba 0.008 0.9823 0.2423 1>
     ButtonColor2=<rgba 0.008 0.9823 0.2423 1>
     ButtonHover=<rgba 1 1 1 1>
     TextColor=<rgba 0.0052 0.0052 0.0056 1>
     TextHover=<rgba 0 0 0 1>
     TextPressed=<rgba 0 0 0 1>
     TextPressedHover=<rgba 0 0 0 1>
    }
    triggeringInputAction {dataTable=&InputActionDataTable, rowName=GenericRightShoulder}
   }
   slot {class=VerticalBoxSlot, properties {horizontalAlignment=HAlign_Left, verticalAlignment=VAlign_Center, padding=<ltrb 0 5 0 0>}}
  }
  PanelButton_VisibilityBorder_NonTouch {
   widgetClassPath=&CommonHardwareVisibilityBorder
   parent=CommonWidgetSwitcher_DetailsButtons
   widget=PanelButton_VisibilityBorder_NonTouch
   properties {visibilityQuery {
     autoDescription=" NONE( Input.Touch )"
     tokenStreamVersion=0
     userDescription=""
     queryTokenStream [0,1,3,1,0]
     tagDictionary [{tagName=Input.Touch}]
    }}
   slot {class=WidgetSwitcherSlot, properties {horizontalAlignment=HAlign_Left, verticalAlignment=VAlign_Center}}
  }
  OverviewPanelButton {
   widgetClassPath=&EventsSimpleButton
   bIsVariable=true
   parent=PanelButton_VisibilityBorder_NonTouch
   widget=OverviewPanelButton
   properties {
    buttonNameText=BACK
    clickMethod=DownAndUp
    onButtonBaseClicked=(&self:WidgetTree.OverviewPanelButton.HandleOnButtonClicked)
    onButtonBaseDoubleClicked=(&self:WidgetTree.OverviewPanelButton.HandleOnButtonDoubleClicked)
    onButtonBaseHovered=(&self:WidgetTree.OverviewPanelButton.HandleOnButtonHovered)
    onButtonBaseUnhovered=(&self:WidgetTree.OverviewPanelButton.HandleOnButtonUnhovered)
    onSelectedChangedBase=(&self:WidgetTree.OverviewPanelButton.HandleOnSelectedChanged)
    pressMethod=DownAndUp
    xL=false
    color {
     ButtonColor1=<rgba 0.008 0.9823 0.2423 1>
     ButtonColor2=<rgba 0.008 0.9823 0.2423 1>
     ButtonHover=<rgba 1 1 1 1>
     TextColor=<rgba 0.0052 0.0052 0.0056 1>
     TextHover=<rgba 0 0 0 1>
     TextPressed=<rgba 0 0 0 1>
     TextPressedHover=<rgba 0 0 0 1>
    }
    triggeringInputAction {dataTable=&InputActionDataTable, rowName=Back}
   }
   slot {class=BorderSlot, properties {padding=<ltrb 0 0 0 0>}}
  }
  Spacer_191 {
   widgetClassPath=&Spacer
   parent=HorizontalBox_0
   widget=Spacer_191
   slot {class=HorizontalBoxSlot, properties {size {sizeRule=Fill, value=0.5}}}
  }
  CommonBorder_1 {
   widgetClassPath=&CommonBorder
   parent=VerticalBoxEventContent
   widget=CommonBorder_1
   properties {verticalAlignment=VAlign_Top}
   slot {class=VerticalBoxSlot, properties {horizontalAlignment=HAlign_Left}}
  }
  CommonBorder_Leaderboard {
   widgetClassPath=&CommonBorder
   parent=CommonBorder_1
   widget=CommonBorder_Leaderboard
   properties {clipping=ClipToBounds, padding=<ltrb 0 5 0 0>}
   slot {class=BorderSlot, properties {verticalAlignment=VAlign_Top}}
  }
  VerticalBox_3 {
   widgetClassPath=&VerticalBox
   parent=CommonBorder_Leaderboard
   widget=VerticalBox_3
   slot {class=BorderSlot, properties {padding=<ltrb 0 5 0 0>}}
  }
  CommonBorder_0 {
   widgetClassPath=&CommonBorder
   parent=VerticalBox_3
   widget=CommonBorder_0
   properties {verticalAlignment=VAlign_Center, padding=<ltrb 0 80 0 80>}
   slot {class=VerticalBoxSlot, properties {padding=<ltrb 35 10 0 0>, size {sizeRule=Fill, value=1}}}
  }
  MFAContainer {
   widgetClassPath=&VerticalBox
   bIsVariable=true
   parent=CommonBorder_0
   widget=MFAContainer
   slot {class=BorderSlot, properties {verticalAlignment=VAlign_Center, padding=<ltrb 0 80 0 80>}}
  }
  MFA_WarningText {
   widgetClassPath=&CommonTextBlock
   bIsVariable=true
   parent=MFAContainer
   widget=MFA_WarningText
   properties {
    justification=Center
    lineHeightPercentage=0.9
    text="Two-Factor Authentication (2FA) is required to compete for cash prizes!"
    font {
     bForceMonospaced=false
     bMaterialIsStencil=false
     fontMaterial=null
     fontObject=&Burbank
     letterSpacing=0
     monospacedWidth=1
     size=20
     skewAmount=0
     typefaceFontName=Burbank-PlayerPower
     outlineSettings {
      bApplyOutlineToDropShadows=false
      bMiteredCorners=false
      bSeparateFillAlpha=false
      outlineMaterial=null
      outlineSize=0
      outlineColor=<rgba 0 0 0 1>
     }
    }
   }
   slot {class=VerticalBoxSlot, properties {horizontalAlignment=HAlign_Left}}
  }
  Button_EnableMFA {
   widgetClassPath=&IconTextButton
   bIsVariable=true
   parent=MFAContainer
   widget=Button_EnableMFA
   properties {
    buttonText=Enable 2FA Now
    contentAlignment=HAlign_Center
    navigation=&self:WidgetTree.Button_EnableMFA.WidgetNavigation_0
    onButtonBaseClicked=(&self:WidgetTree.Button_EnableMFA.HandleOnButtonClicked)
    onButtonBaseDoubleClicked=(&self:WidgetTree.Button_EnableMFA.HandleOnButtonDoubleClicked)
    onButtonBaseHovered=(&self:WidgetTree.Button_EnableMFA.HandleOnButtonHovered)
    onButtonBaseUnhovered=(&self:WidgetTree.Button_EnableMFA.HandleOnButtonUnhovered)
    onSelectedChangedBase=(&self:WidgetTree.Button_EnableMFA.HandleOnSelectedChanged)
    style=/FortUILibrary/Foundation/Buttons/ButtonStyles/TopBar/ButtonStyle-Outline.ButtonStyle-Outline_C
    iconBrush {
     bIsDynamicallyLoaded=false
     drawAs=Image
     imageType=NoImage
     mirroring=NoMirror
     resourceName=None
     resourceObject=/CompeteOldTab/UI/Login/Art/M-Icon-ExternalLink-64
     tiling=NoTile
     imageSize=<0 0>
     margin=<ltrb 0 0 0 0>
     outlineSettings {
      bUseBrushTransparency=false
      roundingType=HalfHeightRadius
      width=0
      color {colorUseRule=UseColor_Specified, specifiedColor=<rgba 0 0 0 0>}
      cornerRadii=<0 0 0 1>
     }
     tintColor {colorUseRule=UseColor_Specified, specifiedColor=<rgba 1 1 1 1>}
     uVRegion {bIsValid=false, max=<0 0>, min=<0 0>}
    }
    triggeringInputAction {dataTable=&InputActionDataTable, rowName=EventDetailsOpenMFA}
   }
   slot {class=VerticalBoxSlot, properties {horizontalAlignment=HAlign_Left, padding=<ltrb 0 10 0 0>}}
  }
  EventLeaderboardSummary {
   widgetClassPath=/CompeteOldTab/UI/Events/Leaderboards/EventLeaderboardSummary
   bIsVariable=true
   parent=VerticalBox_3
   widget=EventLeaderboardSummary
   slot {class=VerticalBoxSlot, properties {padding=<ltrb 20 15 0 0>}}
  }
  CommonWidgetSwitcher_WindowsDetails {
   widgetClassPath=&CommonWidgetSwitcherLegacy
   bIsVariable=true
   parent=VerticalBoxEventContent
   widget=CommonWidgetSwitcher_WindowsDetails
   slot {class=VerticalBoxSlot, properties {size {sizeRule=Fill, value=1}}}
  }
  ScaleBox_4 {
   widgetClassPath=&ScaleBox
   parent=CommonWidgetSwitcher_WindowsDetails
   widget=ScaleBox_4
   properties {stretch=ScaleBySafeZone}
   slot {class=WidgetSwitcherSlot}
  }
  VerticalBoxEventWindows {
   widgetClassPath=&VerticalBox
   parent=ScaleBox_4
   widget=VerticalBoxEventWindows
   slot {class=ScaleBoxSlot, properties {horizontalAlignment=HAlign_Left, verticalAlignment=VAlign_Bottom}}
  }
  VerticalBoxRoundTitle {
   widgetClassPath=&VerticalBox
   parent=VerticalBoxEventWindows
   widget=VerticalBoxRoundTitle
   slot {class=VerticalBoxSlot, properties {padding=<ltrb 45 0 0 0>}}
  }
  TournamentMatchType {
   widgetClassPath=&CommonTextBlock
   bIsVariable=true
   parent=VerticalBoxRoundTitle
   widget=TournamentMatchType
   properties {
    text=SOLO
    font {
     bForceMonospaced=false
     bMaterialIsStencil=false
     fontMaterial=null
     fontObject=&Burbank
     letterSpacing=0
     monospacedWidth=1
     size=60
     skewAmount=0
     typefaceFontName=BurbankRegular-Black
     outlineSettings {
      bApplyOutlineToDropShadows=false
      bMiteredCorners=false
      bSeparateFillAlpha=false
      outlineMaterial=null
      outlineSize=0
      outlineColor=<rgba 0 0 0 1>
     }
    }
    renderTransform {angle=0, scale=<0.9 1>, shear=<-7 0>, translation=<0 0>}
    renderTransformPivot=<0 0.5>
   }
   slot {class=VerticalBoxSlot}
  }
  TournamentRoundText {
   widgetClassPath=&CommonTextBlock
   bIsVariable=true
   parent=VerticalBoxRoundTitle
   widget=TournamentRoundText
   properties {
    text=ROUND 1
    colorAndOpacity {colorUseRule=UseColor_Specified, specifiedColor=<rgba 0.0185 1 0.9047 1>}
    font {
     bForceMonospaced=false
     bMaterialIsStencil=false
     fontMaterial=null
     fontObject=&Burbank
     letterSpacing=0
     monospacedWidth=1
     size=40
     skewAmount=0
     typefaceFontName=BurbankRegular-Black
     outlineSettings {
      bApplyOutlineToDropShadows=false
      bMiteredCorners=false
      bSeparateFillAlpha=false
      outlineMaterial=null
      outlineSize=0
      outlineColor=<rgba 0 0 0 1>
     }
    }
    renderTransform {angle=0, scale=<0.95 1>, shear=<-7 0>, translation=<0 0>}
    renderTransformPivot=<0 0.5>
   }
   slot {class=VerticalBoxSlot, properties {padding=<ltrb 30 -25 0 25>}}
  }
  SizeBox_EventWindowContainer {
   widgetClassPath=&SizeBox
   parent=VerticalBoxEventWindows
   widget=SizeBox_EventWindowContainer
   properties {bOverride_HeightOverride=true, bOverride_WidthOverride=true, heightOverride=215, widthOverride=600}
   slot {class=VerticalBoxSlot, properties {horizontalAlignment=HAlign_Left, padding=<ltrb 300 0 0 0>}}
  }
  EventsScrollBox {
   widgetClassPath=&ScrollBox
   bIsVariable=true
   parent=SizeBox_EventWindowContainer
   widget=EventsScrollBox
   properties {
    bAllowRightClickDragScrolling=false
    clipping=Inherit
    consumeMouseWheel=Never
    navigation=&self:WidgetTree.EventsScrollBox.WidgetNavigation_0
    navigationDestination=Center
    orientation=Orient_Horizontal
    scrollBarVisibility=Collapsed
    visibility=SelfHitTestInvisible
    widgetStyle {
     barThickness=9
     bottomShadowBrush {
      bIsDynamicallyLoaded=false
      drawAs=NoDrawType
      imageType=FullColor
      mirroring=NoMirror
      resourceName=../../../Engine/Content/Slate/Common/ScrollBoxShadowTop.png
      resourceObject=null
      tiling=NoTile
      imageSize=<16 8>
      margin=<ltrb 0.5 1 0.5 0>
      outlineSettings {
       bUseBrushTransparency=false
       roundingType=HalfHeightRadius
       width=0
       color {colorUseRule=UseColor_Specified, specifiedColor=<rgba 0 0 0 0>}
       cornerRadii=<0 0 0 0>
      }
      tintColor {colorUseRule=UseColor_Specified, specifiedColor=<rgba 1 1 1 1>}
      uVRegion {bIsValid=false, max=<0 0>, min=<0 0>}
     }
     horizontalScrolledContentPadding=<ltrb 0 0 1 0>
     leftShadowBrush {
      bIsDynamicallyLoaded=false
      drawAs=NoDrawType
      imageType=FullColor
      mirroring=NoMirror
      resourceName=../../../Engine/Content/Slate/Common/ScrollBoxShadowTop.png
      resourceObject=null
      tiling=NoTile
      imageSize=<16 8>
      margin=<ltrb 0.5 1 0.5 0>
      outlineSettings {
       bUseBrushTransparency=false
       roundingType=HalfHeightRadius
       width=0
       color {colorUseRule=UseColor_Specified, specifiedColor=<rgba 0 0 0 0>}
       cornerRadii=<0 0 0 0>
      }
      tintColor {colorUseRule=UseColor_Specified, specifiedColor=<rgba 1 1 1 1>}
      uVRegion {bIsValid=false, max=<0 0>, min=<0 0>}
     }
     rightShadowBrush {
      bIsDynamicallyLoaded=false
      drawAs=NoDrawType
      imageType=FullColor
      mirroring=NoMirror
      resourceName=../../../Engine/Content/Slate/Common/ScrollBoxShadowTop.png
      resourceObject=null
      tiling=NoTile
      imageSize=<16 8>
      margin=<ltrb 0.5 1 0.5 0>
      outlineSettings {
       bUseBrushTransparency=false
       roundingType=HalfHeightRadius
       width=0
       color {colorUseRule=UseColor_Specified, specifiedColor=<rgba 0 0 0 0>}
       cornerRadii=<0 0 0 0>
      }
      tintColor {colorUseRule=UseColor_Specified, specifiedColor=<rgba 1 1 1 1>}
      uVRegion {bIsValid=false, max=<0 0>, min=<0 0>}
     }
     topShadowBrush {
      bIsDynamicallyLoaded=false
      drawAs=NoDrawType
      imageType=FullColor
      mirroring=NoMirror
      resourceName=../../../Engine/Content/Slate/Common/ScrollBoxShadowTop.png
      resourceObject=null
      tiling=NoTile
      imageSize=<16 8>
      margin=<ltrb 0.5 1 0.5 0>
      outlineSettings {
       bUseBrushTransparency=false
       roundingType=HalfHeightRadius
       width=0
       color {colorUseRule=UseColor_Specified, specifiedColor=<rgba 0 0 0 0>}
       cornerRadii=<0 0 0 0>
      }
      tintColor {colorUseRule=UseColor_Specified, specifiedColor=<rgba 1 1 1 1>}
      uVRegion {bIsValid=false, max=<0 0>, min=<0 0>}
     }
     verticalScrolledContentPadding=<ltrb 0 0 0 1>
    }
   }
   slot {class=SizeBoxSlot, properties {verticalAlignment=VAlign_Bottom}}
  }
  ShowdownEventTile {
   widgetClassPath=/CompeteOldTab/Frontend/Showdown/ShowdownEventTile
   bIsVariable=true
   parent=EventsScrollBox
   widget=ShowdownEventTile
   properties {
    bSelectable=false
    bShouldSelectUponReceivingFocus=true
    onButtonBaseClicked=(&self:WidgetTree.ShowdownEventTile.HandleOnButtonClicked)
    onButtonBaseDoubleClicked=(&self:WidgetTree.ShowdownEventTile.HandleOnButtonDoubleClicked)
    onButtonBaseHovered=(&self:WidgetTree.ShowdownEventTile.HandleOnButtonHovered)
    onButtonBaseUnhovered=(&self:WidgetTree.ShowdownEventTile.HandleOnButtonUnhovered)
    onSelectedChangedBase=(&self:WidgetTree.ShowdownEventTile.HandleOnSelectedChanged)
   }
   slot {class=ScrollBoxSlot, properties {padding=<ltrb 3 0 3 0>}}
  }
  ShowdownEventTile_0 {
   widgetClassPath=&ShowdownEventTile
   bIsVariable=true
   parent=EventsScrollBox
   widget=ShowdownEventTile_0
   properties {
    bSelectable=false
    bShouldSelectUponReceivingFocus=true
    onButtonBaseClicked=(&self:WidgetTree.ShowdownEventTile_0.HandleOnButtonClicked)
    onButtonBaseDoubleClicked=(&self:WidgetTree.ShowdownEventTile_0.HandleOnButtonDoubleClicked)
    onButtonBaseHovered=(&self:WidgetTree.ShowdownEventTile_0.HandleOnButtonHovered)
    onButtonBaseUnhovered=(&self:WidgetTree.ShowdownEventTile_0.HandleOnButtonUnhovered)
    onSelectedChangedBase=(&self:WidgetTree.ShowdownEventTile_0.HandleOnSelectedChanged)
   }
   slot {class=ScrollBoxSlot, properties {padding=<ltrb 3 0 3 0>}}
  }
  ShowdownEventTile_1 {
   widgetClassPath=&ShowdownEventTile
   bIsVariable=true
   parent=EventsScrollBox
   widget=ShowdownEventTile_1
   properties {
    bSelectable=false
    bShouldSelectUponReceivingFocus=true
    onButtonBaseClicked=(&self:WidgetTree.ShowdownEventTile_1.HandleOnButtonClicked)
    onButtonBaseDoubleClicked=(&self:WidgetTree.ShowdownEventTile_1.HandleOnButtonDoubleClicked)
    onButtonBaseHovered=(&self:WidgetTree.ShowdownEventTile_1.HandleOnButtonHovered)
    onButtonBaseUnhovered=(&self:WidgetTree.ShowdownEventTile_1.HandleOnButtonUnhovered)
    onSelectedChangedBase=(&self:WidgetTree.ShowdownEventTile_1.HandleOnSelectedChanged)
   }
   slot {class=ScrollBoxSlot, properties {padding=<ltrb 3 0 3 0>}}
  }
  ShowdownEventTile_2 {
   widgetClassPath=&ShowdownEventTile
   bIsVariable=true
   parent=EventsScrollBox
   widget=ShowdownEventTile_2
   properties {
    bSelectable=false
    bShouldSelectUponReceivingFocus=true
    onButtonBaseClicked=(&self:WidgetTree.ShowdownEventTile_2.HandleOnButtonClicked)
    onButtonBaseDoubleClicked=(&self:WidgetTree.ShowdownEventTile_2.HandleOnButtonDoubleClicked)
    onButtonBaseHovered=(&self:WidgetTree.ShowdownEventTile_2.HandleOnButtonHovered)
    onButtonBaseUnhovered=(&self:WidgetTree.ShowdownEventTile_2.HandleOnButtonUnhovered)
    onSelectedChangedBase=(&self:WidgetTree.ShowdownEventTile_2.HandleOnSelectedChanged)
   }
   slot {class=ScrollBoxSlot, properties {padding=<ltrb 3 0 3 0>}}
  }
  CommonBorderEventDetails {
   widgetClassPath=&CommonBorder
   parent=VerticalBoxEventWindows
   widget=CommonBorderEventDetails
   properties {clipping=ClipToBounds, verticalAlignment=VAlign_Top, desiredSizeScale=<1 0>, padding=<ltrb 0 0 0 0>}
   slot {class=VerticalBoxSlot, properties {verticalAlignment=VAlign_Top}}
  }
  CommonBorder_9 {
   widgetClassPath=&CommonBorder
   parent=CommonBorderEventDetails
   widget=CommonBorder_9
   properties {verticalAlignment=VAlign_Top, padding=<ltrb 0 0 0 0>}
   slot {class=BorderSlot, properties {verticalAlignment=VAlign_Top, padding=<ltrb 0 0 0 0>}}
  }
  Overlay_5 {
   widgetClassPath=&Overlay
   parent=CommonBorder_9
   widget=Overlay_5
   slot {class=BorderSlot, properties {verticalAlignment=VAlign_Top, padding=<ltrb 0 0 0 0>}}
  }
  SizeBox_5 {
   widgetClassPath=&SizeBox
   parent=Overlay_5
   widget=SizeBox_5
   properties {
    bOverride_MaxDesiredHeight=true
    bOverride_MinDesiredHeight=true
    clipping=ClipToBounds
    heightOverride=440
    maxDesiredHeight=505
    minDesiredHeight=505
   }
   slot {class=OverlaySlot, properties {horizontalAlignment=HAlign_Fill}}
  }
  HorizontalBoxEventWindowDetails {
   widgetClassPath=&HorizontalBox
   parent=SizeBox_5
   widget=HorizontalBoxEventWindowDetails
   slot {class=SizeBoxSlot, properties {verticalAlignment=VAlign_Top, padding=<ltrb 0 50 0 80>}}
  }
  TournamentEventWindowDetails {
   widgetClassPath=/CompeteOldTab/Frontend/Showdown/TournamentEventWindowDetails
   bIsVariable=true
   parent=HorizontalBoxEventWindowDetails
   widget=TournamentEventWindowDetails
   properties {
    allowLeaderboardAccess=true
    bTimeRemainingShowSeconds=true
    timeRemainingFormat="This event window starts in {0}. Come back then to test your mettle!"
    timeRemainingFormatType=Approximate
   }
   slot {class=HorizontalBoxSlot, properties {size {sizeRule=Automatic, value=0.8}}}
  }
  HorizontalBox_2 {
   widgetClassPath=&HorizontalBox
   parent=CommonWidgetSwitcher_WindowsDetails
   widget=HorizontalBox_2
   slot {class=WidgetSwitcherSlot}
  }
  VerticalBox_6 {
   widgetClassPath=&VerticalBox
   parent=HorizontalBox_2
   widget=VerticalBox_6
   slot {class=HorizontalBoxSlot, properties {size {sizeRule=Fill, value=0.48}}}
  }
  ScrollBox_DetailAndRules {
   widgetClassPath=&ScrollBox
   bIsVariable=true
   parent=VerticalBox_6
   widget=ScrollBox_DetailAndRules
   properties {scrollWhenFocusChanges=InstantScroll}
   slot {class=VerticalBoxSlot, properties {size {sizeRule=Fill, value=1}}}
  }
  Overlay_1 {widgetClassPath=&Overlay, parent=ScrollBox_DetailAndRules, widget=Overlay_1, slot {class=ScrollBoxSlot}}
  VerticalBox_4 {widgetClassPath=&VerticalBox, parent=Overlay_1, widget=VerticalBox_4, slot {class=OverlaySlot}}
  CommonBorderTournamentDescription {
   widgetClassPath=&CommonBorder
   parent=VerticalBox_4
   widget=CommonBorderTournamentDescription
   properties {clipping=ClipToBounds, desiredSizeScale=<1 0>, padding=<ltrb 40 40 40 40>}
   slot {class=VerticalBoxSlot, properties {size {sizeRule=Fill, value=1}}}
  }
  CommonBorder_7 {
   widgetClassPath=&CommonBorder
   parent=CommonBorderTournamentDescription
   widget=CommonBorder_7
   properties {padding=<ltrb 0 0 40 0>}
   slot {class=BorderSlot, properties {padding=<ltrb 40 40 40 40>}}
  }
  CommonTextBlock_TournamentDescription {
   widgetClassPath=&CommonTextBlock
   bIsVariable=true
   parent=CommonBorder_7
   widget=CommonTextBlock_TournamentDescription
   properties {
    text="Competition begins Lorem ipsum dolor sit amet, "
    wrapTextAt=880
    font {
     bForceMonospaced=false
     bMaterialIsStencil=false
     fontMaterial=null
     fontObject=&Burbank
     letterSpacing=0
     monospacedWidth=1
     size=20
     skewAmount=0
     typefaceFontName=Burbank-PlayerPower
     outlineSettings {
      bApplyOutlineToDropShadows=false
      bMiteredCorners=false
      bSeparateFillAlpha=false
      outlineMaterial=null
      outlineSize=0
      outlineColor=<rgba 0 0 0 1>
     }
    }
   }
   slot {class=BorderSlot, properties {padding=<ltrb 0 0 40 0>}}
  }
  ScaleBox_1 {
   widgetClassPath=&ScaleBox
   parent=VerticalBox_4
   widget=ScaleBox_1
   properties {stretch=ScaleToFitX, stretchDirection=DownOnly}
   slot {class=VerticalBoxSlot}
  }
  OverlayEventDetails {
   widgetClassPath=&Overlay
   parent=ScaleBox_1
   widget=OverlayEventDetails
   slot {class=ScaleBoxSlot, properties {verticalAlignment=VAlign_Fill}}
  }
  SizeBox_3 {
   widgetClassPath=&SizeBox
   parent=OverlayEventDetails
   widget=SizeBox_3
   properties {bOverride_WidthOverride=true, heightOverride=460.3933, widthOverride=1091.4078}
   slot {class=OverlaySlot, properties {horizontalAlignment=HAlign_Fill, verticalAlignment=VAlign_Bottom, padding=<ltrb 30 30 30 30>}}
  }
  CommonBorder_ScoringBG {
   widgetClassPath=&CommonBorder
   bIsVariable=true
   parent=SizeBox_3
   widget=CommonBorder_ScoringBG
   properties {
    background {
     bIsDynamicallyLoaded=false
     drawAs=Image
     imageType=NoImage
     mirroring=NoMirror
     resourceName=None
     resourceObject=null
     tiling=NoTile
     imageSize=<32 32>
     margin=<ltrb 0 0 0 0>
     outlineSettings {
      bUseBrushTransparency=false
      roundingType=HalfHeightRadius
      width=0
      color {colorUseRule=UseColor_Specified, specifiedColor=<rgba 0 0 0 0>}
      cornerRadii=<0 0 0 0>
     }
     tintColor {colorUseRule=UseColor_Specified, specifiedColor=<rgba 1 1 1 1>}
     uVRegion {bIsValid=false, max=<0 0>, min=<0 0>}
    }
    brushColor=<rgba 0.0203 0.063 0.3372 0>
    padding=<ltrb 20 20 20 20>
   }
   slot {class=SizeBoxSlot}
  }
  VerticalBox_0 {
   widgetClassPath=&VerticalBox
   parent=CommonBorder_ScoringBG
   widget=VerticalBox_0
   slot {class=BorderSlot, properties {padding=<ltrb 20 20 20 20>}}
  }
  UniformGridPanel_TournamentScoringRules {
   widgetClassPath=/Script/UMG.UniformGridPanel
   bIsVariable=true
   parent=VerticalBox_0
   widget=UniformGridPanel_TournamentScoringRules
   properties {slotPadding=<ltrb 0 0 7 7>}
   slot {class=VerticalBoxSlot}
  }
  ShowdownTournamentDetailsScoringRule {
   widgetClassPath=/CompeteOldTab/Frontend/Showdown/ShowdownTournamentDetailsScoringRule
   bIsVariable=true
   parent=UniformGridPanel_TournamentScoringRules
   widget=ShowdownTournamentDetailsScoringRule
   slot {class=UniformGridSlot, properties {horizontalAlignment=HAlign_Fill, verticalAlignment=VAlign_Fill}}
  }
  ShowdownTournamentDetailsScoringRule_C_3 {
   widgetClassPath=&ShowdownTournamentDetailsScoringRule
   bIsVariable=true
   parent=UniformGridPanel_TournamentScoringRules
   widget=ShowdownTournamentDetailsScoringRule_C_3
   slot {class=UniformGridSlot, properties {horizontalAlignment=HAlign_Fill, row=1, verticalAlignment=VAlign_Fill}}
  }
  SizeBox_6 {
   widgetClassPath=&SizeBox
   parent=OverlayEventDetails
   widget=SizeBox_6
   properties {
    bOverride_MinDesiredHeight=true
    bOverride_MinDesiredWidth=true
    minDesiredHeight=61
    minDesiredWidth=260
    renderTransform {angle=0, scale=<1 1>, shear=<0 0>, translation=<4 -15>}
   }
   slot {class=OverlaySlot}
  }
  CommonBorder_ScoringTitleBG {
   widgetClassPath=&CommonBorder
   bIsVariable=true
   parent=SizeBox_6
   widget=CommonBorder_ScoringTitleBG
   properties {
    verticalAlignment=VAlign_Center
    background {
     bIsDynamicallyLoaded=false
     drawAs=Image
     imageType=NoImage
     mirroring=NoMirror
     resourceName=None
     resourceObject=/Game/UI/Frontend/Showdown/Materials/M_UI_SkewBox_TitleCut
     tiling=NoTile
     imageSize=<176 64>
     margin=<ltrb 0.9558 0.5752 0.0442 0.292>
     outlineSettings {
      bUseBrushTransparency=false
      roundingType=HalfHeightRadius
      width=0
      color {colorUseRule=UseColor_Specified, specifiedColor=<rgba 0 0 0 0>}
      cornerRadii=<0 0 0 0>
     }
     tintColor {colorUseRule=UseColor_Specified, specifiedColor=<rgba 1 1 1 1>}
     uVRegion {bIsValid=false, max=<0 0>, min=<0 0>}
    }
    brushColor=<rgba 0 0.2086 1 1>
    padding=<ltrb 33 10 33 10>
   }
   slot {class=SizeBoxSlot}
  }
  CommonTextBlock_8 {
   widgetClassPath=&CommonTextBlock
   parent=CommonBorder_ScoringTitleBG
   widget=CommonTextBlock_8
   properties {
    text=SCORING
    font {
     bForceMonospaced=false
     bMaterialIsStencil=false
     fontMaterial=null
     fontObject=&Burbank
     letterSpacing=0
     monospacedWidth=1
     size=24
     skewAmount=0
     typefaceFontName=BurbankRegular-Black
     outlineSettings {
      bApplyOutlineToDropShadows=false
      bMiteredCorners=false
      bSeparateFillAlpha=false
      outlineMaterial=null
      outlineSize=0
      outlineColor=<rgba 0 0 0 1>
     }
    }
   }
   slot {class=BorderSlot, properties {verticalAlignment=VAlign_Center, padding=<ltrb 33 10 33 10>}}
  }
  UnlockPinContent {
   widgetClassPath=&HorizontalBox
   bIsVariable=true
   parent=VerticalBox_6
   widget=UnlockPinContent
   slot {class=VerticalBoxSlot, properties {horizontalAlignment=HAlign_Center}}
  }
  CommonTextBlock_UnlockCriteria {
   widgetClassPath=&CommonTextBlock
   bIsVariable=true
   parent=UnlockPinContent
   widget=CommonTextBlock_UnlockCriteria
   properties {
    justification=Center
    text=UNLOCK PIN AT
    colorAndOpacity {colorUseRule=UseColor_Specified, specifiedColor=<rgba 0.0185 1 0.9047 1>}
    font {
     bForceMonospaced=false
     bMaterialIsStencil=false
     fontMaterial=null
     fontObject=&Burbank
     letterSpacing=0
     monospacedWidth=1
     size=24
     skewAmount=0
     typefaceFontName=BurbankRegular-Black
     outlineSettings {
      bApplyOutlineToDropShadows=false
      bMiteredCorners=false
      bSeparateFillAlpha=false
      outlineMaterial=null
      outlineSize=0
      outlineColor=<rgba 0 0 0 1>
     }
    }
   }
   slot {class=HorizontalBoxSlot, properties {horizontalAlignment=HAlign_Center, verticalAlignment=VAlign_Center}}
  }
  Text_PinUnlockScore {
   widgetClassPath=&CommonTextBlock
   bIsVariable=true
   parent=UnlockPinContent
   widget=Text_PinUnlockScore
   properties {
    justification=Center
    text=45
    font {
     bForceMonospaced=false
     bMaterialIsStencil=false
     fontMaterial=null
     fontObject=&Burbank
     letterSpacing=0
     monospacedWidth=1
     size=50
     skewAmount=0
     typefaceFontName=Burbank-Black
     outlineSettings {
      bApplyOutlineToDropShadows=false
      bMiteredCorners=false
      bSeparateFillAlpha=false
      outlineMaterial=null
      outlineSize=0
      outlineColor=<rgba 0 0 0 1>
     }
    }
   }
   slot {class=HorizontalBoxSlot, properties {horizontalAlignment=HAlign_Center, padding=<ltrb 20 -8 20 -10>}}
  }
  Spacer_67 {
   widgetClassPath=&Spacer
   parent=HorizontalBox_2
   widget=Spacer_67
   slot {class=HorizontalBoxSlot, properties {size {sizeRule=Fill, value=0.5}}}
  }
  EventLeaderboardScreen {
   widgetClassPath=/CompeteOldTab/UI/Events/Leaderboards/EventLeaderboardScreen
   bIsVariable=true
   parent=CommonWidgetSwitcher_FullScreenSwitch
   widget=EventLeaderboardScreen
   slot {class=WidgetSwitcherSlot}
  }
  TournamentPayoutModal {
   widgetClassPath=/CompeteOldTab/UI/Events/TournamentPayoutModal
   bIsVariable=true
   parent=CommonWidgetSwitcher_FullScreenSwitch
   widget=TournamentPayoutModal
   slot {class=WidgetSwitcherSlot}
  }
  TournamentWatchList {
   widgetClassPath=/CompeteOldTab/UI/Events/TournamentLiveGames
   bIsVariable=true
   parent=CommonWidgetSwitcher_FullScreenSwitch
   widget=TournamentWatchList
   slot {class=WidgetSwitcherSlot}
  }
  Text_TimeRemaining {widgetClassPath=&CommonTextBlock, bInherited=true, widget=null}
  Text_DebugId {widgetClassPath=&TextBlock, bInherited=true, widget=null, properties=~}
  Image_PosterFront {widgetClassPath=/Script/EpicCMSUIFramework.EpicCMSImage, bInherited=true, widget=null}
  Image_PosterBack {widgetClassPath=&EpicCMSImage, bInherited=true, widget=null}
  Image_LoadingScreen {widgetClassPath=&EpicCMSImage, bInherited=true, widget=null}
  Image_PlaylistTile {widgetClassPath=&EpicCMSImage, bInherited=true, widget=null}
 ]
}
