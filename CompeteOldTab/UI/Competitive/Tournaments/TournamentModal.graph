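# Tournament modal widget, derived from FortTournamentModal, that hosts the tournament details page over a full-screen background in the competitive UI.
# An overlay stacks the background image beneath the inherited TournamentDetailsPage.
# On activation it plays the Intro animation, which fades the background in.
# When a tournament is set, it writes the event's left and right colors into the background's dynamic material to tint it per tournament.
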WidgetBlueprint /CompeteOldTab/UI/Competitive/Tournaments/TournamentModal
animations [Intro {
  binding_count=1
  display_label=""
  length_seconds=0.9532
  object=&self:Intro
  sequence {bindings [Image_BG {
     id=#1
     possessed_class=Image
     tracks [{
       type=MovieSceneFloatTrack
       display_name=Render Opacity
       sections [{type=MovieSceneFloatSection, end_frame=19, start_frame=0, channels [None {type=MovieSceneScriptingFloatChannel, keys [[0,0],[19.0643,1]]}]}]
      }]
    }]}
 }]
graphs [EventGraph {nodes [
   Event On Activated()
   > PlayAnimation(InAnimation=Get Intro())
   n1 = Event OnTournamentSet!() -> (OutputDelegate, then, LeftColor:Linear Color Structure, RightColor:Linear Color Structure) {is=title K2Node_Event}
   > n2 = GetDynamicMaterial(self=Get Image_BG() -> (Image_BG:Image Object Reference))
   > SetVectorParameterValue(self=@n2, ParameterName=Color1, Value=@n1.LeftColor)
   > SetVectorParameterValue(self=@n2, ParameterName=Color2, Value=@n1.RightColor)
  ]}]
variables []
widgets {
 info {inheritedWidgetCount=1, parentClass=/Script/FortniteUI.FortTournamentModal, rootWidgetClass=&Overlay, widgetCount=4}
 widgets [
  Overlay_0 {widgetClassPath=&Overlay, widget=Overlay_0}
  Image_BG {
   widgetClassPath=&Image
   bIsVariable=true
   parent=Overlay_0
   widget=Image_BG
   properties {
    visibility=SelfHitTestInvisible
    brush {
     bIsDynamicallyLoaded=false
     drawAs=Image
     imageType=NoImage
     mirroring=NoMirror
     resourceName=None
     resourceObject=/CompeteOldTab/Frontend/Showdown/Materials/M_UI_TournamentDetailsBG_Inst
     tiling=NoTile
     imageSize=<32 32>
     margin=<ltrb 0 0 0 0>
     outlineSettings {
      bUseBrushTransparency=false
      roundingType=HalfHeightRadius
      width=0
      color {colorUseRule=UseColor_Specified, specifiedColor=<rgba 0 0 0 0>}
      cornerRadii=<0 0 0 0>
     }
     tintColor {colorUseRule=UseColor_Specified, specifiedColor=<rgba 1 1 1 1>}
     uVRegion {bIsValid=false, max=<0 0>, min=<0 0>}
    }
   }
   slot {class=OverlaySlot, properties {horizontalAlignment=HAlign_Fill, verticalAlignment=VAlign_Fill}}
  }
  OverlayMain {
   widgetClassPath=&Overlay
   parent=Overlay_0
   widget=OverlayMain
   properties {navigation=&self:WidgetTree.OverlayMain.WidgetNavigation_0}
   slot {class=OverlaySlot, properties {horizontalAlignment=HAlign_Fill, verticalAlignment=VAlign_Fill}}
  }
  TournamentDetailsPage {
   widgetClassPath=&/TournamentDetails
   bInherited=true
   bIsVariable=true
   parent=OverlayMain
   widget=TournamentDetailsPage
   slot {class=OverlaySlot, properties {horizontalAlignment=HAlign_Fill, verticalAlignment=VAlign_Fill}}
  }
 ]
}
